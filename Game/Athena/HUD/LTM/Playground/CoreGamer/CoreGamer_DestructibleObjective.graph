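# HUD widget for a destructible objective in Playground, showing its icon, team color and progress.
# When the tracked objective is set, it async loads a preset icon from a soft texture array and tints the backplate material friendly or enemy by local team.
# Capture and decapture events drive the material's fill, playing a scale pulse whenever progress hits a new low.
# At zero, the mask turns on and the widget dims.
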WidgetBlueprint /Game/Athena/HUD/LTM/Playground/CoreGamer/CoreGamer_DestructibleObjective
animations [DamagedAnimation {
  binding_count=1
  length_seconds=0.2
  object=&self:DamagedAnimation
  sequence {bindings [MainContainer {
     id=#1
     possessed_class=Overlay
     tracks [{
       type=MovieScene2DTransformTrack
       sections [{
         type=MovieScene2DTransformSection
         end_frame=4
         start_frame=0
         channels [
          Scale.X {type=MovieSceneScriptingFloatChannel, keys [[0,1],[1,1.25],[3,0.75],[4,1]]}
          Scale.Y {type=MovieSceneScriptingFloatChannel, keys [[0,1],[1,1.25],[3,0.75],[4,1]]}
         ]
        }]
      }]
    }]}
 }]
graphs [
 EventGraph {nodes [
   n1 = LoadIndicatorIcon!() -> (OutputDelegate, then, Index:int) {is=title K2Node_CustomEvent}
   > n2 = Branch(Condition="integer > integer"(A=@n1.Index, B=-1))
   > n3 = Async Load Asset(Asset="Get (a copy)"(
     Array:Array of Texture 2D Soft Object References=Get ObjectiveIndicatorIcons() -> (ObjectiveIndicatorIcons:Array of Texture 2D Soft Object References)
     Dimension 1=int + int(A=@n1.Index)
    ) -> (Output:Texture 2D Soft Object Reference))
   n4 = Cast To Texture2D(execute=@n3.Completed, Object=@n3)
   > SetBrush(
    self=Get IconImage() -> (IconImage:Image Object Reference)
    InBrush=MakeBrushFromTexture!(Texture:Texture 2D Object Reference=@n4, Width:int=24, Height:int=24) -> (ReturnValue:Slate Brush Structure)
   )
   > SetVisibility(self=Get IconImage(), InVisibility=SelfHitTestInvisible)
   SetVisibility(execute=@n2.else, self=Get IconImage(), InVisibility=Collapsed)
   n5 = Event OnTrackedObjectiveSet!() -> (OutputDelegate, then, NewTrackedObjective:"Tracked Objective Structure (by ref)") {is=title K2Node_Event}
   > Sequence()
   n6 = Break Tracked Objective!(TrackedObjective:"Tracked Objective Structure (by ref)"=@n5.NewTrackedObjective) -> (
    TrackedIndex:int
    TeamIndex:Byte
    OldTeamIndex:Byte
    IndicatorPresetIndex:int
    ObjectiveSortOrder:int
    ProgressPercentage:float=~
    TrackedState:Byte
    TrackedStateOwnerTeam:Byte
    OldTrackedState:Byte
    ObjectiveType:Byte
    bIsInteractionAllowed:bool=~
    bOldIsInteractionAllowed:bool=~
   ) {is=title K2Node_BreakStruct}
   > LoadIndicatorIcon!(execute, self, Index:int=@n6.IndicatorPresetIndex) -> (then) {is=title K2Node_CallFunction}
   > n7 = GetDynamicMaterial(self=Get Image_BackplateProgress() -> (Image_BackplateProgress:Image Object Reference))
   > SetVectorParameterValue(
    self=@n7
    ParameterName=FillColor
    Value=SelectColor(
     A=Get FriendlyColor() -> (FriendlyColor:Linear Color Structure)
     B=Get EnemyColor() -> (EnemyColor:Linear Color Structure)
     bPickA=IsLocalTeam!(InTeamByte:Byte=@n6.TeamIndex) -> (IsLocal:bool)
    )
   )
   K2Node_Event_4 = UpdateProgress!() -> (OutputDelegate, then, CapturingTeam:Byte, Percentage:float=~) {is=title K2Node_CustomEvent}
   > n8 = GetDynamicMaterial(self=Get Image_BackplateProgress())
   > SetScalarParameterValue(self=@n8, ParameterName=PercentFill, Value=@K2Node_Event_4.Percentage)
   > n9 = Sequence()
   > Branch(Condition="float < float"(A=@K2Node_Event_4.Percentage, B=Get CachedPercentage() -> (CachedPercentage:double)))
   > Set CachedPercentage(CachedPercentage:double=@K2Node_Event_4.Percentage) -> (Output_Get:double)
   > PlayAnimation(InAnimation=Get DamagedAnimation() -> (DamagedAnimation:Widget Animation Object Reference))
   n10 = "float <= float"(A=@K2Node_Event_4.Percentage) -> (ReturnValue=~)
   SetScalarParameterValue(execute=@n9.then_1, self=@n8, ParameterName=UseMask, Value=Select(Option 0:float=0, Option 1:float=1, Index=@n10) -> (ReturnValue:float))
   > SetRenderOpacity(self=Get MainContainer() -> (MainContainer:Overlay Object Reference), InOpacity=Select(Option 0=1, Option 1=0.5, Index=@n10))
   n11 = Event OnObjectiveBeingCaptured!() -> (OutputDelegate, then, CapturingTeam:Byte=0, Percentage:float) {is=title K2Node_Event}
   > UpdateProgress!(execute, self, CapturingTeam:Byte=@n11.CapturingTeam, Percentage:float=@n11.Percentage) -> (then) {is=title K2Node_CallFunction}
   n12 = Event OnObjectiveBeingDecaptured!() -> (OutputDelegate, then, DecapturingTeam:Byte=0, Percentage:float) {is=title K2Node_Event}
   > UpdateProgress(CapturingTeam=@n12.DecapturingTeam, Percentage=@n12.Percentage)
  ]}
 IsLocalTeam {nodes [
   n1 = IsLocalTeam!() -> (then, InTeamByte:Byte) {is=title K2Node_FunctionEntry}
   > n2 = Cast To FortPlayerController(Object=GetOwningPlayer())
   > Return Node!(
    execute
    IsLocal:bool="Equal (Byte)"(A:Byte=@n1, B:Byte=GetActorTeam!(Actor:Actor Object Reference=@n2) -> (ReturnValue:Byte=0)) -> (ReturnValue:bool=~) {is=title K2Node_PromotableOperator}
   )
   Return Node(execute=@n2.CastFailed, IsLocal=false)
  ]}
]
variables [
 CachedPercentage {type=number, default=9999}
 EnemyColor {type=LinearColor, default=<rgba 0.6078 0 0 1>}
 FriendlyColor {type=LinearColor, default=<rgba 0 0.1266 0.6823 1>}
 NeutralColor {type=LinearColor, default=<rgba 0.05 0.05 0.05 1>}
 ObjectiveIndicatorIcons {
  type="array<Texture2D>"
  default [
   /Game/Athena/UI/InGame/Creative/PropertyEditor/Icon_Picker/Textures/T_UI_IconPicker_NONE
   /Game/Athena/UI/InGame/Creative/PropertyEditor/Icon_Picker/Textures/64x64/T_UI_IconLibrary_A_64
   /Game/Athena/UI/InGame/Creative/PropertyEditor/Icon_Picker/Textures/64x64/T_UI_IconLibrary_B_64
   /Game/Athena/UI/InGame/Creative/PropertyEditor/Icon_Picker/Textures/64x64/T_UI_IconLibrary_C_64
   /Game/Athena/UI/InGame/Creative/PropertyEditor/Icon_Picker/Textures/64x64/T_UI_IconLibrary_D_64
   /Game/Athena/UI/InGame/Creative/PropertyEditor/Icon_Picker/Textures/64x64/T_UI_IconLibrary_W_64
   /Game/Athena/UI/InGame/Creative/PropertyEditor/Icon_Picker/Textures/64x64/T_UI_IconLibrary_X_64
   /Game/Athena/UI/InGame/Creative/PropertyEditor/Icon_Picker/Textures/64x64/T_UI_IconLibrary_Y_64
   /Game/Athena/UI/InGame/Creative/PropertyEditor/Icon_Picker/Textures/64x64/T_UI_IconLibrary_Z_64
   /Game/Athena/UI/InGame/Creative/PropertyEditor/Icon_Picker/Textures_V2/64x64/T_UI_IconLibrary_AMarker_64
   /Game/Athena/UI/InGame/Creative/PropertyEditor/Icon_Picker/Textures_V2/64x64/T_UI_IconLibrary_BMarker_64
   /Game/Athena/UI/InGame/Creative/PropertyEditor/Icon_Picker/Textures_V2/64x64/T_UI_IconLibrary_CMarker_64
   /Game/Athena/UI/InGame/Creative/PropertyEditor/Icon_Picker/Textures_V2/64x64/T_UI_IconLibrary_DMarker_64
   /Game/Athena/UI/InGame/Creative/PropertyEditor/Icon_Picker/Textures_V2/64x64/T_UI_IconLibrary_WMarker_64
   /Game/Athena/UI/InGame/Creative/PropertyEditor/Icon_Picker/Textures_V2/64x64/T_UI_IconLibrary_XMarker_64
   /Game/Athena/UI/InGame/Creative/PropertyEditor/Icon_Picker/Textures_V2/64x64/T_UI_IconLibrary_YMarker_64
   /Game/Athena/UI/InGame/Creative/PropertyEditor/Icon_Picker/Textures_V2/64x64/T_UI_IconLibrary_ZMarker_64
   /Game/Athena/UI/InGame/Creative/PropertyEditor/Icon_Picker/Textures/64x64/T_UI_IconLibrary_0_64
   /Game/Athena/UI/InGame/Creative/PropertyEditor/Icon_Picker/Textures/64x64/T_UI_IconLibrary_1_64
   /Game/Athena/UI/InGame/Creative/PropertyEditor/Icon_Picker/Textures/64x64/T_UI_IconLibrary_2_64
   /Game/Athena/UI/InGame/Creative/PropertyEditor/Icon_Picker/Textures/64x64/T_UI_IconLibrary_3_64
   /Game/Athena/UI/InGame/Creative/PropertyEditor/Icon_Picker/Textures/64x64/T_UI_IconLibrary_4_64
   /Game/Athena/UI/InGame/Creative/PropertyEditor/Icon_Picker/Textures/64x64/T_UI_IconLibrary_5_64
   /Game/Athena/UI/InGame/Creative/PropertyEditor/Icon_Picker/Textures/64x64/T_UI_IconLibrary_6_64
   /Game/Athena/UI/InGame/Creative/PropertyEditor/Icon_Picker/Textures/64x64/T_UI_IconLibrary_7_64
   /Game/Athena/UI/InGame/Creative/PropertyEditor/Icon_Picker/Textures/64x64/T_UI_IconLibrary_8_64
   /Game/Athena/UI/InGame/Creative/PropertyEditor/Icon_Picker/Textures/64x64/T_UI_IconLibrary_9_64
   /Game/Athena/UI/InGame/Creative/PropertyEditor/Icon_Picker/Textures/64x64/T_UI_IconLibrary_droplet_64
   /Game/Athena/UI/InGame/Creative/PropertyEditor/Icon_Picker/Textures/64x64/T_UI_IconLibrary_fishing_64
   /Game/Athena/UI/InGame/Creative/PropertyEditor/Icon_Picker/Textures/64x64/T_UI_IconLibrary_trophy_64
   /Game/Athena/UI/InGame/Creative/PropertyEditor/Icon_Picker/MilitaryTextures/T_64x64/T_UI_IconLibrary_pistol_64
   /Game/Athena/UI/InGame/Creative/PropertyEditor/Icon_Picker/MilitaryTextures/T_64x64/T_UI_IconLibrary_assaultrifle_64
   /Game/Athena/UI/InGame/Creative/PropertyEditor/Icon_Picker/MilitaryTextures/T_64x64/T_UI_IconLibrary_sniperrifle_64
   /Game/Athena/UI/InGame/Creative/PropertyEditor/Icon_Picker/MilitaryTextures/T_64x64/T_UI_IconLibrary_rocketlauncher_64
   /Game/Athena/UI/InGame/Creative/PropertyEditor/Icon_Picker/MilitaryTextures/T_64x64/T_UI_IconLibrary_grenade_64
   /Game/Athena/UI/InGame/Creative/PropertyEditor/Icon_Picker/Textures/64x64/T_UI_IconLibrary_wolf_64
   /Game/Athena/UI/InGame/Creative/PropertyEditor/Icon_Picker/Textures/64x64/T_UI_IconLibrary_raptor_64
   /Game/Athena/UI/InGame/Creative/PropertyEditor/Icon_Picker/Textures/64x64/T_UI_IconLibrary_chicken_64
   /Game/Athena/UI/InGame/Creative/PropertyEditor/Icon_Picker/Textures/64x64/T_UI_IconLibrary_frog_64
   /Game/Athena/UI/InGame/Creative/PropertyEditor/Icon_Picker/Textures/64x64/T_UI_IconLibrary_boar_64
   /Game/Athena/UI/InGame/Creative/PropertyEditor/Icon_Picker/Textures/64x64/T_UI_IconLibrary_fish_64
   /Game/Athena/UI/InGame/Creative/PropertyEditor/Icon_Picker/Textures/64x64/T_UI_IconLibrary_exclamation_64
   /Game/Athena/UI/InGame/Creative/PropertyEditor/Icon_Picker/Textures/64x64/T_UI_IconLibrary_checkmark_64
   /Game/Athena/UI/InGame/Creative/PropertyEditor/Icon_Picker/Textures/64x64/T_UI_IconLibrary_questionmark_64
   /Game/Athena/UI/InGame/Creative/PropertyEditor/Icon_Picker/Textures/64x64/T_UI_IconLibrary_coins_64
   /Game/Athena/UI/InGame/Creative/PropertyEditor/Icon_Picker/Textures/64x64/T_UI_IconLibrary_moneybag_64
   /Game/Athena/UI/InGame/Creative/PropertyEditor/Icon_Picker/Textures/64x64/T_UI_IconLibrary_heart_64
   /Game/Athena/UI/InGame/Creative/PropertyEditor/Icon_Picker/Textures/64x64/T_UI_IconLibrary_hand_64
   /Game/Athena/UI/InGame/Creative/PropertyEditor/Icon_Picker/Textures/64x64/T_UI_IconLibrary_fist_64
   /Game/Athena/UI/InGame/Creative/PropertyEditor/Icon_Picker/Textures/64x64/T_UI_IconLibrary_flame_64
   /Game/Athena/UI/InGame/Creative/PropertyEditor/Icon_Picker/Textures/64x64/T_UI_IconLibrary_radiation_64
   /Game/Athena/UI/InGame/Creative/PropertyEditor/Icon_Picker/Textures/64x64/T_UI_IconLibrary_sword_64
   /Game/Athena/UI/InGame/Creative/PropertyEditor/Icon_Picker/Textures/64x64/T_UI_IconLibrary_shield_64
   /Game/Athena/UI/InGame/Creative/PropertyEditor/Icon_Picker/Textures/64x64/T_UI_IconLibrary_shoppingcart_64
   /Game/Athena/UI/InGame/Creative/PropertyEditor/Icon_Picker/Textures/64x64/T_UI_IconLibrary_car_64
   /Game/Athena/UI/InGame/Creative/PropertyEditor/Icon_Picker/Textures/64x64/T_UI_IconLibrary_truck_64
   /Game/Athena/UI/InGame/Creative/PropertyEditor/Icon_Picker/Textures/64x64/T_UI_IconLibrary_helicopter_64
   /Game/Athena/UI/InGame/Creative/PropertyEditor/Icon_Picker/Textures/64x64/T_UI_IconLibrary_biplane_64
   /Game/Athena/UI/InGame/Creative/PropertyEditor/Icon_Picker/Textures/64x64/T_UI_IconLibrary_boat_64
   /Game/Athena/UI/InGame/Creative/PropertyEditor/Icon_Picker/Textures/64x64/T_UI_IconLibrary_package_64
   /Game/Athena/UI/InGame/Creative/PropertyEditor/Icon_Picker/MilitaryTextures/T_64x64/T_UI_IconLibrary_magazine_64
   /Game/Athena/UI/InGame/Creative/PropertyEditor/Icon_Picker/Textures/64x64/T_UI_IconLibrary_medkit_64
   /Game/Athena/UI/InGame/Creative/PropertyEditor/Icon_Picker/Textures/64x64/T_UI_IconLibrary_bed_64
   /Game/Athena/UI/InGame/Creative/PropertyEditor/Icon_Picker/Textures/64x64/T_UI_IconLibrary_garage_64
   /Game/Athena/UI/InGame/Creative/PropertyEditor/Icon_Picker/Textures/64x64/T_UI_IconLibrary_biohazard_64
   /Game/Athena/UI/InGame/Creative/PropertyEditor/Icon_Picker/Textures/64x64/T_UI_IconLibrary_recycle_64
   /Game/Athena/UI/InGame/Creative/PropertyEditor/Icon_Picker/Textures/64x64/T_UI_IconLibrary_signal_64
   /Game/Athena/UI/InGame/Creative/PropertyEditor/Icon_Picker/Textures/64x64/T_UI_IconLibrary_signpost_64
   /Game/Athena/UI/InGame/Creative/PropertyEditor/Icon_Picker/Textures/64x64/T_UI_IconLibrary_happy_64
   /Game/Athena/UI/InGame/Creative/PropertyEditor/Icon_Picker/Textures/64x64/T_UI_IconLibrary_sad_64
   /Game/Athena/UI/InGame/Creative/PropertyEditor/Icon_Picker/Textures/64x64/T_UI_IconLibrary_cool_64
   /Game/Athena/UI/InGame/Creative/PropertyEditor/Icon_Picker/Textures/64x64/T_UI_IconLibrary_surprised_64
   /Game/Athena/UI/InGame/Creative/PropertyEditor/Icon_Picker/Textures/64x64/T_UI_IconLibrary_tongueout_64
   /Game/Athena/UI/InGame/Creative/PropertyEditor/Icon_Picker/Textures/64x64/T_UI_IconLibrary_wink_64
   /Game/Athena/UI/InGame/Creative/PropertyEditor/Icon_Picker/Textures/64x64/T_UI_IconLibrary_crying_64
   /Game/Athena/UI/InGame/Creative/PropertyEditor/Icon_Picker/Textures/64x64/T_UI_IconLibrary_running_64
   /Game/Athena/UI/InGame/Creative/PropertyEditor/Icon_Picker/Textures/64x64/T_UI_IconLibrary_sprint_64
   /Game/Athena/UI/InGame/Creative/PropertyEditor/Icon_Picker/Textures/64x64/T_UI_IconLibrary_tree_64
   /Game/Athena/UI/InGame/Creative/PropertyEditor/Icon_Picker/Textures/64x64/T_UI_IconLibrary_forest_64
   /Game/Athena/UI/InGame/Creative/PropertyEditor/Icon_Picker/Textures/64x64/T_UI_IconLibrary_anchor_64
   /Game/Athena/UI/InGame/Creative/PropertyEditor/Icon_Picker/Textures/64x64/T_UI_IconLibrary_wheel_64
   /Game/Athena/UI/InGame/Creative/PropertyEditor/Icon_Picker/Textures/64x64/T_UI_IconLibrary_steeringwheel_64
   /Game/Athena/UI/InGame/Creative/PropertyEditor/Icon_Picker/Textures/64x64/T_UI_IconLibrary_speedometer_64
   /Game/Athena/UI/InGame/Creative/PropertyEditor/Icon_Picker/Textures/64x64/T_UI_IconLibrary_campfire_64
   /Game/Athena/UI/InGame/Creative/PropertyEditor/Icon_Picker/Textures/64x64/T_UI_IconLibrary_cave_64
   /Game/Athena/UI/InGame/Creative/PropertyEditor/Icon_Picker/Textures/64x64/T_UI_IconLibrary_house_64
   /Game/Athena/UI/InGame/Creative/PropertyEditor/Icon_Picker/Textures/64x64/T_UI_IconLibrary_information_64
   /Game/Athena/UI/InGame/Creative/PropertyEditor/Icon_Picker/Textures/64x64/T_UI_IconLibrary_medical_64
   /Game/Athena/UI/InGame/Creative/PropertyEditor/Icon_Picker/Textures/64x64/T_UI_IconLibrary_star_64
   /Game/Athena/UI/InGame/Creative/PropertyEditor/Icon_Picker/Textures/64x64/T_UI_IconLibrary_door_64
   /Game/Athena/UI/InGame/Creative/PropertyEditor/Icon_Picker/Textures/64x64/T_UI_IconLibrary_flag_64
   /Game/Athena/UI/InGame/Creative/PropertyEditor/Icon_Picker/Textures/64x64/T_UI_IconLibrary_chequeredflag_64
   /Game/Athena/UI/InGame/Creative/PropertyEditor/Icon_Picker/Textures/64x64/T_UI_IconLibrary_waypoint_64
   /Game/Athena/UI/InGame/Creative/PropertyEditor/Icon_Picker/Textures/64x64/T_UI_IconLibrary_marker_64
   /Game/Athena/UI/InGame/Creative/PropertyEditor/Icon_Picker/Textures/64x64/T_UI_IconLibrary_padlock_64
   /Game/Athena/UI/InGame/Creative/PropertyEditor/Icon_Picker/Textures/64x64/T_UI_IconLibrary_unlocked_64
   /Game/Athena/UI/InGame/Creative/PropertyEditor/Icon_Picker/Textures/64x64/T_UI_IconLibrary_key_64
   /Game/Athena/UI/InGame/Creative/PropertyEditor/Icon_Picker/Textures/64x64/T_UI_IconLibrary_gaspump_64
   /Game/Athena/UI/InGame/Creative/PropertyEditor/Icon_Picker/Textures/64x64/T_UI_IconLibrary_danger_64
   /Game/Athena/UI/InGame/Creative/PropertyEditor/Icon_Picker/Textures/64x64/T_UI_IconLibrary_enemy_64
   /Game/Athena/UI/InGame/Creative/PropertyEditor/Icon_Picker/Textures_V2/64x64/T_UI_IconLibrary_Pawprint_64
   /Game/Athena/UI/InGame/Creative/PropertyEditor/Icon_Picker/Textures_V2/64x64/T_UI_IconLibrary_PetFoodBowl_64
   /Game/Athena/UI/InGame/Creative/PropertyEditor/Icon_Picker/Textures_V2/64x64/T_UI_IconLibrary_LeftArrow_64
   /Game/Athena/UI/InGame/Creative/PropertyEditor/Icon_Picker/Textures_V2/64x64/T_UI_IconLibrary_RightArrow_64
   /Game/Athena/UI/InGame/Creative/PropertyEditor/Icon_Picker/Textures_V2/64x64/T_UI_IconLibrary_UpArrow_64
   /Game/Athena/UI/InGame/Creative/PropertyEditor/Icon_Picker/Textures_V2/64x64/T_UI_IconLibrary_DownArrow_64
   /Game/Athena/UI/InGame/Creative/PropertyEditor/Icon_Picker/Textures_V2/64x64/T_UI_IconLibrary_SpeechBubble_64
   /Game/Athena/UI/InGame/Creative/PropertyEditor/Icon_Picker/Textures_V2/64x64/T_UI_IconLibrary_List_64
   /Game/Athena/UI/InGame/Creative/PropertyEditor/Icon_Picker/Textures_V2/64x64/T_UI_IconLibrary_OfficeBuilding_64
   /Game/Athena/UI/InGame/Creative/PropertyEditor/Icon_Picker/Textures_V2/64x64/T_UI_IconLibrary_ApartmentBuilding_64
   /Game/Athena/UI/InGame/Creative/PropertyEditor/Icon_Picker/Textures_V2/64x64/T_UI_IconLibrary_Education_64
   /Game/Athena/UI/InGame/Creative/PropertyEditor/Icon_Picker/Textures_V2/64x64/T_UI_IconLibrary_Museum_64
   /Game/Athena/UI/InGame/Creative/PropertyEditor/Icon_Picker/Textures_V2/64x64/T_UI_IconLibrary_Briefcase_64
   /Game/Athena/UI/InGame/Creative/PropertyEditor/Icon_Picker/Textures_V2/64x64/T_UI_IconLibrary_Store_64
   /Game/Athena/UI/InGame/Creative/PropertyEditor/Icon_Picker/Textures_V2/64x64/T_UI_IconLibrary_Compass_64
   /Game/Athena/UI/InGame/Creative/PropertyEditor/Icon_Picker/Textures_V2/64x64/T_UI_IconLibrary_Alien_64
   /Game/Athena/UI/InGame/Creative/PropertyEditor/Icon_Picker/Textures_V2/64x64/T_UI_IconLibrary_MusicNote_64
   /Game/Athena/UI/InGame/Creative/PropertyEditor/Icon_Picker/Textures_V2/64x64/T_UI_IconLibrary_Binoculars_64
   /Game/Athena/UI/InGame/Creative/PropertyEditor/Icon_Picker/Textures_V2/64x64/T_UI_IconLibrary_UFO_64
   /Game/Athena/UI/InGame/Creative/PropertyEditor/Icon_Picker/Textures_V2/64x64/T_UI_IconLibrary_Football_64
   /Game/Athena/UI/InGame/Creative/PropertyEditor/Icon_Picker/Textures_V2/64x64/T_UI_IconLibrary_Arena_64
   /Game/Athena/UI/InGame/Creative/PropertyEditor/Icon_Picker/Textures_V2/64x64/T_UI_IconLibrary_Bridge_64
   /Game/Athena/UI/InGame/Creative/PropertyEditor/Icon_Picker/Textures_V2/64x64/T_UI_IconLibrary_Gravestone_64
   /Game/Athena/UI/InGame/Creative/PropertyEditor/Icon_Picker/Textures_V2/64x64/T_UI_IconLibrary_Hospital_64
   /Game/Athena/UI/InGame/Creative/PropertyEditor/Icon_Picker/Textures_V2/64x64/T_UI_IconLibrary_PoliceBadge_64
   /Game/Athena/UI/InGame/Creative/PropertyEditor/Icon_Picker/Textures_V2/64x64/T_UI_IconLibrary_Cutlery_64
   /Game/Athena/UI/InGame/Creative/PropertyEditor/Icon_Picker/Textures_V2/64x64/T_UI_IconLibrary_SheriffBadge_64
   /Game/Athena/UI/InGame/Creative/PropertyEditor/Icon_Picker/Textures_V2/64x64/T_UI_IconLibrary_Laptop_64
   /Game/Athena/UI/InGame/Creative/PropertyEditor/Icon_Picker/Textures_V2/64x64/T_UI_IconLibrary_Power_64
   /Game/Athena/UI/InGame/Creative/PropertyEditor/Icon_Picker/Textures_V2/64x64/T_UI_IconLibrary_Hammer_64
   /Game/Athena/UI/InGame/Creative/PropertyEditor/Icon_Picker/Textures_V2/64x64/T_UI_IconLibrary_Paintbrush_64
   /Game/Athena/UI/InGame/Creative/PropertyEditor/Icon_Picker/Textures_V2/64x64/T_UI_IconLibrary_Shovel_64
   /Game/Athena/UI/InGame/Creative/PropertyEditor/Icon_Picker/Textures_V2/64x64/T_UI_IconLibrary_Restaurant_64
   /Game/Athena/UI/InGame/Creative/PropertyEditor/Icon_Picker/Textures_V2/64x64/T_UI_IconLibrary_Gift_64
   /Game/Athena/UI/InGame/Creative/PropertyEditor/Icon_Picker/Textures_V2/64x64/T_UI_IconLibrary_Handshake_64
   /Game/Athena/UI/InGame/Creative/PropertyEditor/Icon_Picker/Textures_V2/64x64/T_UI_IconLibrary_Pin_64
   /Game/Athena/UI/InGame/Creative/PropertyEditor/Icon_Picker/Textures_V2/64x64/T_UI_IconLibrary_Camera_64
   /Game/Athena/UI/InGame/Creative/PropertyEditor/Icon_Picker/Textures_V2/64x64/T_UI_IconLibrary_PiggyBank_64
   /Game/Athena/UI/InGame/Creative/PropertyEditor/Icon_Picker/Textures_V2/64x64/T_UI_IconLibrary_BankNotes_64
   /Game/Athena/UI/InGame/Creative/PropertyEditor/Icon_Picker/Textures_V2/64x64/T_UI_IconLibrary_Burger_64
   /Game/Athena/UI/InGame/Creative/PropertyEditor/Icon_Picker/Textures_V2/64x64/T_UI_IconLibrary_Mushroom_64
   /Game/Athena/UI/InGame/Creative/PropertyEditor/Icon_Picker/Textures_V2/64x64/T_UI_IconLibrary_PizzaSlice_64
   /Game/Athena/UI/InGame/Creative/PropertyEditor/Icon_Picker/Textures_V2/64x64/T_UI_IconLibrary_GasMask_64
   /Game/Athena/UI/InGame/Creative/PropertyEditor/Icon_Picker/Textures_V2/64x64/T_UI_IconLibrary_Heartbeat_64.T_UI_IconLibrary_HeartBeat_64
   /Game/Athena/UI/InGame/Creative/PropertyEditor/Icon_Picker/Textures_V2/64x64/T_UI_IconLibrary_IceSkate_64
   /Game/Athena/UI/InGame/Creative/PropertyEditor/Icon_Picker/Textures_V2/64x64/T_UI_IconLibrary_SnowFlake_64.T_UI_IconLibrary_Snowflake_64
   /Game/Athena/UI/InGame/Creative/PropertyEditor/Icon_Picker/Textures_V2/64x64/T_UI_IconLibrary_Ghost_64
   /Game/Athena/UI/InGame/Creative/PropertyEditor/Icon_Picker/Textures_V2/64x64/T_UI_IconLibrary_Electrical_64
   /Game/Athena/UI/InGame/Creative/PropertyEditor/Icon_Picker/Textures_V2/64x64/T_UI_IconLibrary_CookingPot_64
   /Game/Athena/UI/InGame/Creative/PropertyEditor/Icon_Picker/Textures_V2/64x64/T_UI_IconLibrary_Cabin_64
   /Game/Athena/UI/InGame/Creative/PropertyEditor/Icon_Picker/Textures_V2/64x64/T_UI_IconLibrary_OfficePhone_64
   /Game/Athena/UI/InGame/Creative/PropertyEditor/Icon_Picker/Textures_V2/64x64/T_UI_IconLibrary_TV_64
   /Game/Athena/UI/InGame/Creative/PropertyEditor/Icon_Picker/Textures_V2/64x64/T_UI_IconLibrary_Gear_64
   /Game/Athena/UI/InGame/Creative/PropertyEditor/Icon_Picker/Textures_V2/64x64/T_UI_IconLibrary_Helix_64
   /Game/Athena/UI/InGame/Creative/PropertyEditor/Icon_Picker/Textures_V2/64x64/T_UI_IconLibrary_Zap_64
   /Game/Athena/UI/InGame/Creative/PropertyEditor/Icon_Picker/Textures_V2/64x64/T_UI_IconLibrary_Factory_64
   /Game/Athena/UI/InGame/Creative/PropertyEditor/Icon_Picker/Textures_V2/64x64/T_UI_IconLibrary_RadioTower_64
   /Game/Athena/UI/InGame/Creative/PropertyEditor/Icon_Picker/Textures_V2/64x64/T_UI_IconLibrary_Microscope_64
   /Game/Athena/UI/InGame/Creative/PropertyEditor/Icon_Picker/Textures_V2/64x64/T_UI_IconLibrary_SatelliteDish_64
   /Game/Athena/UI/InGame/Creative/PropertyEditor/Icon_Picker/Textures_V2/64x64/T_UI_IconLibrary_Map_64
   /Game/Athena/UI/InGame/Creative/PropertyEditor/Icon_Picker/Textures_V2/64x64/T_UI_IconLibrary_Play_64
   /Game/Athena/UI/InGame/Creative/PropertyEditor/Icon_Picker/Textures_V2/64x64/T_UI_IconLibrary_Stop_64
   /Game/Athena/UI/InGame/Creative/PropertyEditor/Icon_Picker/Textures_V2/64x64/T_UI_IconLibrary_Pause_64
   /Game/Athena/UI/InGame/Creative/PropertyEditor/Icon_Picker/Textures_V2/64x64/T_UI_IconLibrary_FastForward_64
   /Game/Athena/UI/InGame/Creative/PropertyEditor/Icon_Picker/Textures_V2/64x64/T_UI_IconLibrary_Rewind_64
   /Game/Athena/UI/InGame/Creative/PropertyEditor/Icon_Picker/Textures_V2/64x64/T_UI_IconLibrary_Volume_64
   /Game/Athena/UI/InGame/Creative/PropertyEditor/Icon_Picker/Textures_V2/64x64/T_UI_IconLibrary_Repeat_64
   /Game/Athena/UI/InGame/Creative/PropertyEditor/Icon_Picker/Textures_V2/64x64/T_UI_IconLibrary_Shuffle_64
   /Game/Athena/UI/InGame/Creative/PropertyEditor/Icon_Picker/Textures_V2/64x64/T_UI_IconLibrary_Chat_64
   /Game/Athena/UI/InGame/Creative/PropertyEditor/Icon_Picker/Textures_V2/64x64/T_UI_IconLibrary_Crown_64
   /Game/Athena/UI/InGame/Creative/PropertyEditor/Icon_Picker/Textures_V2/64x64/T_UI_IconLibrary_Eye_64
   /Game/Athena/UI/InGame/Creative/PropertyEditor/Icon_Picker/Textures_V2/64x64/T_UI_IconLibrary_Message_64
   /Game/Athena/UI/InGame/Creative/PropertyEditor/Icon_Picker/Textures_V2/64x64/T_UI_IconLibrary_LightBulb_64
   /Game/Athena/UI/InGame/Creative/PropertyEditor/Icon_Picker/Textures_V2/64x64/T_UI_IconLibrary_Bomb_64
   /Game/Athena/UI/InGame/Creative/PropertyEditor/Icon_Picker/Textures_V2/64x64/T_UI_IconLibrary_Cannon_64
   /Game/Athena/UI/InGame/Creative/PropertyEditor/Icon_Picker/MilitaryTextures/T_64x64/T_UI_IconLibrary_JetFighter_64
   /Game/Athena/UI/InGame/Creative/PropertyEditor/Icon_Picker/Textures_V2/64x64/T_UI_IconLibrary_Tent_64
   /Game/Athena/UI/InGame/Creative/PropertyEditor/Icon_Picker/Textures_V2/64x64/T_UI_IconLibrary_CCTV_64
   /Game/Athena/UI/InGame/Creative/PropertyEditor/Icon_Picker/Textures_V2/64x64/T_UI_IconLibrary_Radio_64
   /Game/Athena/UI/InGame/Creative/PropertyEditor/Icon_Picker/Textures_V2/64x64/T_UI_IconLibrary_Wifi_64
   /Game/Athena/UI/InGame/Creative/PropertyEditor/Icon_Picker/Textures_V2/64x64/T_UI_IconLibrary_Grass_64
   /Game/Athena/UI/InGame/Creative/PropertyEditor/Icon_Picker/Textures_V2/64x64/T_UI_IconLibrary_Leaf_64
   /Game/Athena/UI/InGame/Creative/PropertyEditor/Icon_Picker/Textures_V2/64x64/T_UI_IconLibrary_PalmTree_64
   /Game/Athena/UI/InGame/Creative/PropertyEditor/Icon_Picker/Textures_V2/64x64/T_UI_IconLibrary_Book_64
   /Game/Athena/UI/InGame/Creative/PropertyEditor/Icon_Picker/Textures_V2/64x64/T_UI_IconLibrary_Toolbox_64
   /Game/Athena/UI/InGame/Creative/PropertyEditor/Icon_Picker/Textures_V2/64x64/T_UI_IconLibrary_Bug_64
   /Game/Athena/UI/InGame/Creative/PropertyEditor/Icon_Picker/Textures_V2/64x64/T_UI_IconLibrary_FloppyDisc_64
   /Game/Athena/UI/InGame/Creative/PropertyEditor/Icon_Picker/Textures_V2/64x64/T_UI_IconLibrary_Cogs_64
   /Game/Athena/UI/InGame/Creative/PropertyEditor/Icon_Picker/Textures_V2/64x64/T_UI_IconLibrary_Clock_64
   /Game/Athena/UI/InGame/Creative/PropertyEditor/Icon_Picker/Textures_V2/64x64/T_UI_IconLibrary_TestTube_64
   /Game/Athena/UI/InGame/Creative/PropertyEditor/Icon_Picker/Textures_V2/64x64/T_UI_IconLibrary_Stopwatch_64
   /Game/Athena/UI/InGame/Creative/PropertyEditor/Icon_Picker/Textures_V2/64x64/T_UI_IconLibrary_Basketball_64
   /Game/Athena/UI/InGame/Creative/PropertyEditor/Icon_Picker/Textures_V2/64x64/T_UI_IconLibrary_BoxingGlove_64
   /Game/Athena/UI/InGame/Creative/PropertyEditor/Icon_Picker/Textures_V2/64x64/T_UI_IconLibrary_SoccerBall_64
   /Game/Athena/UI/InGame/Creative/PropertyEditor/Icon_Picker/Textures_V2/64x64/T_UI_IconLibrary_Roadblock_64
   /Game/Athena/UI/InGame/Creative/PropertyEditor/Icon_Picker/Textures_V2/64x64/T_UI_IconLibrary_Rocket_64
   /Game/Athena/UI/InGame/Creative/PropertyEditor/Icon_Picker/Textures_V2/64x64/T_UI_IconLibrary_Taxi_64
   /Game/Athena/UI/InGame/Creative/PropertyEditor/Icon_Picker/Textures_V2/64x64/T_UI_IconLibrary_Footprint_64
   /Game/Athena/UI/InGame/Creative/PropertyEditor/Icon_Picker/Textures_V2/64x64/T_UI_IconLibrary_Shoeprint_64
   /Game/Athena/UI/InGame/Creative/PropertyEditor/Icon_Picker/Textures_V2/64x64/T_UI_IconLibrary_Torch_64
   /Game/Athena/UI/InGame/Creative/PropertyEditor/Icon_Picker/Textures_V2/64x64/T_UI_IconLibrary_Axe_64
   /Game/Athena/UI/InGame/Creative/PropertyEditor/Icon_Picker/Textures_V2/64x64/T_UI_IconLibrary_Link_64
   /Game/Athena/UI/InGame/Creative/PropertyEditor/Icon_Picker/Textures_V2/64x64/T_UI_IconLibrary_Jigsaw_64
   /Game/Athena/UI/InGame/Creative/PropertyEditor/Icon_Picker/Textures_V2/64x64/T_UI_IconLibrary_Magnifier_64
   /Game/Athena/UI/InGame/Creative/PropertyEditor/Icon_Picker/Textures_V2/64x64/T_UI_IconLibrary_Wrench_64
   /Game/Athena/UI/InGame/Creative/PropertyEditor/Icon_Picker/Textures_V2/64x64/T_UI_IconLibrary_Input_64
   /Game/Athena/UI/InGame/Creative/PropertyEditor/Icon_Picker/Textures_V2/64x64/T_UI_IconLibrary_Calendar_64
   /Game/Athena/UI/InGame/Creative/PropertyEditor/Icon_Picker/Textures_V2/64x64/T_UI_IconLibrary_Projector_64
   /Game/Athena/UI/InGame/Creative/PropertyEditor/Icon_Picker/Textures_V2/64x64/T_UI_IconLibrary_Drama_64
   /Game/Athena/UI/InGame/Creative/PropertyEditor/Icon_Picker/Textures_V2/64x64/T_UI_IconLibrary_Hourglass_64
   /Game/Athena/UI/InGame/Creative/PropertyEditor/Icon_Picker/Textures_V2/64x64/T_UI_IconLibrary_Ticket_64
   /Game/Athena/UI/InGame/Creative/PropertyEditor/Icon_Picker/Textures_V2/64x64/T_UI_IconLibrary_Cactus_64
   /Game/Athena/UI/InGame/Creative/PropertyEditor/Icon_Picker/Textures_V2/64x64/T_UI_IconLibrary_JetPlane_64
   /Game/Athena/UI/InGame/Creative/PropertyEditor/Icon_Picker/Textures_V2/64x64/T_UI_IconLibrary_Railway_64
   /Game/Athena/UI/InGame/Creative/PropertyEditor/Icon_Picker/Textures_V2/64x64/T_UI_IconLibrary_Handcuffs_64
   /Game/Athena/UI/InGame/Creative/PropertyEditor/Icon_Picker/Textures_V2/64x64/T_UI_IconLibrary_Monument_64
   /Game/Athena/UI/InGame/Creative/PropertyEditor/Icon_Picker/Textures_V2/64x64/T_UI_IconLibrary_Boot_64
   /Game/Athena/UI/InGame/Creative/PropertyEditor/Icon_Picker/Textures_V2/64x64/T_UI_IconLibrary_Shirt_64
   /Game/Athena/UI/InGame/Creative/PropertyEditor/Icon_Picker/Textures_V2/64x64/T_UI_IconLibrary_Controller_64
   /Game/Athena/UI/InGame/Creative/PropertyEditor/Icon_Picker/Textures_V2/64x64/T_UI_IconLibrary_Keyboard_64
   /Game/Athena/UI/InGame/Creative/PropertyEditor/Icon_Picker/Textures_V2/64x64/T_UI_IconLibrary_Tape_64
   /Game/Athena/UI/InGame/Creative/PropertyEditor/Icon_Picker/Textures_V2/64x64/T_UI_IconLibrary_Plant_64
   /Game/Athena/UI/InGame/Creative/PropertyEditor/Icon_Picker/Textures_V2/64x64/T_UI_IconLibrary_WateringCan_64
   /Game/Athena/UI/InGame/Creative/PropertyEditor/Icon_Picker/Textures_V2/64x64/T_UI_IconLibrary_PaperClip_64
   /Game/Athena/UI/InGame/Creative/PropertyEditor/Icon_Picker/Textures_V2/64x64/T_UI_IconLibrary_PaintPot_64
   /Game/Athena/UI/InGame/Creative/PropertyEditor/Icon_Picker/Textures_V2/64x64/T_UI_IconLibrary_Clipboard_64
   /Game/Athena/UI/InGame/Creative/PropertyEditor/Icon_Picker/Textures_V2/64x64/T_UI_IconLibrary_TrashCan_64
   /Game/Athena/UI/InGame/Creative/PropertyEditor/Icon_Picker/Textures_V2/64x64/T_UI_IconLibrary_Scissors_64
   /Game/Athena/UI/InGame/Creative/PropertyEditor/Icon_Picker/Textures_V2/64x64/T_UI_IconLibrary_Pencil_64
   /Game/Athena/UI/InGame/Creative/PropertyEditor/Icon_Picker/Textures_V2/64x64/T_UI_IconLibrary_CoffeeCup_64
   /Game/Athena/UI/InGame/Creative/PropertyEditor/Icon_Picker/Textures_V2/64x64/T_UI_IconLibrary_CoffeeMug_64
   /Game/Athena/UI/InGame/Creative/PropertyEditor/Icon_Picker/Textures_V2/64x64/T_UI_IconLibrary_Cooker_64
   /Game/Athena/UI/InGame/Creative/PropertyEditor/Icon_Picker/Textures_V2/64x64/T_UI_IconLibrary_FrenchFries_64
   /Game/Athena/UI/InGame/Creative/PropertyEditor/Icon_Picker/Textures_V2/64x64/T_UI_IconLibrary_Fridge_64
   /Game/Athena/UI/InGame/Creative/PropertyEditor/Icon_Picker/Textures_V2/64x64/T_UI_IconLibrary_IceCream_64
   /Game/Athena/UI/InGame/Creative/PropertyEditor/Icon_Picker/Textures_V2/64x64/T_UI_IconLibrary_Microwave_64
   /Game/Athena/UI/InGame/Creative/PropertyEditor/Icon_Picker/Textures_V2/64x64/T_UI_IconLibrary_Cards_64
   /Game/Athena/UI/InGame/Creative/PropertyEditor/Icon_Picker/Textures_V2/64x64/T_UI_IconLibrary_Shamrock_64
   /Game/Athena/UI/InGame/Creative/PropertyEditor/Icon_Picker/Textures_V2/64x64/T_UI_IconLibrary_Dice_64
   /Game/Athena/UI/InGame/Creative/PropertyEditor/Icon_Picker/Textures_V2/64x64/T_UI_IconLibrary_ChessPiece_64
   /Game/Athena/UI/InGame/Creative/PropertyEditor/Icon_Picker/Textures_V2/64x64/T_UI_IconLibrary_Bell_64
   /Game/Athena/UI/InGame/Creative/PropertyEditor/Icon_Picker/Textures_V2/64x64/T_UI_IconLibrary_Candle_64
   /Game/Athena/UI/InGame/Creative/PropertyEditor/Icon_Picker/Textures_V2/64x64/T_UI_IconLibrary_Pumpkin_64
   /Game/Athena/UI/InGame/Creative/PropertyEditor/Icon_Picker/Textures_V2/64x64/T_UI_IconLibrary_Bat_64
   /Game/Athena/UI/InGame/Creative/PropertyEditor/Icon_Picker/Textures_V2/64x64/T_UI_IconLibrary_Sofa_64
   /Game/Athena/UI/InGame/Creative/PropertyEditor/Icon_Picker/Textures_V2/64x64/T_UI_IconLibrary_Tap_64
   /Game/Athena/UI/InGame/Creative/PropertyEditor/Icon_Picker/Textures_V2/64x64/T_UI_IconLibrary_Stairs_64
   /Game/Athena/UI/InGame/Creative/PropertyEditor/Icon_Picker/Textures_V2/64x64/T_UI_IconLibrary_Circuit_64
   /Game/Athena/UI/InGame/Creative/PropertyEditor/Icon_Picker/Textures_V2/64x64/T_UI_IconLibrary_Processor_64
   /Game/Athena/UI/InGame/Creative/PropertyEditor/Icon_Picker/Textures_V2/64x64/T_UI_IconLibrary_RobotArm_64
   /Game/Athena/UI/InGame/Creative/PropertyEditor/Icon_Picker/Textures_V2/64x64/T_UI_IconLibrary_Globe_64
   /Game/Athena/UI/InGame/Creative/PropertyEditor/Icon_Picker/Textures_V2/64x64/T_UI_IconLibrary_Satellite_64
   /Game/Athena/UI/InGame/Creative/PropertyEditor/Icon_Picker/Textures_V2/64x64/T_UI_IconLibrary_Atom_64
   /Game/Athena/UI/InGame/Creative/PropertyEditor/Icon_Picker/Textures_V2/64x64/T_UI_IconLibrary_Microphone_64
   /Game/Athena/UI/InGame/Creative/PropertyEditor/Icon_Picker/Textures_V2/64x64/T_UI_IconLibrary_TrebleClef_64
   /Game/Athena/UI/InGame/Creative/PropertyEditor/Icon_Picker/Textures_V2/64x64/T_UI_IconLibrary_Cloud_64
   /Game/Athena/UI/InGame/Creative/PropertyEditor/Icon_Picker/Textures_V2/64x64/T_UI_IconLibrary_Connected_64
   /Game/Athena/UI/InGame/Creative/PropertyEditor/Icon_Picker/Textures_V2/64x64/T_UI_IconLibrary_Disconnected_64
   /Game/Athena/UI/InGame/Creative/PropertyEditor/Icon_Picker/Textures_V2/64x64/T_UI_IconLibrary_Server_64
   /Game/Athena/UI/InGame/Creative/PropertyEditor/Icon_Picker/Textures_V2/64x64/T_UI_IconLibrary_Clapperboard_64
   /Game/Athena/UI/InGame/Creative/PropertyEditor/Icon_Picker/Textures_V2/64x64/T_UI_IconLibrary_Hashtag_64
   /Game/Athena/UI/InGame/Creative/PropertyEditor/Icon_Picker/Textures_V2/64x64/T_UI_IconLibrary_Infinity_64
   /Game/Athena/UI/InGame/Creative/PropertyEditor/Icon_Picker/Textures_V2/64x64/T_UI_IconLibrary_Parachute_64
   /Game/Athena/UI/InGame/Creative/PropertyEditor/Icon_Picker/Textures_V2/64x64/T_UI_IconLibrary_Bus_64
   /Game/Athena/UI/InGame/Creative/PropertyEditor/Icon_Picker/Textures_V2/64x64/T_UI_IconLibrary_BeachBall_64
   /Game/Athena/UI/InGame/Creative/PropertyEditor/Icon_Picker/Textures_V2/64x64/T_UI_IconLibrary_Umbrella_64
   /Game/Athena/UI/InGame/Creative/PropertyEditor/Icon_Picker/Textures_V2/64x64/T_UI_IconLibrary_Suitcase_64
   /Game/Athena/UI/InGame/Creative/PropertyEditor/Icon_Picker/Textures_V2/64x64/T_UI_IconLibrary_AlarmClock_64
   /Game/Athena/UI/InGame/Creative/PropertyEditor/Icon_Picker/Textures_V2/64x64/T_UI_IconLibrary_Calculator_64
   /Game/Athena/UI/InGame/Creative/PropertyEditor/Icon_Picker/Textures_V2/64x64/T_UI_IconLibrary_Raincloud_64
   /Game/Athena/UI/InGame/Creative/PropertyEditor/Icon_Picker/Textures_V2/64x64/T_UI_IconLibrary_SnowCloud_64.T_UI_IconLibrary_Snowcloud_64
   /Game/Athena/UI/InGame/Creative/PropertyEditor/Icon_Picker/Textures_V2/64x64/T_UI_IconLibrary_Stormcloud_64
   /Game/Athena/UI/InGame/Creative/PropertyEditor/Icon_Picker/Textures_V2/64x64/T_UI_IconLibrary_Sun_64
   /Game/Athena/UI/InGame/Creative/PropertyEditor/Icon_Picker/Textures_V2/64x64/T_UI_IconLibrary_Moon_64
   /Game/Athena/UI/InGame/Creative/PropertyEditor/Icon_Picker/Textures_V2/64x64/T_UI_IconLibrary_Tornado_64
   /Game/Athena/UI/InGame/Creative/PropertyEditor/Icon_Picker/Textures_V2/64x64/T_UI_IconLibrary_Flower_64
   /Game/Athena/UI/InGame/Creative/PropertyEditor/Icon_Picker/Textures_V2/64x64/T_UI_IconLibrary_Chest_64
   /Game/Athena/UI/InGame/Creative/PropertyEditor/Icon_Picker/Textures_V2/64x64/T_UI_IconLibrary_Palette_64
   /Game/Athena/UI/InGame/Creative/PropertyEditor/Icon_Picker/Textures_V2/64x64/T_UI_IconLibrary_Cheese_64
   /Game/Athena/UI/InGame/Creative/PropertyEditor/Icon_Picker/Textures_V2/64x64/T_UI_IconLibrary_BowlingPins_64
   /Game/Athena/UI/InGame/Creative/PropertyEditor/Icon_Picker/Textures_V2/64x64/T_UI_IconLibrary_BowlingBall_64
   /Game/Athena/UI/InGame/Creative/PropertyEditor/Icon_Picker/Textures_V2/64x64/T_UI_IconLibrary_Swimming_64
   /Game/Athena/UI/InGame/Creative/PropertyEditor/Icon_Picker/Textures_V2/64x64/T_UI_IconLibrary_BaseballBat_64
   /Game/Athena/UI/InGame/Creative/PropertyEditor/Icon_Picker/Textures_V2/64x64/T_UI_IconLibrary_Baseball_64
   /Game/Athena/UI/InGame/Creative/PropertyEditor/Icon_Picker/Textures_V2/64x64/T_UI_IconLibrary_VolleyBall_64.T_UI_IconLibrary_Volleyball_64
   /Game/Athena/UI/InGame/Creative/PropertyEditor/Icon_Picker/Textures_V2/64x64/T_UI_IconLibrary_Hockey_64
   /Game/Athena/UI/InGame/Creative/PropertyEditor/Icon_Picker/Textures_V2/64x64/T_UI_IconLibrary_Tennis_64
   /Game/Athena/UI/InGame/Creative/PropertyEditor/Icon_Picker/Textures_V2/64x64/T_UI_IconLibrary_Golf_64
   /Game/Athena/UI/InGame/Creative/PropertyEditor/Icon_Picker/Textures_V2/64x64/T_UI_IconLibrary_Archery_64
   /Game/Athena/UI/InGame/Creative/PropertyEditor/Icon_Picker/Textures_V2/64x64/T_UI_IconLibrary_Target_64
   /Game/Athena/UI/InGame/Creative/PropertyEditor/Icon_Picker/Textures_V2/64x64/T_UI_IconLibrary_Crosshair_64
   /Game/Athena/UI/InGame/Creative/PropertyEditor/Icon_Picker/Textures_V2/64x64/T_UI_IconLibrary_Balloon_64
   /Game/Athena/UI/InGame/Creative/PropertyEditor/Icon_Picker/Textures_V2/64x64/T_UI_IconLibrary_Quadcrasher_64.T_UI_IconLibrary_QuadCrasher_64
   /Game/Athena/UI/InGame/Creative/PropertyEditor/Icon_Picker/Textures_V2/64x64/T_UI_IconLibrary_GolfCart_64
   /Game/Athena/UI/InGame/Creative/PropertyEditor/Icon_Picker/Textures_V2/64x64/T_UI_IconLibrary_Driftboard_64.T_UI_IconLibrary_DriftBoard_64
   /Game/Athena/UI/InGame/Creative/PropertyEditor/Icon_Picker/Textures_V2/64x64/T_UI_IconLibrary_HamsterBall_64
   /Game/Athena/UI/InGame/Creative/PropertyEditor/Icon_Picker/Textures_V2/64x64/T_UI_IconLibrary_Motorboat_64
   /Game/Athena/UI/InGame/Creative/PropertyEditor/Icon_Picker/Textures_V2/64x64/T_UI_IconLibrary_Mech_64
   /Game/Athena/UI/InGame/Creative/PropertyEditor/Icon_Picker/Textures_V2/64x64/T_UI_IconLibrary_Pow_64
   /Game/Athena/UI/InGame/Creative/PropertyEditor/Icon_Picker/Textures_V2/64x64/T_UI_IconLibrary_Impact_64
   /Game/Athena/UI/InGame/Creative/PropertyEditor/Icon_Picker/Textures_V2/64x64/T_UI_IconLibrary_Mountains_64
   /Game/Athena/UI/InGame/Creative/PropertyEditor/Icon_Picker/Textures_V2/64x64/T_UI_IconLibrary_Waves_64
   /Game/Athena/UI/InGame/Creative/PropertyEditor/Icon_Picker/Textures_V2/64x64/T_UI_IconLibrary_Water_64
   /Game/Athena/UI/InGame/Creative/PropertyEditor/Icon_Picker/Textures_V2/64x64/T_UI_IconLibrary_Road_64
   /Game/Athena/UI/InGame/Creative/PropertyEditor/Icon_Picker/Textures_V2/64x64/T_UI_IconLibrary_Caution_64
   /Game/Athena/UI/InGame/Creative/PropertyEditor/Icon_Picker/Textures_V2/64x64/T_UI_IconLibrary_StarMarker_64
   /Game/Athena/UI/InGame/Creative/PropertyEditor/Icon_Picker/Textures_V2/64x64/T_UI_IconLibrary_RightLaneMerge_64
   /Game/Athena/UI/InGame/Creative/PropertyEditor/Icon_Picker/Textures_V2/64x64/T_UI_IconLibrary_LeftorRightTurn_64
   /Game/Athena/UI/InGame/Creative/PropertyEditor/Icon_Picker/Textures_V2/64x64/T_UI_IconLibrary_LeftCurve_64
   /Game/Athena/UI/InGame/Creative/PropertyEditor/Icon_Picker/Textures_V2/64x64/T_UI_IconLibrary_RightCurve_64
   /Game/Athena/UI/InGame/Creative/PropertyEditor/Icon_Picker/Textures_V2/64x64/T_UI_IconLibrary_LeftTurn_64
   /Game/Athena/UI/InGame/Creative/PropertyEditor/Icon_Picker/Textures_V2/64x64/T_UI_IconLibrary_RightTurn_64
   /Game/Athena/UI/InGame/Creative/PropertyEditor/Icon_Picker/Textures_V2/64x64/T_UI_IconLibrary_U-Turn_64
   /Game/Athena/UI/InGame/Creative/PropertyEditor/Icon_Picker/Textures_V2/64x64/T_UI_IconLibrary_NarrowLane_64
   /Game/Athena/UI/InGame/Creative/PropertyEditor/Icon_Picker/Textures_V2/64x64/T_UI_IconLibrary_NewLane_64
   /Game/Athena/UI/InGame/Creative/PropertyEditor/Icon_Picker/Textures_V2/64x64/T_UI_IconLibrary_Intersection_64
   /Game/Athena/UI/InGame/Creative/PropertyEditor/Icon_Picker/Textures_V2/64x64/T_UI_IconLibrary_People_64
   /Game/Athena/UI/InGame/Creative/PropertyEditor/Icon_Picker/Textures_V2/64x64/T_UI_IconLibrary_Bicycle_64
   /Game/Athena/UI/InGame/Creative/PropertyEditor/Icon_Picker/Textures_V2/64x64/T_UI_IconLibrary_Wood_64
   /Game/Athena/UI/InGame/Creative/PropertyEditor/Icon_Picker/Textures_V2/64x64/T_UI_IconLibrary_Stone_64
   /Game/Athena/UI/InGame/Creative/PropertyEditor/Icon_Picker/Textures_V2/64x64/T_UI_IconLibrary_Metal_64
   /Game/Athena/UI/InGame/Creative/PropertyEditor/Icon_Picker/Textures_V2/64x64/T_UI_IconLibrary_Design_64
   /Game/Athena/UI/InGame/Creative/PropertyEditor/Icon_Picker/Textures_V2/64x64/T_UI_IconLibrary_CandyBar_64
   /Game/Athena/UI/InGame/Creative/PropertyEditor/Icon_Picker/Textures_V2/64x64/T_UI_IconLibrary_Bookshelf_64
   /Game/Athena/UI/InGame/Creative/PropertyEditor/Icon_Picker/Textures_V2/64x64/T_UI_IconLibrary_Llama_64
   /Game/Athena/UI/InGame/Creative/PropertyEditor/Icon_Picker/Textures_V2/64x64/T_UI_IconLibrary_Saw_64
   /Game/Athena/UI/InGame/Creative/PropertyEditor/Icon_Picker/Textures_V2/64x64/T_UI_IconLibrary_Gnome_64
   /Game/Athena/UI/InGame/Creative/PropertyEditor/Icon_Picker/Textures_V2/64x64/T_UI_IconLibrary_Ham_64
   /Game/Athena/UI/InGame/Creative/PropertyEditor/Icon_Picker/Textures_V2/64x64/T_UI_IconLibrary_Fingerprint_64
   /Game/Athena/UI/InGame/Creative/PropertyEditor/Icon_Picker/Textures_V2/64x64/T_UI_IconLibrary_Hack_64
   /Game/Athena/UI/InGame/Creative/PropertyEditor/Icon_Picker/Textures_V2/64x64/T_UI_IconLibrary_Place_64
   /Game/Athena/UI/InGame/Creative/PropertyEditor/Icon_Picker/Textures_V2/64x64/T_UI_IconLibrary_Cube_64
   /Game/Athena/UI/InGame/Creative/PropertyEditor/Icon_Picker/Textures_V2/64x64/T_UI_IconLibrary_Alarm_64
   /Game/Athena/UI/InGame/Creative/PropertyEditor/Icon_Picker/Textures_V2/64x64/T_UI_IconLibrary_Comet_64
   /Game/Athena/UI/InGame/Creative/PropertyEditor/Icon_Picker/Textures_V2/64x64/T_UI_IconLibrary_Castle_64
   /Game/Athena/UI/InGame/Creative/PropertyEditor/Icon_Picker/Textures_V2/64x64/T_UI_IconLibrary_YinYang_64
   /Game/Athena/UI/InGame/Creative/PropertyEditor/Icon_Picker/Textures_V2/64x64/T_UI_IconLibrary_Helmet_64
   /Game/Athena/UI/InGame/Creative/PropertyEditor/Icon_Picker/Textures_V2/64x64/T_UI_IconLibrary_Jonesy_64
   /Game/Athena/UI/InGame/Creative/PropertyEditor/Icon_Picker/Textures_V2/64x64/T_UI_IconLibrary_Melee_64
   /Game/Athena/UI/InGame/Creative/PropertyEditor/Icon_Picker/Textures_V2/64x64/T_UI_IconLibrary_Medal_64
   /Game/Athena/UI/InGame/Creative/PropertyEditor/Icon_Picker/Textures_V2/64x64/T_UI_IconLibrary_Wing_64
   /Game/Athena/UI/InGame/Creative/PropertyEditor/Icon_Picker/Textures_V2/64x64/T_UI_IconLibrary_Swap_64
   /Game/Athena/UI/InGame/Creative/PropertyEditor/Icon_Picker/Textures_V2/64x64/T_UI_IconLibrary_Reset_64
   /Game/Athena/UI/InGame/Creative/PropertyEditor/Icon_Picker/Textures_V2/64x64/T_UI_IconLibrary_Shark_64
   /Game/Athena/UI/InGame/Creative/PropertyEditor/Icon_Picker/Textures_V2/64x64/T_UI_IconLibrary_FreezerChest_64
   /Game/Athena/UI/InGame/Creative/PropertyEditor/Icon_Picker/Textures_V2/64x64/T_UI_IconLibrary_PaperAirplane_64
   /Game/Athena/UI/InGame/Creative/PropertyEditor/Icon_Picker/Textures_V2/64x64/T_UI_IconLibrary_Crafting_64
   /Game/Athena/UI/InGame/Creative/PropertyEditor/Icon_Picker/Textures_V2/64x64/T_UI_IconLibrary_T-Rex_64
   /Game/Athena/UI/InGame/Creative/PropertyEditor/Icon_Picker/Textures_V2/64x64/T_UI_IconLibrary_Zipline_64.T_UI_IconLibrary_ZipLine_64
   /Game/Athena/UI/InGame/Creative/PropertyEditor/Icon_Picker/Textures_V2/64x64/T_UI_IconLibrary_Tree_64
   /Game/Athena/UI/InGame/Creative/PropertyEditor/Icon_Picker/Textures_V2/64x64/T_UI_IconLibrary_Diamond_64
   /Game/Athena/UI/InGame/Creative/PropertyEditor/Icon_Picker/Textures_V2/64x64/T_UI_IconLibrary_Directions_64
   /Game/Athena/UI/InGame/Creative/PropertyEditor/Icon_Picker/Textures_V2/64x64/T_UI_IconLibrary_Share_64
   /Game/Athena/UI/InGame/Creative/PropertyEditor/Icon_Picker/Textures_V2/64x64/T_UI_IconLibrary_SkiMask_64
   /Game/Athena/UI/InGame/Creative/PropertyEditor/Icon_Picker/Textures_V2/64x64/T_UI_IconLibrary_Folder_64
   /Game/Athena/UI/InGame/Creative/PropertyEditor/Icon_Picker/Textures_V2/64x64/T_UI_IconLibrary_Muscle_64
   /Game/Athena/UI/InGame/Creative/PropertyEditor/Icon_Picker/Textures_V2/64x64/T_UI_IconLibrary_Mask_64
   /Game/Athena/UI/InGame/Creative/PropertyEditor/Icon_Picker/Textures_V2/64x64/T_UI_IconLibrary_Lamp_64
   /Game/Athena/UI/InGame/Creative/PropertyEditor/Icon_Picker/Textures_V2/64x64/T_UI_IconLibrary_Incognito_64
   /Game/Athena/UI/InGame/Creative/PropertyEditor/Icon_Picker/Textures_V2/64x64/T_UI_IconLibrary_Radar_64
   /Game/Athena/UI/InGame/Creative/PropertyEditor/Icon_Picker/Textures_V2/64x64/T_UI_IconLibrary_ClothesHanger_64
   /Game/Athena/UI/InGame/Creative/PropertyEditor/Icon_Picker/Textures_V2/64x64/T_UI_IconLibrary_ShoppingBag_64
   /Game/Athena/UI/InGame/Creative/PropertyEditor/Icon_Picker/Textures_V2/64x64/T_UI_IconLibrary_Scenery_64
   /Game/Athena/UI/InGame/Creative/PropertyEditor/Icon_Picker/Textures_V2/64x64/T_UI_IconLibrary_DogTags_64
   /Game/Athena/UI/InGame/Creative/PropertyEditor/Icon_Picker/Textures_V2/64x64/T_UI_IconLibrary_Locker_64
   /Game/Athena/UI/InGame/Creative/PropertyEditor/Icon_Picker/Textures_V2/64x64/T_UI_IconLibrary_Bracket_64
  ]
 }
]
widgets {
 info {inheritedWidgetCount=0, parentClass=/Script/FortniteUI.FortCoreGamerObjectiveRadial, rootWidgetClass=&SizeBox, widgetCount=4}
 widgets [
  SizeBox_0 {
   widgetClassPath=&SizeBox
   widget=SizeBox_0
   properties {bOverride_HeightOverride=true, bOverride_WidthOverride=true, heightOverride=150, widthOverride=150}
  }
  MainContainer {
   widgetClassPath=&Overlay
   bIsVariable=true
   parent=SizeBox_0
   widget=MainContainer
   slot {class=SizeBoxSlot, properties {padding=<ltrb 15 17.5 15 15>}}
  }
  Image_BackplateProgress {
   widgetClassPath=&Image
   bIsVariable=true
   parent=MainContainer
   widget=Image_BackplateProgress
   properties {
    visibility=SelfHitTestInvisible
    brush {
     bIsDynamicallyLoaded=false
     drawAs=Image
     imageType=NoImage
     mirroring=NoMirror
     resourceName=None
     resourceObject=/Game/Athena/HUD/LTM/Playground/CoreGamer/Materials/MI_CoreGamer_DestructibleProgress
     tiling=NoTile
     imageSize=<64 64>
     margin=<ltrb 0.09 0.09 0.09 0.09>
     outlineSettings {
      bUseBrushTransparency=false
      roundingType=HalfHeightRadius
      width=0
      color {colorUseRule=UseColor_Specified, specifiedColor=<rgba 0 0 0 0>}
      cornerRadii=<0 0 0 0>
     }
     tintColor {colorUseRule=UseColor_Specified, specifiedColor=<rgba 1 1 1 1>}
     uVRegion {bIsValid=false, max=<0 0>, min=<0 0>}
    }
   }
   slot {class=OverlaySlot, properties {horizontalAlignment=HAlign_Fill, verticalAlignment=VAlign_Fill}}
  }
  IconImage {
   widgetClassPath=&Image
   bIsVariable=true
   parent=MainContainer
   widget=IconImage
   properties {
    visibility=SelfHitTestInvisible
    brush {
     bIsDynamicallyLoaded=false
     drawAs=Image
     imageType=NoImage
     mirroring=NoMirror
     resourceName=None
     resourceObject=&T_UI_IconLibrary_A_64
     tiling=NoTile
     imageSize=<56 56>
     margin=<ltrb 0 0 0 0>
     outlineSettings {
      bUseBrushTransparency=false
      roundingType=HalfHeightRadius
      width=0
      color {colorUseRule=UseColor_Specified, specifiedColor=<rgba 0 0 0 0>}
      cornerRadii=<0 0 0 0>
     }
     tintColor {colorUseRule=UseColor_Specified, specifiedColor=<rgba 1 1 1 1>}
     uVRegion {bIsValid=false, max=<0 0>, min=<0 0>}
    }
    renderTransform {angle=0, scale=<1 1>, shear=<0 0>, translation=<0 -7.5>}
   }
   slot {class=OverlaySlot, properties {horizontalAlignment=HAlign_Center, verticalAlignment=VAlign_Center, padding=<ltrb 0 0 0 5>}}
  }
 ]
}
