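# Radial HUD marker widget for a Core Gamer capture area objective, showing capture progress and team ownership.
# Capture and team-change events drive the dynamic marker material's progress and team-state colors, taken from a data asset.
# InWorldTimerUpdate applies the area's icon settings: line-of-sight trace, range and minigame-state visibility, distance-based scale and distance text, while clamp direction toggles offscreen arrows.
# UMG animations provide owning, contested, pulse and flare effects.
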WidgetBlueprint /Game/Athena/HUD/LTM/Playground/CoreGamer/CoreGamer_ObjectiveRadial
animations [
 Anim_ObjectiveOwning {
  binding_count=1
  length_seconds=0.5667
  object=&self:Anim_ObjectiveOwning
  sequence {
   bindings [Image_Marker {
     id=#1
     possessed_class=Image
     tracks [{
       type=MovieSceneWidgetMaterialTrack
       sections [{
         type=MovieSceneParameterSection
         channels [
          Rotation Angle {type=MovieSceneScriptingFloatChannel, keys [[2,0.125],[13,0],[15,0.003],[17,0]]}
          Base Scale {type=MovieSceneScriptingFloatChannel, thinned_from=8, keys [[0,0.7],[3,0.65],[7,0.75],[9,0.7],[12,0.7],[14,0.65],[17,0.7]]}
          Icon Scale {type=MovieSceneScriptingFloatChannel, keys [[0,0.5],[3,0.48],[5,0.55],[9,0.55],[12,0.5],[17,0.5]]}
         ]
        }]
      }]
    }]
   display_rate {numerator=30}
  }
 }
 Anim_Pulse_ObjectiveOwning {
  binding_count=1
  length_seconds=0.8334
  object=&self:Anim_Pulse_ObjectiveOwning
  sequence {
   bindings [Image_Marker_Pulse {
     id=#2
     possessed_class=Image
     tracks [
      {
       type=MovieSceneWidgetMaterialTrack
       sections [{
         type=MovieSceneParameterSection
         channels [
          Progress Background Opacity {type=MovieSceneScriptingFloatChannel, const [0,1]}
          Base Scale {type=MovieSceneScriptingFloatChannel, keys [[15,0.6996],[25,1]]}
          Progress Opacity {type=MovieSceneScriptingFloatChannel, keys [[15,1],[25,0]]}
          Progress Background Thickness {type=MovieSceneScriptingFloatChannel, const [0,1]}
          Color for TransparencyR {type=MovieSceneScriptingFloatChannel, const [0.0097,1]}
          Color for TransparencyG {type=MovieSceneScriptingFloatChannel, const [0.2705,1]}
          Color for TransparencyB {type=MovieSceneScriptingFloatChannel, const [0,1]}
          Color for TransparencyA {type=MovieSceneScriptingFloatChannel, const [1,1]}
         ]
        }]
      }
      {
       type=MovieSceneByteTrack
       display_name=Visibility
       sections [{type=MovieSceneByteSection, end_frame=25, start_frame=0, channels [None {type=MovieSceneScriptingByteChannel, keys [[0,1],[15,4],[25,1]]}]}]
      }
      {
       type=MovieScene2DTransformTrack
       sections [{type=MovieScene2DTransformSection, end_frame=25, start_frame=0, channels [Angle {type=MovieSceneScriptingFloatChannel, const [0,1]}]}]
      }
     ]
    }]
   display_rate {numerator=30}
   tracks [{
     type=MovieSceneEventTrack
     sections [{
       type=MovieSceneEventTriggerSection
       channels [None {type=MovieSceneScriptingEventChannel, const ["<Struct 'MovieSceneEvent' (0x467a555d7a0) {}>", 1]}]
      }]
    }]
  }
 }
 Anim_Contested {
  binding_count=1
  length_seconds=0.4333
  object=&self:Anim_Contested
  sequence {
   bindings [Image_Marker {
     id=#3
     possessed_class=Image
     tracks [
      {
       type=MovieSceneWidgetMaterialTrack
       sections [{
         type=MovieSceneParameterSection
         end_frame=13
         start_frame=0
         channels [Base Scale {type=MovieSceneScriptingFloatChannel, thinned_from=5, keys [[0,0.7],[5,0.75],[12,0.68],[13,0.7]]}]
        }]
      }
      {
       type=MovieScene2DTransformTrack
       sections [{
         type=MovieScene2DTransformSection
         end_frame=13
         start_frame=0
         channels [
          Scale.X {type=MovieSceneScriptingFloatChannel, keys [[0,1],[5,1.05],[10,0.9863],[13,1]]}
          Scale.Y {type=MovieSceneScriptingFloatChannel, keys [[0,1],[5,1.05],[10,0.9863],[13,1]]}
         ]
        }]
      }
     ]
    }]
   display_rate {numerator=30}
  }
 }
 Anim_Pulse_Contested {
  binding_count=1
  length_seconds=1.5
  object=&self:Anim_Pulse_Contested
  sequence {
   bindings [Image_Marker_Pulse {
     id=#2
     possessed_class=Image
     tracks [
      {
       type=MovieSceneWidgetMaterialTrack
       sections [{
         type=MovieSceneParameterSection
         channels [
          Progress Background Opacity {type=MovieSceneScriptingFloatChannel, keys [[0,1],[15,0]]}
          Base Scale {type=MovieSceneScriptingFloatChannel, keys [[0,0.7],[15,1]]}
          Progress Opacity {type=MovieSceneScriptingFloatChannel, const [0,1]}
          Progress Background Thickness {type=MovieSceneScriptingFloatChannel, const [1,1]}
          Color for TransparencyR {type=MovieSceneScriptingFloatChannel, const [1,1]}
          Color for TransparencyG {type=MovieSceneScriptingFloatChannel, const [1,1]}
          Color for TransparencyB {type=MovieSceneScriptingFloatChannel, const [1,1]}
          Color for TransparencyA {type=MovieSceneScriptingFloatChannel, const [1,1]}
         ]
        }]
      }
      {
       type=MovieSceneByteTrack
       display_name=Visibility
       sections [{type=MovieSceneByteSection, end_frame=15, start_frame=0, channels [None {type=MovieSceneScriptingByteChannel, keys [[0,4],[15,1]]}]}]
      }
     ]
    }]
   display_rate {numerator=30}
   tracks [{
     type=MovieSceneEventTrack
     sections [{
       type=MovieSceneEventTriggerSection
       channels [None {type=MovieSceneScriptingEventChannel, const ["<Struct 'MovieSceneEvent' (0x467a555e9c0) {}>", 1]}]
      }]
    }]
  }
 }
 Anim_Pulse_Contested_Reset {
  binding_count=1
  length_seconds=0.0333
  object=&self:Anim_Pulse_Contested_Reset
  sequence {
   bindings [Image_Marker_Pulse {
     id=#2
     possessed_class=Image
     tracks [
      {
       type=MovieSceneWidgetMaterialTrack
       sections [{
         type=MovieSceneParameterSection
         channels [
          Progress Background Opacity {type=MovieSceneScriptingFloatChannel, const [0,1]}
          Base Scale {type=MovieSceneScriptingFloatChannel, const [0.7,1]}
          Progress Opacity {type=MovieSceneScriptingFloatChannel, const [0,1]}
          Progress Background Thickness {type=MovieSceneScriptingFloatChannel, const [0,1]}
         ]
        }]
      }
      {
       type=MovieSceneByteTrack
       display_name=Visibility
       sections [{type=MovieSceneByteSection, end_frame=1, start_frame=0, channels [None {type=MovieSceneScriptingByteChannel, const [1,1]}]}]
      }
      {
       type=MovieScene2DTransformTrack
       sections [{type=MovieScene2DTransformSection, end_frame=1, start_frame=0, channels [Angle {type=MovieSceneScriptingFloatChannel, const [-45,1]}]}]
      }
     ]
    }]
   display_rate {numerator=30}
  }
 }
 Anim_Pulse_Contested_OnObjectiveOwning {
  binding_count=1
  length_seconds=0.5667
  object=&self:Anim_Pulse_Contested_OnObjectiveOwning
  sequence {
   bindings [Image_Marker_Pulse {
     id=#4
     possessed_class=Image
     tracks [
      {
       type=MovieScene2DTransformTrack
       sections [{
         type=MovieScene2DTransformSection
         end_frame=17
         start_frame=0
         channels [Angle {type=MovieSceneScriptingFloatChannel, keys [[0,-45],[14,0],[17,0]]}]
        }]
      }
      {
       type=MovieSceneWidgetMaterialTrack
       sections [{
         type=MovieSceneParameterSection
         end_frame=17
         start_frame=0
         channels [Base Scale {type=MovieSceneScriptingFloatChannel, thinned_from=8, keys [[0,0.7],[3,0.65],[7,0.75],[9,0.7],[12,0.7],[14,0.65],[17,0.7]]}]
        }]
      }
     ]
    }]
   display_rate {numerator=30}
  }
 }
 Anim_Pulse_Contested_IsObjectiveOwned {
  binding_count=1
  length_seconds=0.0333
  object=&self:Anim_Pulse_Contested_IsObjectiveOwned
  sequence {
   bindings [Image_Marker_Pulse {
     id=#4
     possessed_class=Image
     tracks [
      {
       type=MovieScene2DTransformTrack
       sections [{type=MovieScene2DTransformSection, end_frame=1, start_frame=0, channels [Angle {type=MovieSceneScriptingFloatChannel, const [0,1]}]}]
      }
      {
       type=MovieSceneWidgetMaterialTrack
       sections [{type=MovieSceneParameterSection, end_frame=1, start_frame=0, channels [Base Scale {type=MovieSceneScriptingFloatChannel, const [0.7,1]}]}]
      }
     ]
    }]
   display_rate {numerator=30}
  }
 }
 Anim_Flare {
  binding_count=2
  length_seconds=1.35
  object=&self:Anim_Flare
  sequence {bindings [
    Image_Flare {
     id=#5
     possessed_class=Image
     tracks [
      {
       type=MovieSceneWidgetMaterialTrack
       sections [{
         type=MovieSceneParameterSection
         end_frame=27
         start_frame=0
         channels [
          R Opacity {type=MovieSceneScriptingFloatChannel, keys [[0,0],[3,1],[17,1],[27,0.3]]}
          R Rotation Angle {type=MovieSceneScriptingFloatChannel, keys [[0,0],[27,1]]}
          G Opacity {type=MovieSceneScriptingFloatChannel, keys [[0,0],[6,0.6],[22,0],[27,0.2]]}
          G Rotation Angle {type=MovieSceneScriptingFloatChannel, keys [[0,0],[27,-0.8]]}
          B Opacity {type=MovieSceneScriptingFloatChannel, keys [[0,0],[9,0.9025],[27,0.35]]}
          B Rotation Angle {type=MovieSceneScriptingFloatChannel, keys [[0,0],[27,0.6]]}
          A Opacity {type=MovieSceneScriptingFloatChannel, keys [[3,0],[12,0.7],[27,0.18]]}
          A Rotation Angle {type=MovieSceneScriptingFloatChannel, keys [[7,0],[27,0.3]]}
         ]
        }]
      }
      {
       type=MovieScene2DTransformTrack
       sections [{
         type=MovieScene2DTransformSection
         end_frame=27
         start_frame=0
         channels [
          Translation.X {type=MovieSceneScriptingFloatChannel, keys [[0,-39.1711],[24,0]]}
          Translation.Y {type=MovieSceneScriptingFloatChannel, keys [[0,40],[24,-60]]}
          Angle {type=MovieSceneScriptingFloatChannel}
          Scale.X {type=MovieSceneScriptingFloatChannel, keys [[0,2],[14,5],[24,4]]}
          Scale.Y {type=MovieSceneScriptingFloatChannel, keys [[0,2],[14,5],[24,4]]}
         ]
        }]
      }
      {
       type=MovieSceneFloatTrack
       display_name=Render Opacity
       sections [{type=MovieSceneFloatSection, end_frame=27, start_frame=0, channels [None {type=MovieSceneScriptingFloatChannel, const [1,1]}]}]
      }
     ]
    }
    Image_Rings {
     id=#6
     possessed_class=Image
     tracks [
      {
       type=MovieSceneFloatTrack
       display_name=Render Opacity
       sections [{
         type=MovieSceneFloatSection
         end_frame=27
         start_frame=0
         channels [None {type=MovieSceneScriptingFloatChannel, thinned_from=6, keys [[0,1],[7,1],[13,0.748],[17,0.5],[25,0]]}]
        }]
      }
      {
       type=MovieSceneWidgetMaterialTrack
       sections [{
         type=MovieSceneParameterSection
         end_frame=27
         start_frame=0
         channels [
          Ring 1 Size {type=MovieSceneScriptingFloatChannel, keys [[0,0],[20,0.6]]}
          RIng 1 Opacity {type=MovieSceneScriptingFloatChannel, keys [[0,0],[2,0.35],[14,0]]}
          Ring 1 Glow Max {type=MovieSceneScriptingFloatChannel, keys [[4,0.1],[7,0.75]]}
          Ring 2 Size {type=MovieSceneScriptingFloatChannel, keys [[0,0],[2,0],[22,0.75]]}
          Ring 2 Opacity {type=MovieSceneScriptingFloatChannel, keys [[2,0],[4,0.5],[17,0]]}
          Ring 2 Glow Max {type=MovieSceneScriptingFloatChannel, keys [[4,0.065],[17,0.035]]}
          Ring 2 Glow Min {type=MovieSceneScriptingFloatChannel, keys [[4,0],[17,-0.19]]}
          Ring 3 Size {type=MovieSceneScriptingFloatChannel, keys [[6,0.1246],[15,0.25]]}
          Ring 3 Opacity {type=MovieSceneScriptingFloatChannel, const [0,2]}
          Ring 3 Glow Max {type=MovieSceneScriptingFloatChannel, keys [[6,0.01],[10,0.11]]}
         ]
        }]
      }
      {
       type=MovieScene2DTransformTrack
       sections [{
         type=MovieScene2DTransformSection
         end_frame=25
         start_frame=0
         channels [
          Translation.X {type=MovieSceneScriptingFloatChannel, keys [[0,-40],[24,0]]}
          Translation.Y {type=MovieSceneScriptingFloatChannel, keys [[0,40],[24,-60]]}
          Scale.X {type=MovieSceneScriptingFloatChannel, const [3,1]}
          Scale.Y {type=MovieSceneScriptingFloatChannel, const [3,1]}
         ]
        }]
      }
     ]
    }
   ]}
 }
 Anim_Flare_Out {
  binding_count=1
  length_seconds=0.2506
  object=&self:Anim_Flare_Out
  sequence {bindings [Image_Flare {
     id=#5
     possessed_class=Image
     tracks [
      {
       type=MovieSceneWidgetMaterialTrack
       sections [{
         type=MovieSceneParameterSection
         end_frame=5
         start_frame=0
         channels [
          R Opacity {type=MovieSceneScriptingFloatChannel, const [0.3,1]}
          R Rotation Angle {type=MovieSceneScriptingFloatChannel, const [1,1]}
          G Opacity {type=MovieSceneScriptingFloatChannel, const [0.2,1]}
          G Rotation Angle {type=MovieSceneScriptingFloatChannel, const [-0.8,1]}
          B Opacity {type=MovieSceneScriptingFloatChannel, const [0.35,1]}
          B Rotation Angle {type=MovieSceneScriptingFloatChannel, const [0.6,1]}
          A Opacity {type=MovieSceneScriptingFloatChannel, const [0.18,1]}
          A Rotation Angle {type=MovieSceneScriptingFloatChannel, const [0.3,1]}
         ]
        }]
      }
      {
       type=MovieScene2DTransformTrack
       sections [{
         type=MovieScene2DTransformSection
         end_frame=4
         start_frame=0
         channels [
          Translation.X {type=MovieSceneScriptingFloatChannel, const [0,1]}
          Translation.Y {type=MovieSceneScriptingFloatChannel, const [-60,1]}
          Angle {type=MovieSceneScriptingFloatChannel}
          Scale.X {type=MovieSceneScriptingFloatChannel, const [4,1]}
          Scale.Y {type=MovieSceneScriptingFloatChannel, const [4,1]}
         ]
        }]
      }
      {
       type=MovieSceneFloatTrack
       display_name=Render Opacity
       sections [{type=MovieSceneFloatSection, end_frame=5, start_frame=0, channels [None {type=MovieSceneScriptingFloatChannel, keys [[0,1],[5.012,0]]}]}]
      }
     ]
    }]}
 }
 Anim_Flare_InAndOut {
  binding_count=2
  length_seconds=1.35
  object=&self:Anim_Flare_InAndOut
  sequence {bindings [
    Image_Flare {
     id=#5
     possessed_class=Image
     tracks [
      {
       type=MovieSceneWidgetMaterialTrack
       sections [{
         type=MovieSceneParameterSection
         end_frame=27
         start_frame=0
         channels [
          R Opacity {type=MovieSceneScriptingFloatChannel, keys [[0,0],[3,1],[17,1],[22,0]]}
          R Rotation Angle {type=MovieSceneScriptingFloatChannel, keys [[0,0],[27,1]]}
          G Opacity {type=MovieSceneScriptingFloatChannel, keys [[0,0],[6,0.6],[20,0],[24,0]]}
          G Rotation Angle {type=MovieSceneScriptingFloatChannel, keys [[0,0],[27,-0.8]]}
          B Opacity {type=MovieSceneScriptingFloatChannel, keys [[0,0],[9,0.9025],[22,0]]}
          B Rotation Angle {type=MovieSceneScriptingFloatChannel, keys [[0,0],[22,0.5947],[27,0.6]]}
          A Opacity {type=MovieSceneScriptingFloatChannel, keys [[3,0],[12,0.7],[27,0.18]]}
          A Rotation Angle {type=MovieSceneScriptingFloatChannel, keys [[7,0],[27,0.3]]}
         ]
        }]
      }
      {
       type=MovieScene2DTransformTrack
       sections [{
         type=MovieScene2DTransformSection
         end_frame=27
         start_frame=0
         channels [
          Translation.X {type=MovieSceneScriptingFloatChannel, keys [[0,-39.1711],[24,40]]}
          Translation.Y {type=MovieSceneScriptingFloatChannel, keys [[0,40],[24,-40]]}
          Angle {type=MovieSceneScriptingFloatChannel}
          Scale.X {type=MovieSceneScriptingFloatChannel, keys [[0,2],[14,5],[24,4]]}
          Scale.Y {type=MovieSceneScriptingFloatChannel, keys [[0,2],[14,5],[24,4]]}
         ]
        }]
      }
      {
       type=MovieSceneFloatTrack
       display_name=Render Opacity
       sections [{type=MovieSceneFloatSection, end_frame=27, start_frame=0, channels [None {type=MovieSceneScriptingFloatChannel, keys [[0,1],[17,1],[27,0]]}]}]
      }
     ]
    }
    Image_Rings {
     id=#6
     possessed_class=Image
     tracks [
      {
       type=MovieSceneFloatTrack
       display_name=Render Opacity
       sections [{
         type=MovieSceneFloatSection
         end_frame=27
         start_frame=0
         channels [None {type=MovieSceneScriptingFloatChannel, thinned_from=6, keys [[0,1],[7,1],[13,0.748],[17,0.5],[25,0]]}]
        }]
      }
      {
       type=MovieSceneWidgetMaterialTrack
       sections [{
         type=MovieSceneParameterSection
         end_frame=27
         start_frame=0
         channels [
          Ring 1 Size {type=MovieSceneScriptingFloatChannel, keys [[0,0],[20,0.6]]}
          RIng 1 Opacity {type=MovieSceneScriptingFloatChannel, keys [[0,0],[2,0.35],[14,0]]}
          Ring 1 Glow Max {type=MovieSceneScriptingFloatChannel, keys [[4,0.1],[7,0.75]]}
          Ring 2 Size {type=MovieSceneScriptingFloatChannel, keys [[0,0],[2,0],[22,0.75]]}
          Ring 2 Opacity {type=MovieSceneScriptingFloatChannel, keys [[2,0],[4,0.5],[17,0]]}
          Ring 2 Glow Max {type=MovieSceneScriptingFloatChannel, keys [[4,0.065],[17,0.035]]}
          Ring 2 Glow Min {type=MovieSceneScriptingFloatChannel, keys [[4,0],[17,-0.19]]}
          Ring 3 Size {type=MovieSceneScriptingFloatChannel, keys [[6,0.1246],[15,0.25]]}
          Ring 3 Opacity {type=MovieSceneScriptingFloatChannel, const [0,2]}
          Ring 3 Glow Max {type=MovieSceneScriptingFloatChannel, keys [[6,0.01],[10,0.11]]}
         ]
        }]
      }
      {
       type=MovieScene2DTransformTrack
       sections [{
         type=MovieScene2DTransformSection
         end_frame=25
         start_frame=0
         channels [
          Translation.X {type=MovieSceneScriptingFloatChannel, keys [[0,-40],[24,40]]}
          Translation.Y {type=MovieSceneScriptingFloatChannel, keys [[0,40],[24,-40]]}
          Scale.X {type=MovieSceneScriptingFloatChannel, const [3,1]}
          Scale.Y {type=MovieSceneScriptingFloatChannel, const [3,1]}
         ]
        }]
      }
     ]
    }
   ]}
 }
]
graphs [
 CaptureOwnerUpdated {nodes [
   n1 = CaptureOwnerUpdated!() -> (then, CaptureOwner:Byte, PrevCaptureOwner:Byte) {is=title K2Node_FunctionEntry}
   > OnObjectiveOwningTeamChanged!(
    execute
    self:Fort Core Gamer Objective Radial Object Reference
    NewTeam:Byte=@n1
    OldTeam:Byte=@n1.PrevCaptureOwner
    bForceUpdate:bool=false
   ) -> (then)
  ]}
 CaptureProgressUpdated {nodes [
   CaptureProgressUpdated!() -> (then, NewProgress:float=~) {is=title K2Node_FunctionEntry}
   > n1 = Get!(execute:Exec) -> (then:Exec, else:Exec, CaptureComponent:Capture Component Object Reference)
   > n2 = GetDynamicMaterial(self=Get Image_Marker() -> (Image_Marker:Image Object Reference))
   > SetScalarParameterValue(self=@n2, ParameterName=Percentage, Value=Get CaptureProgress!(self:Capture Component Object Reference=@n1) -> (CaptureProgress:float))
  ]}
 CaptureStateUpdated {nodes [
   n1 = CaptureStateUpdated!() -> (then, NewCaptureState:ECaptureState Enum=Neutral, PreviousCaptureState:ECaptureState Enum=Neutral) {is=title K2Node_FunctionEntry}
   > n2 = Get()
   > UpdateCaptureState!(
    execute
    self:Fort Core Gamer Objective Radial Object Reference
    NewCaptureState:ECaptureState Enum=@n1
    OldCaptureState:ECaptureState Enum=@n1.PreviousCaptureState
    TrackedStateOwnerTeam:Byte=Get StateOwner!(self:Capture Component Object Reference=@n2) -> (StateOwner:Byte=0)
    Progress:float=Get CaptureProgress(self=@n2)
   ) -> (then)
  ]}
 DistanceCheck {nodes [
   DistanceCheck() {is=title K2Node_FunctionEntry}
   n1 = Get CaptureComponent() -> (CaptureComponent:Capture Component Object Reference)
   n2 = Get CaptureArea() -> (CaptureArea:Building Prop Capture Area Object Reference)
   n3 = GetOwningPlayerPawn!(self:User Widget Object Reference) -> (ReturnValue:Pawn Object Reference)
   > n4 = Set ObjectsValid(ObjectsValid:bool=AND Boolean!(A=IsValid(Object=@n3), B=IsValid(Object=@n1), C:bool=IsValid(Object=@n2)) -> (ReturnValue)) -> (Output_Get:bool)
   > n5 = Branch(Condition=@n4)
   n6 = Get CaptureArea()
   n7 = GetOwningPlayerPawn()
   > Set CurrentDistance(CurrentDistance:double="Distance (Vector)"!(V1:Vector=Get Actor Location(self=@n6), V2:Vector=Get Actor Location(self=@n7)) -> (ReturnValue:double)) -> (Output_Get:double)
   n8 = Remap Float(
    Input=Get CurrentDistance() -> (CurrentDistance:double)
    OldRange=MakeVector2D(X=Get MinDistance() -> (MinDistance:double), Y=Get MaxDistance() -> (MaxDistance:double))
    NewRange=MakeVector2D(X=Get MinDistanceScale() -> (MinDistanceScale:double), Y=Get MaxDistanceScale() -> (MaxDistanceScale:double))
    Clamp=true
   ) -> (Output=~)
   > SetRenderScale!(execute, self:Widget Object Reference, Scale_X:double=@n8, Scale_Y:double=@n8) -> (then)
   n9 = Get CaptureArea()
   > Return Node!(
    execute
    InRange:bool="float <= float"(
     A=Get CurrentDistance()
     B=GetCaptureIconData!(self:Building Prop Capture Area Object Reference=@n9) -> (
      OutCaptureIconData_IconTexture:Texture 2D Soft Object Reference
      OutCaptureIconData_TextScale:float=~
      OutCaptureIconData_VisibleDistance:float=~
      OutCaptureIconData_bDisplaysDistanceText:bool=~
      OutCaptureIconData_bDisplaysOwningTeam:bool=~
      OutCaptureIconData_bRequiresLineOfSight:bool=~
      OutCaptureIconData_bShowOffscreenArrow:bool=~
      OutCaptureIconData_bEnabled:bool=~
      OutCaptureIconData_TextOverrides:Array of Strings
     ).OutCaptureIconData_VisibleDistance
    ) -> (ReturnValue=~)
   )
   Return Node(execute=@n5.else, InRange=true)
  ]}
 EventGraph {nodes [
   n1 = LoadIndicatorIconByIndex!() -> (OutputDelegate, then, Index:int) {is=title K2Node_CustomEvent}
   > n2 = Branch(Condition="integer > integer"(A=@n1.Index, B=-1))
   > LoadIndicatorIcon!(
    execute
    self
    InSoftTextureReference:Texture 2D Soft Object Reference="Get (a copy)"(
     Array:Array of Texture 2D Soft Object References=Get ObjectiveIndicatorIcons() -> (ObjectiveIndicatorIcons:Array of Texture 2D Soft Object References)
     Dimension 1=int + int(A=@n1.Index)
    ) -> (Output:Texture 2D Soft Object Reference)
   ) -> (then)
   UpdateMarkerIcon!(execute=@n2.else, self, InUpdateOpacity:bool=true, InIconOpacity:float=0, InUpdateTexture:bool=false, InIconTexture:Texture Object Reference) -> (then)
   n3 = LoadIndicatorIcon!() -> (OutputDelegate, then, InSoftTextureReference:Texture 2D Soft Object Reference) {is=title K2Node_CustomEvent}
   > n4 = Async Load Asset(Asset=@n3.InSoftTextureReference)
   n5 = Cast To Texture2D(execute=@n4.Completed, Object=@n4)
   > UpdateMarkerIcon(InUpdateOpacity=true, InIconOpacity=1, InUpdateTexture=true, InIconTexture=@n5)
   Event OnObjectiveContested() {is=title K2Node_Event}
   > UpdateMarkerColorsByCaptureState!(execute, self, InCaptureState:ECaptureState Enum=Contested, InTeamByte:Byte=0) -> (then)
   n6 = Event OnTrackedObjectiveSet!() -> (OutputDelegate, then, NewTrackedObjective:"Tracked Objective Structure (by ref)") {is=title K2Node_Event}
   n7 = Break Tracked Objective!(TrackedObjective:"Tracked Objective Structure (by ref)"=@n6.NewTrackedObjective) -> (
    TrackedIndex:int
    TeamIndex:Byte
    OldTeamIndex:Byte
    IndicatorPresetIndex:int
    ObjectiveSortOrder:int
    ProgressPercentage:float=~
    TrackedState:Byte
    TrackedStateOwnerTeam:Byte
    OldTrackedState:Byte
    ObjectiveType:Byte
    bIsInteractionAllowed:bool=~
    bOldIsInteractionAllowed:bool=~
   ) {is=title K2Node_BreakStruct}
   > LoadIndicatorIconByIndex!(execute, self, Index:int=@n7.IndicatorPresetIndex) -> (then) {is=title K2Node_CallFunction}
   > OnObjectiveOwningTeamChanged(NewTeam=@n7.TeamIndex, OldTeam=0, bForceUpdate=true)
   Event Construct()
   > UpdateAllMarkerMaterialParams!(
    execute
    self
    Percentage:double=0
    AnglePercent:double=0.125
    inTeamColors:S UI Creative Core Gamer Team Colors Structure=GetTeamColorByState!(
     self:PDA UI Creative Core Gamer Object Reference=Get PrimaryDataAsset() -> (PrimaryDataAsset:PDA UI Creative Core Gamer Object Reference)
     InTeamState:E_UI_Creative_CoreGamer_TeamStates Enum=NewEnumerator3
    ) -> (OutColorSet:S UI Creative Core Gamer Team Colors Structure)
   ) -> (then)
   > n8 = Cast To FortPlayerStateAthena!(
    execute
    Object=GetRelevantPlayerState!(PlayerController:Player Controller Object Reference=GetOwningPlayer()) -> (ReturnValue:Player State Object Reference)
   ) -> (then, CastFailed, AsFort Player State Athena:Fort Player State Athena Object Reference) {is=title K2Node_DynamicCast}
   > Bind Event to On Team Index Changed Delegate(self:Fort Player State Athena Object Reference=@n8, Delegate=Create Event()) {is=title K2Node_AddDelegate}
   n9 = Event PreConstruct()
   > Branch(Condition=@n9.IsDesignTime)
   > UpdateAllMarkerMaterialParams(Percentage=0.7, AnglePercent=0.125, inTeamColors=GetTeamColorByState(self=Get PrimaryDataAsset(), InTeamState=NewEnumerator3))
   n10 = Event OnMinigameStateChanged!() -> (OutputDelegate, then, NewMinigameState:EFortMinigameState Enum=PreGame) {is=title K2Node_Event}
   > n11 = Switch on EFortMinigameState!(execute, Selection:EFortMinigameState Enum=@n10.NewMinigameState) -> (
    PreGame:Exec
    Setup:Exec
    Transitioning:Exec
    WaitingForCameras:Exec
    Warmup:Exec
    InProgress:Exec
    PostGameTimeDilation:Exec
    PostRoundEnd:Exec
    PostGameEnd:Exec
    PostGameAbandon:Exec
    PostGameReset:Exec
    Teardown:Exec
   ) {is=title K2Node_SwitchEnum}
   OnObjectiveOwningTeamChanged(execute=[@n11.PostGameReset, @n11.PreGame], NewTeam=0, OldTeam=0, bForceUpdate=true)
   Event OnObjectiveBecameNeutral() {is=title K2Node_Event}
   > UpdateAllMarkerMaterialParams(Percentage=0, AnglePercent=0.125, inTeamColors=GetTeamColorByState(self=Get PrimaryDataAsset(), InTeamState=NewEnumerator3))
   n12 = Event OnObjectiveCaptured!() -> (OutputDelegate, then, OwningTeam:Byte=0) {is=title K2Node_Event}
   > UpdateProgress!(execute, self, Percentage:float=1) -> (then)
   > UpdateMarkerColorsByCaptureState(InCaptureState=Captured, InTeamByte=@n12.OwningTeam)
   K2Node_Event_4 = UpdateProgress!() -> (OutputDelegate, then, Percentage:float=~) {is=title K2Node_CustomEvent}
   > n13 = GetDynamicMaterial(self=Get Image_Marker())
   > SetScalarParameterValue(self=@n13, ParameterName=Percentage, Value=@K2Node_Event_4.Percentage)
   n14 = Event OnObjectiveBeingCaptured!() -> (OutputDelegate, then, CapturingTeam:Byte=0, Percentage:float) {is=title K2Node_Event}
   > UpdateProgress(Percentage=@n14.Percentage) {is=title K2Node_CallFunction}
   > UpdateMarkerColorsByCaptureState(InCaptureState=Capturing, InTeamByte=@n14.CapturingTeam)
   n15 = Event OnObjectiveBeingDecaptured!() -> (OutputDelegate, then, DecapturingTeam:Byte=0, Percentage:float) {is=title K2Node_Event}
   > UpdateProgress(Percentage=@n15.Percentage)
   > UpdateMarkerColorsByCaptureState(InCaptureState=Decapturing, InTeamByte=@n15.DecapturingTeam)
   n16 = Event OnObjectiveBeingNeutralized!() -> (OutputDelegate, then, NeutralizingTeam:Byte=0, Percentage:float) {is=title K2Node_Event}
   > UpdateProgress(Percentage=@n16.Percentage)
   > UpdateMarkerColorsByCaptureState(InCaptureState=Neutralizing, InTeamByte=@n16.NeutralizingTeam)
   n17 = Event InitFromObject!() -> (OutputDelegate, then, InitObject:Object Reference) {is=title K2Node_Event}
   > n18 = Cast To BuildingProp_CaptureArea!(execute, Object=@n17.InitObject) -> (then, CastFailed, AsBuilding Prop Capture Area:Building Prop Capture Area Object Reference) {is=title K2Node_DynamicCast}
   > n19 = Set CaptureArea(CaptureArea:Building Prop Capture Area Object Reference=@n18) -> (Output_Get:Building Prop Capture Area Object Reference)
   > Set CaptureComponent(CaptureComponent:Capture Component Object Reference=Get CaptureComponent!(self:Building Prop Capture Area Object Reference=@n19) -> (CaptureComponent)) -> (Output_Get:Capture Component Object Reference)
   > SetupInWorldWidget()
   n20 = GetMinigame!(self:Fort Core Gamer Objective Radial Object Reference) -> (ReturnValue:Fort Minigame Object Reference)
   > Is Valid(InputObject=@n20)
   > Bind Event to State Changed Delegate(self:Fort Minigame Object Reference=@n20, Delegate=Create Event()) {is=title K2Node_AddDelegate}
   > MinigameStateChanged!(
    execute
    self
    Minigame:Fort Minigame Object Reference=@n20
    MinigameState:EFortMinigameState Enum=Get CurrentState!(self:Fort Minigame Object Reference=@n20) -> (CurrentState:EFortMinigameState Enum=PreGame)
   ) -> (then)
   > OnClampDirectionChanged!(execute, self, ClampToScreenDirection:EClampToScreenDirection Enum=None) -> (then)
   n21 = Event OnObjectiveBeingDeneutralized!() -> (OutputDelegate, then, DeneutralizingTeam:Byte=0, Percentage:float) {is=title K2Node_Event}
   > UpdateProgress(Percentage=@n21.Percentage)
   > UpdateMarkerColorsByCaptureState(InCaptureState=Deneutralizing, InTeamByte=@n21.DeneutralizingTeam)
  ]}
 GetDistanceBetweenActorsText {nodes [
   n1 = GetDistanceBetweenActorsText!() -> (then, ActorA:Actor Object Reference, ActorB:Actor Object Reference) {is=title K2Node_FunctionEntry}
   > n2 = Branch(Condition=AND Boolean!(A=IsValid(Object=@n1), B=IsValid(Object=@n1.ActorB)) -> (ReturnValue))
   n3 = float / float(A="Distance (Vector)"(V1=Get Actor Location(self=@n1), V2=Get Actor Location(self=@n1.ActorB)), B=100) {is=title K2Node_CallFunction}
   > Return Node!(execute, Text:Text=Format Text!(Format="LOCTEXT(\"{Distance}m\")", Distance:int=Floor(A=@n3) {is=title K2Node_CallFunction}) -> (Result))
   Return Node(execute=@n2.else)
  ]}
 GetTeamAffiliation {nodes [
   GetTeamAffiliation() {is=title K2Node_FunctionEntry}
   n1 = Get CachedOwningTeam() -> (CachedOwningTeam:Byte=0)
   > n2 = Branch(Condition=IsNeutralTeam!(InTeamByte:Byte=@n1) -> (IsNeutral:bool))
   > Return Node!(execute, TeamAffiliation:EFortTeamAffiliation Enum=Neutral)
   n3 = Branch(execute=@n2.else, Condition=IsLocalTeam!(InTeamByte:Byte=@n1) -> (IsLocal:bool))
   > Return Node(TeamAffiliation=Friendly)
   Return Node(execute=@n3.else, TeamAffiliation=Hostile)
  ]}
 InWorldTimerUpdate {nodes [
   InWorldTimerUpdate() {is=title K2Node_FunctionEntry}
   n1 = Get CaptureComponent!() -> (CaptureComponent)
   n2 = Get CaptureArea()
   > Branch(Condition=AND Boolean(A=IsValid(Object=@n2), B=IsValid(Object=@n1)))
   > n3 = Set CaptureIconData!(
    execute
    CaptureIconData:Capture Icon Data Structure=GetCaptureIconData!(self=Get CaptureArea()) -> (OutCaptureIconData:Capture Icon Data Structure)
   ) -> (
    then
    Output_Get_IconTexture:Texture 2D Soft Object Reference
    Output_Get_TextScale:float=~
    Output_Get_VisibleDistance:float=~
    Output_Get_bDisplaysDistanceText:bool=~
    Output_Get_bDisplaysOwningTeam:bool=~
    Output_Get_bRequiresLineOfSight:bool=~
    Output_Get_bShowOffscreenArrow:bool=~
    Output_Get_bEnabled:bool=~
    Output_Get_TextOverrides:Array of Strings
   )
   > n4 = Branch(Condition=@n3.Output_Get_bEnabled)
   > n5 = Sequence!(execute) -> (then_0, then_1, then_2:Exec, then_3:Exec)
   > UpdateOpacity()
   n6 = LineOfSightCheck!(execute=@n5.then_2, self) -> (then, HasLineOfSight:bool)
   > n7 = DistanceCheck!(execute, self) -> (then, InRange:bool) {is=title K2Node_CallFunction}
   n8 = Get Minigame State() -> (Minigame State:EFortMinigameState Enum=PreGame)
   n9 = OR Boolean(A="Equal (Enum)"(A:EFortMinigameState Enum=@n8, B:EFortMinigameState Enum=InProgress), B="Equal (Enum)"(A=@n8, B=PreGame))
   > SetVisibility(InVisibility=Select(Option 0=Collapsed, Option 1=SelfHitTestInvisible, Index=AND Boolean!(A=@n7, B=@n6, C:bool=@n9) -> (ReturnValue)))
   n10 = Branch(
    execute=@n5.then_3
    Condition=Get!() -> (
     CaptureIconData_IconTexture:Texture 2D Soft Object Reference
     CaptureIconData_TextScale:float=~
     CaptureIconData_VisibleDistance:float=~
     CaptureIconData_bDisplaysDistanceText:bool=~
     CaptureIconData_bDisplaysOwningTeam:bool=~
     CaptureIconData_bRequiresLineOfSight:bool=~
     CaptureIconData_bShowOffscreenArrow:bool=~
     CaptureIconData_bEnabled:bool=~
     CaptureIconData_TextOverrides:Array of Strings
    ).CaptureIconData_bDisplaysDistanceText
   )
   n11 = Get DistanceText() -> (DistanceText:Common Text Object Reference)
   > SetVisibility(self=@n11, InVisibility=SelfHitTestInvisible)
   n12 = GetOwningPlayerPawn()
   n13 = Get CaptureArea()
   > Set Text(
    Text=TextToUpper!(InText:"Text (by ref)"=GetDistanceBetweenActorsText!(ActorA:Actor Object Reference=@n13, ActorB:Actor Object Reference=@n12) -> (Text:Text) {is=title K2Node_CallFunction}) -> (ReturnValue:Text)
    self=Get DistanceText()
   )
   SetVisibility(execute=@n10.else, self=@n11, InVisibility=Collapsed)
   SetVisibility(execute=@n5.then_1, self=Get OptionalText() -> (OptionalText:Common Text Object Reference), InVisibility=SelfHitTestInvisible)
   > UpdateTextByAffiliation()
   SetVisibility(execute=@n4.else, InVisibility=Collapsed)
  ]}
 IsLocalTeam {nodes [
   n1 = IsLocalTeam!() -> (then, InTeamByte:Byte) {is=title K2Node_FunctionEntry}
   > n2 = Cast To FortPlayerController(Object=GetOwningPlayer())
   > Return Node!(
    execute
    IsLocal:bool="Equal (Byte)"(A:Byte=@n1, B:Byte=GetActorTeam!(Actor:Actor Object Reference=@n2) -> (ReturnValue:Byte=0)) -> (ReturnValue:bool=~) {is=title K2Node_PromotableOperator}
   )
   Return Node(execute=@n2.CastFailed, IsLocal=false)
  ]}
 IsNeutralTeam {nodes [
   n1 = IsNeutralTeam!() -> (then, InTeamByte:Byte) {is=title K2Node_FunctionEntry}
   > Return Node!(execute, IsNeutral:bool="Equal (Byte)"(A=@n1) -> (ReturnValue=~))
  ]}
 LineOfSightCheck {nodes [
   LineOfSightCheck() {is=title K2Node_FunctionEntry}
   > Set TempHasLineOfSight(TempHasLineOfSight:bool=true) -> (Output_Get:bool)
   > n1 = Sequence!(execute) -> (then_0, then_1)
   > Branch(Condition=GetCaptureIconData!(self=Get CaptureArea()) -> (
     OutCaptureIconData_IconTexture:Texture 2D Soft Object Reference
     OutCaptureIconData_TextScale:float=~
     OutCaptureIconData_VisibleDistance:float=~
     OutCaptureIconData_bDisplaysDistanceText:bool=~
     OutCaptureIconData_bDisplaysOwningTeam:bool=~
     OutCaptureIconData_bRequiresLineOfSight:bool=~
     OutCaptureIconData_bShowOffscreenArrow:bool=~
     OutCaptureIconData_bEnabled:bool=~
     OutCaptureIconData_TextOverrides:Array of Strings
    ).OutCaptureIconData_bRequiresLineOfSight)
   n2 = GetOwningPlayerPawn()
   n3 = Get CaptureArea()
   > Branch(Condition=AND Boolean!(A=IsValid(Object=@n3), B=IsValid(Object=@n2)) -> (ReturnValue))
   n4 = Get CaptureArea()
   > n5 = Line Trace By Channel!(
    execute
    Start:Vector=GetCameraLocation!(self:Player Camera Manager Object Reference=GetOwningPlayerCameraManager!(self:User Widget Object Reference) -> (ReturnValue:Player Camera Manager Object Reference)) -> (ReturnValue:Vector)
    End:Vector=Get World Location!(self:Scene Component Object Reference=Get CaptureIconWidgetComponent!(self:Building Prop Capture Area Object Reference=@n4) -> (CaptureIconWidgetComponent:Fort Widget Component Object Reference)) -> (ReturnValue:Vector) {is=title K2Node_CallFunction}
    TraceChannel:ETraceTypeQuery Enum=TraceTypeQuery2
    bTraceComplex:bool=false
    ActorsToIgnore:Array of Actor Object References=Make Array!("[0]":Actor Object Reference=@n3, "[1]":Actor Object Reference=@n2) -> (Array:Array of Actor Object References)
    DrawDebugType:EDrawDebugTrace Enum
    bIgnoreSelf:bool
    TraceColor:Linear Color Structure
    TraceHitColor:Linear Color Structure
    DrawTime:float=Get TimerUpdateRate() -> (TimerUpdateRate:float)
   ) -> (then, OutHit:Hit Result Structure, ReturnValue:bool)
   > Set TempHasLineOfSight(TempHasLineOfSight=NOT Boolean(A=@n5.ReturnValue))
   Return Node!(execute=@n1.then_1, HasLineOfSight:bool=Get TempHasLineOfSight() -> (TempHasLineOfSight:bool))
  ]}
 MinigameStateChanged {nodes [
   n1 = MinigameStateChanged!() -> (then, Minigame:Fort Minigame Object Reference, MinigameState:EFortMinigameState Enum=PreGame) {is=title K2Node_FunctionEntry}
   > Set Minigame State(Minigame State:EFortMinigameState Enum=@n1.MinigameState) -> (Output_Get:EFortMinigameState Enum=PreGame)
  ]}
 OnClampDirectionChanged {nodes [
   n1 = OnClampDirectionChanged!() -> (then, ClampToScreenDirection:EClampToScreenDirection Enum=Left) {is=title K2Node_FunctionEntry}
   > n2 = Get!(execute:Exec) -> (then:Exec, else:Exec, CaptureArea:Building Prop Capture Area Object Reference)
   > n3 = Branch(Condition=GetCaptureIconData(self=@n2) -> (
     OutCaptureIconData_TextScale=~
     OutCaptureIconData_VisibleDistance=~
     OutCaptureIconData_bDisplaysDistanceText=~
     OutCaptureIconData_bDisplaysOwningTeam=~
     OutCaptureIconData_bRequiresLineOfSight=~
     OutCaptureIconData_bShowOffscreenArrow=~
     OutCaptureIconData_bEnabled=~
    ).OutCaptureIconData_bShowOffscreenArrow)
   > n4 = Switch on EClampToScreenDirection!(execute, Selection:EClampToScreenDirection Enum=@n1) -> (Left:Exec, Top:Exec, Right:Exec, Bottom:Exec, None:Exec) {is=title K2Node_SwitchEnum}
   SetVisibility(execute=@n4.Bottom, self=Get ArrowBottom() -> (ArrowBottom:Image Object Reference), InVisibility=SelfHitTestInvisible)
   > SetVisibility(
    self=[
     Get ArrowLeft() -> (ArrowLeft:Image Object Reference)
     Get ArrowRight() -> (ArrowRight:Image Object Reference)
     Get ArrowTop() -> (ArrowTop:Image Object Reference)
    ]
    InVisibility=Hidden
   )
   SetVisibility(execute=@n4.Left, self=Get ArrowLeft(), InVisibility=SelfHitTestInvisible)
   > SetVisibility(self=[Get ArrowBottom(), Get ArrowRight(), Get ArrowTop()], InVisibility=Hidden)
   SetVisibility(execute=@n4.Right, self=Get ArrowRight(), InVisibility=SelfHitTestInvisible)
   > SetVisibility(self=[Get ArrowBottom(), Get ArrowLeft(), Get ArrowTop()], InVisibility=Hidden)
   SetVisibility(execute=@n4.Top, self=Get ArrowTop(), InVisibility=SelfHitTestInvisible)
   > SetVisibility(self=[Get ArrowBottom(), Get ArrowLeft(), Get ArrowRight()], InVisibility=Hidden)
   SetVisibility(execute=@n4.None, self=[Get ArrowBottom(), Get ArrowLeft(), Get ArrowRight(), Get ArrowTop()], InVisibility=Hidden)
   SetVisibility(execute=[@n3.else, @n2.else], self=[Get ArrowTop(), Get ArrowBottom(), Get ArrowLeft(), Get ArrowRight()], InVisibility=Collapsed)
  ]}
 OnLocalPlayerTeamIndexChanged {nodes [
   OnLocalPlayerTeamIndexChanged() {is=title K2Node_FunctionEntry}
   > OnObjectiveOwningTeamChanged(NewTeam=Get CachedOwningTeam() -> (CachedOwningTeam=0), OldTeam=0, bForceUpdate=true)
   n1 = Get CachedOwningTeam() -> (CachedOwningTeam=0)
   > Branch(Condition="Not Equal (Byte)"(A:Byte=@n1, B:Byte) -> (ReturnValue:bool=~) {is=title K2Node_PromotableOperator})
   > UpdateMarkerColorsByCaptureState(InCaptureState=Captured, InTeamByte=@n1)
  ]}
 OnObjectiveOwningTeamChanged {nodes [
   n1 = OnObjectiveOwningTeamChanged!() -> (then, NewTeam:Byte=0, OldTeam:Byte=0, bForceUpdate:bool) {is=title K2Node_FunctionEntry}
   > n2 = Set CachedOwningTeam(CachedOwningTeam:Byte=@n1) -> (Output_Get:Byte=0)
   n3 = GetMinigame()
   > Is Valid(InputObject=@n3)
   n4 = GetMinigameTeamComponent!(self:Fort Minigame Object Reference=@n3) -> (ReturnValue:Fort Minigame Team Component Object Reference)
   > Is Valid(InputObject=@n4)
   > Set CachedOwningTeamName(CachedOwningTeamName:Text=GetTeamName!(self:Fort Minigame Team Component Object Reference=@n4, TeamIndex:Byte=@n2) -> (ReturnValue:Text)) -> (Output_Get:Text)
   n5 = IsInPlayMode!(self:Fort Minigame Object Reference=@n3) -> (ReturnValue:bool)
   n6 = "Not Equal (Byte)"(A=Get NewTeam() -> (NewTeam:Byte=0), B=Get OldTeam() -> (OldTeam:Byte=0)) -> (ReturnValue=~)
   > Branch(Condition=OR Boolean(A=AND Boolean(A=@n6, B=@n5), B=Get bForceUpdate() -> (bForceUpdate:bool)))
   > n7 = GetDynamicMaterial(self=Get Image_Marker())
   > Set MarkerMaterial(MarkerMaterial:Material Instance Dynamic Object Reference=@n7) -> (Output_Get:Material Instance Dynamic Object Reference)
   > n8 = GetDynamicMaterial(self=Get Image_Marker_Pulse() -> (Image_Marker_Pulse:Image Object Reference))
   > Set PulseMaterial(PulseMaterial:Material Instance Dynamic Object Reference=@n8) -> (Output_Get:Material Instance Dynamic Object Reference)
   n9 = GetOwnershipTeamColor!(
    self:PDA UI Creative Core Gamer Object Reference=Get PrimaryDataAsset()
    InMinigame:Fort Minigame Object Reference=GetMinigame()
    InLocalPlayerController:Player Controller Object Reference=GetOwningPlayer()
    Team:Byte=Get CachedOwningTeam() -> (CachedOwningTeam=0)
   ) -> (OutColorSet:S UI Creative Core Gamer Team Colors Structure)
   > UpdateOwningTeamColors!(
    execute
    self
    InSquareRadialMaterial:Material Instance Dynamic Object Reference=Get MarkerMaterial() -> (MarkerMaterial:Material Instance Dynamic Object Reference)
    InTeamColors:S UI Creative Core Gamer Team Colors Structure=@n9
   ) -> (then)
   > UpdateProgressColors!(
    execute
    self
    InSquareRadialMaterial:Material Instance Dynamic Object Reference=Get MarkerMaterial()
    InTeamColors:S UI Creative Core Gamer Team Colors Structure=@n9
   ) -> (then)
   > UpdateProgressColors(InSquareRadialMaterial=Get PulseMaterial() -> (PulseMaterial:Material Instance Dynamic Object Reference), InTeamColors=@n9)
   > n10 = Branch(Condition="Equal (Byte)"(A=Get NewTeam() -> (NewTeam=0)) -> (ReturnValue=~))
   > UpdateProgress(Percentage=0)
   > SetRotationState!(execute, self, NewIsRotated:bool=false) -> (then)
   UpdateProgress(execute=@n10.else, Percentage=1)
   > SetRotationState(NewIsRotated=IsLocalTeam(InTeamByte=Get NewTeam() -> (NewTeam=0)) {is=title K2Node_CallFunction})
  ]}
 Sequencer Events {nodes [
   Pulse Animation Delay() {is=title K2Node_CustomEvent}
   Waiting for Anim_ObjectiveOwning to finish() {is=title K2Node_CustomEvent}
  ]}
 SetRotationState {nodes [
   n1 = SetRotationState!() -> (then, NewIsRotated:bool=~) {is=title K2Node_FunctionEntry}
   > n2 = GetDynamicMaterial(self=Get Image_Marker())
   > Set MarkerMaterial(MarkerMaterial=@n2)
   > Branch(Condition="Not Equal (Boolean)"(A=@n1, B=Get IsRotated() -> (IsRotated:bool)) -> (ReturnValue=~))
   > n3 = Set IsRotated(IsRotated:bool=@n1) -> (Output_Get:bool)
   > n4 = Branch(Condition=@n3)
   > PlayAnimationForward(InAnimation=Get Anim_Pulse_ObjectiveOwning() -> (Anim_Pulse_ObjectiveOwning:Widget Animation Object Reference))
   n5 = Get Anim_ObjectiveOwning() -> (Anim_ObjectiveOwning:Widget Animation Object Reference)
   > PlayAnimationForward(InAnimation=@n5)
   n6 = Get Anim_Pulse_Contested_OnObjectiveOwning() -> (Anim_Pulse_Contested_OnObjectiveOwning:Widget Animation Object Reference)
   > PlayAnimationForward(InAnimation=@n6)
   > PlayAnimationForward(InAnimation=Get Anim_Flare_InAndOut() -> (Anim_Flare_InAndOut:Widget Animation Object Reference))
   > n7 = Get Scalar Parameter Value!(execute, self:Material Instance Dynamic Object Reference=Get MarkerMaterial(), ParameterName:Name=Use Radial progress) -> (then, ReturnValue:float) {is=title K2Node_CallFunction}
   > Branch(Condition="Equal (Float)"(A=@n7))
   > SetScalarParameterValue(self=Get MarkerMaterial(), ParameterName=Start Point, Value=0.875)
   PlayAnimationReverse(execute=@n4.else, InAnimation=@n5)
   > PlayAnimationReverse(InAnimation=@n6)
   > n8 = Get Scalar Parameter Value(self=Get MarkerMaterial(), ParameterName=Use Radial progress)
   > Branch(Condition="Equal (Float)"(A=@n8))
   > SetScalarParameterValue(self=Get MarkerMaterial(), ParameterName=Start Point, Value=1)
  ]}
 SetupInWorldWidget {nodes [
   SetupInWorldWidget() {is=title K2Node_FunctionEntry}
   n1 = Get CaptureComponent()
   n2 = InEditor!() -> (ReturnValue:bool)
   n3 = Get CaptureArea()
   > Branch(Condition=AND Boolean!(A=IsValid(Object=@n3), B=IsValid(Object=@n1), C:bool=NOT Boolean(A=@n2)) -> (ReturnValue))
   > Set CaptureIconData!(execute, CaptureIconData=GetCaptureIconData!(self=Get CaptureArea()) -> (OutCaptureIconData:Capture Icon Data Structure)) -> (then, Output_Get:Capture Icon Data Structure)
   > n4 = Sequence!(execute) -> (then_0, then_1, then_2:Exec, then_3:Exec, then_4:Exec, then_5:Exec)
   > Invalidate Timer Handle!(execute, Handle:"Timer Handle Structure (by ref)"=Get InWorllTimerHandle() -> (InWorllTimerHandle:Timer Handle Structure)) -> (then, ReturnValue:Timer Handle Structure)
   > n5 = Set Timer by Event(Delegate=Create Event(), Time=Get TimerUpdateRate(), bLooping=true)
   > Set InWorllTimerHandle(InWorllTimerHandle:Timer Handle Structure=@n5) -> (Output_Get:Timer Handle Structure)
   LoadIndicatorIcon(
    execute=@n4.then_1
    InSoftTextureReference=Get!() -> (
     CaptureIconData_IconTexture:Texture 2D Soft Object Reference
     CaptureIconData_TextScale:float=~
     CaptureIconData_VisibleDistance:float=~
     CaptureIconData_bDisplaysDistanceText:bool=~
     CaptureIconData_bDisplaysOwningTeam:bool=~
     CaptureIconData_bRequiresLineOfSight:bool=~
     CaptureIconData_bShowOffscreenArrow:bool=~
     CaptureIconData_bEnabled:bool=~
     CaptureIconData_TextOverrides:Array of Strings
    )
   ) {is=title K2Node_CallFunction}
   n6 = Sequence!(execute=@n4.then_2) -> (then_0, then_1, then_2)
   n7 = Create Event()
   n8 = Get CaptureComponent()
   > Unbind Event from On Capture Progress Update(self:Capture Component Object Reference=@n8, Delegate=@n7) {is=title K2Node_RemoveDelegate}
   > Bind Event to On Capture Progress Update(self:Capture Component Object Reference=@n8, Delegate=@n7) {is=title K2Node_AddDelegate}
   > CaptureProgressUpdated!(execute, self, NewProgress:float=Get CaptureProgress(self=Get CaptureComponent())) -> (then) {is=title K2Node_CallFunction}
   n9 = Create Event()
   Unbind Event from On Capture State Changed(execute=@n6.then_1, self:Capture Component Object Reference=@n8, Delegate=@n9) {is=title K2Node_RemoveDelegate}
   > Bind Event to On Capture State Changed(self:Capture Component Object Reference=@n8, Delegate=@n9) {is=title K2Node_AddDelegate}
   > CaptureStateUpdated!(
    execute
    self
    NewCaptureState:ECaptureState Enum=GetCaptureState!(self:Capture Component Object Reference=Get CaptureComponent()) -> (ReturnValue:ECaptureState Enum=Neutral)
    PreviousCaptureState:ECaptureState Enum=Neutral
   ) -> (then) {is=title K2Node_CallFunction}
   n10 = Create Event()
   Unbind Event from On Owning Team Changed(execute=@n6.then_2, self:Capture Component Object Reference=@n8, Delegate=@n10) {is=title K2Node_RemoveDelegate}
   > Bind Event to On Owning Team Changed(self:Capture Component Object Reference=@n8, Delegate=@n10) {is=title K2Node_AddDelegate}
   > CaptureOwnerUpdated!(
    execute
    self
    CaptureOwner:Byte=Get OwningTeam!(self:Capture Component Object Reference=Get CaptureComponent()) -> (OwningTeam:Byte=0)
    PrevCaptureOwner:Byte=0
   ) -> (then) {is=title K2Node_CallFunction}
   n11 = Get CaptureIconWidgetComponent(self=Get CaptureArea())
   Unbind all Events from On Clamp State Changed Event!(execute:Exec=@n4.then_3, self:Fort Widget Component Object Reference=@n11) -> (then:Exec) {is=title K2Node_ClearDelegate}
   > Bind Event to On Clamp State Changed Event(self:Fort Widget Component Object Reference=@n11, Delegate=Create Event()) {is=title K2Node_AddDelegate}
   n12 = Truncate(A=float * float(
     A=Get() -> (
      CaptureIconData_TextScale=~
      CaptureIconData_VisibleDistance=~
      CaptureIconData_bDisplaysDistanceText=~
      CaptureIconData_bDisplaysOwningTeam=~
      CaptureIconData_bRequiresLineOfSight=~
      CaptureIconData_bShowOffscreenArrow=~
      CaptureIconData_bEnabled=~
     ).CaptureIconData_TextScale
     B=24
    ))
   n13 = Find!(
    TargetMap:Map of EBeaconFontStyle Enums to Slate Font Info Structures=Get FontStyleMap() -> (FontStyleMap:Map of EBeaconFontStyle Enums to Slate Font Info Structures=NewEnumerator0)
    Key:"EBeaconFontStyle Enum (by ref)"=NewEnumerator0
   ) -> (
    Value_FontObject:Object Reference
    Value_FontMaterial:Object Reference
    Value_OutlineSettings:Font Outline Settings Structure
    Value_TypefaceFontName:Name
    Value_Size:float=~
    Value_LetterSpacing:int
    Value_SkewAmount:float=~
    Value_bForceMonospaced:bool=~
    Value_bMaterialIsStencil:bool=~
    Value_MonospacedWidth:float=~
    ReturnValue
   )
   SetFont(
    execute=@n4.then_4
    self=[Get DistanceText(), Get OptionalText()]
    InFontInfo=MakeSlateFontInfo!(
     FontObject:Object Reference=@n13
     FontMaterial:Object Reference=@n13.Value_FontMaterial
     OutlineSettings:Font Outline Settings Structure=@n13.Value_OutlineSettings
     TypefaceFontName:Name=@n13.Value_TypefaceFontName
     Size:float="To Float (Integer)"(InInt=@n12)
     LetterSpacing:int=@n13.Value_LetterSpacing
     SkewAmount:float=@n13.Value_SkewAmount
     bForceMonospaced:bool=false
     bMaterialIsStencil:bool=false
     MonospacedWidth:float=1
    ) -> (ReturnValue:Slate Font Info Structure)
   )
  ]}
 UpdateAllMarkerMaterialParams {nodes [
   UpdateAllMarkerMaterialParams!() -> (then, Percentage:double=~, AnglePercent:double=~, inTeamColors:S UI Creative Core Gamer Team Colors Structure) {is=title K2Node_FunctionEntry}
   > n1 = GetDynamicMaterial(self=Get Image_Marker())
   > n2 = Set MarkerMaterial(MarkerMaterial=@n1)
   > SetScalarParameterValue(self=@n2, ParameterName=Percentage, Value=Get Percentage() -> (Percentage:double))
   > SetScalarParameterValue(self=Get MarkerMaterial(), ParameterName=Rotation Angle, Value=Get AnglePercent() -> (AnglePercent:double))
   > UpdateOwningTeamColors(
    InSquareRadialMaterial=Get MarkerMaterial()
    InTeamColors=Get inTeamColors() -> (inTeamColors:S UI Creative Core Gamer Team Colors Structure)
   )
   > UpdateProgressColors(InSquareRadialMaterial=Get MarkerMaterial(), InTeamColors=Get inTeamColors())
   > n3 = GetDynamicMaterial(self=Get Image_Marker_Pulse())
   > UpdateProgressColors(InSquareRadialMaterial=@n3, InTeamColors=Get inTeamColors())
  ]}
 UpdateMarkerColorsByCaptureState {nodes [
   UpdateMarkerColorsByCaptureState!() -> (then, InCaptureState:ECaptureState Enum=Neutral, InTeamByte:Byte) {is=title K2Node_FunctionEntry}
   > n1 = Sequence!(execute) -> (then_0, then_1)
   > n2 = GetDynamicMaterial(self=Get Image_Marker())
   > Set MarkerMaterial(MarkerMaterial=@n2)
   > n3 = GetDynamicMaterial(self=Get Image_Marker_Pulse())
   > Set PulseMaterial(PulseMaterial=@n3)
   n4 = GetProgressTeamColor!(
    self:PDA UI Creative Core Gamer Object Reference=Get PrimaryDataAsset()
    CurrentMinigame:Fort Minigame Object Reference=GetMinigame()
    InLocalPlayerController:Player Controller Object Reference=GetOwningPlayer()
    OwningTeam:Byte=Get CachedOwningTeam() -> (CachedOwningTeam=0)
    CurrentStateTeam:Byte=Get InTeamByte() -> (InTeamByte:Byte=0)
    CurrentState:ECaptureState Enum=Get InCaptureState() -> (InCaptureState:ECaptureState Enum=Neutral)
   ) -> (OutColorSet:S UI Creative Core Gamer Team Colors Structure)
   UpdateProgressColors(execute=@n1.then_1, InSquareRadialMaterial=Get MarkerMaterial(), InTeamColors=@n4)
   > UpdateProgressColors(InSquareRadialMaterial=Get PulseMaterial(), InTeamColors=@n4)
   > n5 = Switch on ECaptureState!(execute, Selection:ECaptureState Enum=Get InCaptureState() -> (InCaptureState=Neutral)) -> (Neutral:Exec, Capturing:Exec, Contested:Exec, Disabled:Exec, Decapturing:Exec, Neutralizing:Exec, Deneutralizing:Exec, Captured:Exec) {is=title K2Node_SwitchEnum}
   PlayAnimationForward(execute=@n5.Contested, InAnimation=Get Anim_Contested() -> (Anim_Contested:Widget Animation Object Reference))
   > n6 = Branch(Condition=Get IsRotated())
   > n7 = PlayAnimationForward(InAnimation=Get Anim_Pulse_Contested_IsObjectiveOwned() -> (Anim_Pulse_Contested_IsObjectiveOwned:Widget Animation Object Reference))
   PlayAnimation(
    execute=[@n7.then, @n6.else]
    InAnimation=Get Anim_Pulse_Contested() -> (Anim_Pulse_Contested:Widget Animation Object Reference)
    NumLoopsToPlay=0
   )
   StopAnimation(execute=[@n5.Captured, @n5.Capturing, @n5.Decapturing, @n5.Deneutralizing, @n5.Neutralizing], InAnimation=Get Anim_Pulse_Contested())
   > PlayAnimationForward(InAnimation=Get Anim_Pulse_Contested_Reset() -> (Anim_Pulse_Contested_Reset:Widget Animation Object Reference))
  ]}
 UpdateMarkerIcon {nodes [
   UpdateMarkerIcon!() -> (then, InUpdateOpacity:bool=~, InIconOpacity:float=~, InUpdateTexture:bool=~, InIconTexture:Texture Object Reference) {is=title K2Node_FunctionEntry}
   > n1 = GetDynamicMaterial(self=Get Image_Marker())
   > Set MarkerMaterial(MarkerMaterial=@n1)
   > n2 = Sequence()
   > Branch(Condition=Get InUpdateOpacity() -> (InUpdateOpacity:bool))
   > SetScalarParameterValue(self=Get MarkerMaterial(), ParameterName=Icon Opacity, Value=Get InIconOpacity() -> (InIconOpacity:float))
   Branch(execute=@n2.then_1, Condition=Get InUpdateTexture() -> (InUpdateTexture:bool))
   > SetTextureParameterValue(self=Get MarkerMaterial(), ParameterName=Icon Texture, Value=Get InIconTexture() -> (InIconTexture:Texture Object Reference))
  ]}
 UpdateOpacity {nodes [
   UpdateOpacity() {is=title K2Node_FunctionEntry}
   > n1 = GetNormalizedDistanceFromScreenCenter!(execute, self:Fort Core Gamer Objective Radial Object Reference) -> (then, ReturnValue:float)
   n2 = Get ScreenCenterOpacityCurve() -> (ScreenCenterOpacityCurve:Curve Float Object Reference)
   > SetRenderOpacity(InOpacity=SelectFloat(
     A=GetFloatValue!(self:Curve Float Object Reference=@n2, InTime:float=@n1) -> (ReturnValue:float)
     B=1
     bPickA=IsUsingGameCamera!(self:Fort Core Gamer Objective Radial Object Reference) -> (ReturnValue:bool)
    ))
  ]}
 UpdateOwningTeamColors {nodes [
   n1 = UpdateOwningTeamColors!() -> (
    then
    InSquareRadialMaterial:Material Instance Dynamic Object Reference
    InTeamColors_Lighter_20_#7:Linear Color Structure
    InTeamColors_Light_21_#8:Linear Color Structure
    InTeamColors_Medium_22_#9:Linear Color Structure
    InTeamColors_Dark_23_#10:Linear Color Structure
    InTeamColors_Darker_24_#11:Linear Color Structure
    InTeamColors_Darker1_27_#12:Linear Color Structure
   ) {is=title K2Node_FunctionEntry}
   > SetVectorParameterValue(self=@n1, ParameterName=Inner Box Color 1, Value=@n1.InTeamColors_Dark_23_#10)
   > SetVectorParameterValue(
    self=Get InSquareRadialMaterial() -> (InSquareRadialMaterial:Material Instance Dynamic Object Reference)
    ParameterName=Inner Box Color 2
    Value=@n1.InTeamColors_Darker_24_#11
   )
  ]}
 UpdateProgressColors {nodes [
   n1 = UpdateProgressColors!() -> (then, InSquareRadialMaterial:Material Instance Dynamic Object Reference, InTeamColors:S UI Creative Core Gamer Team Colors Structure) {is=title K2Node_FunctionEntry}
   n2 = Get!() -> (
    InTeamColors_Lighter_20_#7:Linear Color Structure
    InTeamColors_Light_21_#8:Linear Color Structure
    InTeamColors_Medium_22_#9:Linear Color Structure
    InTeamColors_Dark_23_#10:Linear Color Structure
    InTeamColors_Darker_24_#11:Linear Color Structure
    InTeamColors_Darker1_27_#12:Linear Color Structure
   )
   > SetVectorParameterValue(self=@n1, ParameterName=Progress Color 1, Value=@n2)
   > SetVectorParameterValue(self=Get InSquareRadialMaterial(), ParameterName=Progress Color 2, Value=@n2.InTeamColors_Medium_22_#9)
  ]}
 UpdateTextByAffiliation {nodes [
   UpdateTextByAffiliation() {is=title K2Node_FunctionEntry}
   > n1 = GetTeamAffiliation!(execute, self) -> (then, TeamAffiliation:EFortTeamAffiliation Enum=Friendly) {is=title K2Node_CallFunction}
   > n2 = Get!(execute:Exec) -> (then:Exec, else:Exec, CaptureArea:Building Prop Capture Area Object Reference)
   n3 = GetCaptureIconData!(self=@n2) -> (
    OutCaptureIconData_IconTexture:Texture 2D Soft Object Reference
    OutCaptureIconData_TextScale:float=~
    OutCaptureIconData_VisibleDistance:float=~
    OutCaptureIconData_bDisplaysDistanceText:bool=~
    OutCaptureIconData_bDisplaysOwningTeam:bool=~
    OutCaptureIconData_bRequiresLineOfSight:bool=~
    OutCaptureIconData_bShowOffscreenArrow:bool=~
    OutCaptureIconData_bEnabled:bool=~
    OutCaptureIconData_TextOverrides:Array of Strings
   )
   > n4 = Branch(Condition=@n3.OutCaptureIconData_bDisplaysOwningTeam)
   n5 = Get OptionalText()
   > "SetText (Text)"(self=@n5, InText=TextToUpper(InText=Get CachedOwningTeamName() -> (CachedOwningTeamName:Text)))
   n6 = "Get (a copy)"(
    Array:Array of Strings=@n3.OutCaptureIconData_TextOverrides
    Dimension 1=Select!(Friendly:int=0, Neutral:int=1, Hostile:int=2, Index:EFortTeamAffiliation Enum=@n1) -> (ReturnValue:int)
   ) -> (Output:String)
   "SetText (Text)"(execute=@n4.else, self=@n5, InText=TextToUpper(InText="To Text (String)"(InString=@n6)))
  ]}
]
variables [
 CachedOwningTeam {type=integer}
 CachedOwningTeamName {type=string, default=""}
 CaptureArea {type=BuildingProp_CaptureArea, default=null}
 CaptureComponent {type=CaptureComponent, default=null}
 DistanceUpdateTimerHandle {type=TimerHandle}
 FontStyleMap {default {
   Bold {
    bForceMonospaced=false
    bMaterialIsStencil=false
    fontMaterial=null
    fontObject=&Burbank
    letterSpacing=1
    monospacedWidth=1
    size=20
    skewAmount=0
    typefaceFontName=Burbank-PlayerPower
    outlineSettings {
     bApplyOutlineToDropShadows=false
     bMiteredCorners=false
     bSeparateFillAlpha=false
     outlineMaterial=null
     outlineSize=1
     outlineColor=<rgba 0 0 0 1>
    }
   }
   Subtle {
    bForceMonospaced=false
    bMaterialIsStencil=false
    fontMaterial=null
    fontObject=&NotoSans
    letterSpacing=0
    monospacedWidth=1
    size=24
    skewAmount=0
    typefaceFontName=NotoSans-Bold
    outlineSettings {
     bApplyOutlineToDropShadows=false
     bMiteredCorners=false
     bSeparateFillAlpha=false
     outlineMaterial=null
     outlineSize=0
     outlineColor=<rgba 0 0 0 1>
    }
   }
  }}
 InWorllTimerHandle {type=TimerHandle}
 IsRotated {type=boolean}
 MaxDistance {type=number, default=5000}
 MaxDistanceScale {type=number, default=0.35}
 MinDistance {type=number, default=1000}
 MinDistanceScale {type=number, default=0.5}
 Minigame State {type=EFortMinigameState, default=PreGame}
 ObjectiveIndicatorIcons {
  type="array<Texture2D>"
  default [
   /Game/Athena/UI/InGame/Creative/PropertyEditor/Icon_Picker/Textures/T_UI_IconPicker_NONE
   /Game/Athena/UI/InGame/Creative/PropertyEditor/Icon_Picker/Textures/64x64/T_UI_IconLibrary_A_64
   /Game/Athena/UI/InGame/Creative/PropertyEditor/Icon_Picker/Textures/64x64/T_UI_IconLibrary_B_64
   /Game/Athena/UI/InGame/Creative/PropertyEditor/Icon_Picker/Textures/64x64/T_UI_IconLibrary_C_64
   /Game/Athena/UI/InGame/Creative/PropertyEditor/Icon_Picker/Textures/64x64/T_UI_IconLibrary_D_64
   /Game/Athena/UI/InGame/Creative/PropertyEditor/Icon_Picker/Textures/64x64/T_UI_IconLibrary_W_64
   /Game/Athena/UI/InGame/Creative/PropertyEditor/Icon_Picker/Textures/64x64/T_UI_IconLibrary_X_64
   /Game/Athena/UI/InGame/Creative/PropertyEditor/Icon_Picker/Textures/64x64/T_UI_IconLibrary_Y_64
   /Game/Athena/UI/InGame/Creative/PropertyEditor/Icon_Picker/Textures/64x64/T_UI_IconLibrary_Z_64
   /Game/Athena/UI/InGame/Creative/PropertyEditor/Icon_Picker/Textures_V2/64x64/T_UI_IconLibrary_AMarker_64
   /Game/Athena/UI/InGame/Creative/PropertyEditor/Icon_Picker/Textures_V2/64x64/T_UI_IconLibrary_BMarker_64
   /Game/Athena/UI/InGame/Creative/PropertyEditor/Icon_Picker/Textures_V2/64x64/T_UI_IconLibrary_CMarker_64
   /Game/Athena/UI/InGame/Creative/PropertyEditor/Icon_Picker/Textures_V2/64x64/T_UI_IconLibrary_DMarker_64
   /Game/Athena/UI/InGame/Creative/PropertyEditor/Icon_Picker/Textures_V2/64x64/T_UI_IconLibrary_WMarker_64
   /Game/Athena/UI/InGame/Creative/PropertyEditor/Icon_Picker/Textures_V2/64x64/T_UI_IconLibrary_XMarker_64
   /Game/Athena/UI/InGame/Creative/PropertyEditor/Icon_Picker/Textures_V2/64x64/T_UI_IconLibrary_YMarker_64
   /Game/Athena/UI/InGame/Creative/PropertyEditor/Icon_Picker/Textures_V2/64x64/T_UI_IconLibrary_ZMarker_64
   /Game/Athena/UI/InGame/Creative/PropertyEditor/Icon_Picker/Textures/64x64/T_UI_IconLibrary_0_64
   /Game/Athena/UI/InGame/Creative/PropertyEditor/Icon_Picker/Textures/64x64/T_UI_IconLibrary_1_64
   /Game/Athena/UI/InGame/Creative/PropertyEditor/Icon_Picker/Textures/64x64/T_UI_IconLibrary_2_64
   /Game/Athena/UI/InGame/Creative/PropertyEditor/Icon_Picker/Textures/64x64/T_UI_IconLibrary_3_64
   /Game/Athena/UI/InGame/Creative/PropertyEditor/Icon_Picker/Textures/64x64/T_UI_IconLibrary_4_64
   /Game/Athena/UI/InGame/Creative/PropertyEditor/Icon_Picker/Textures/64x64/T_UI_IconLibrary_5_64
   /Game/Athena/UI/InGame/Creative/PropertyEditor/Icon_Picker/Textures/64x64/T_UI_IconLibrary_6_64
   /Game/Athena/UI/InGame/Creative/PropertyEditor/Icon_Picker/Textures/64x64/T_UI_IconLibrary_7_64
   /Game/Athena/UI/InGame/Creative/PropertyEditor/Icon_Picker/Textures/64x64/T_UI_IconLibrary_8_64
   /Game/Athena/UI/InGame/Creative/PropertyEditor/Icon_Picker/Textures/64x64/T_UI_IconLibrary_9_64
   /Game/Athena/UI/InGame/Creative/PropertyEditor/Icon_Picker/Textures/64x64/T_UI_IconLibrary_droplet_64
   /Game/Athena/UI/InGame/Creative/PropertyEditor/Icon_Picker/Textures/64x64/T_UI_IconLibrary_fishing_64
   /Game/Athena/UI/InGame/Creative/PropertyEditor/Icon_Picker/Textures/64x64/T_UI_IconLibrary_trophy_64
   /Game/Athena/UI/InGame/Creative/PropertyEditor/Icon_Picker/MilitaryTextures/T_64x64/T_UI_IconLibrary_pistol_64
   /Game/Athena/UI/InGame/Creative/PropertyEditor/Icon_Picker/MilitaryTextures/T_64x64/T_UI_IconLibrary_assaultrifle_64
   /Game/Athena/UI/InGame/Creative/PropertyEditor/Icon_Picker/MilitaryTextures/T_64x64/T_UI_IconLibrary_sniperrifle_64
   /Game/Athena/UI/InGame/Creative/PropertyEditor/Icon_Picker/MilitaryTextures/T_64x64/T_UI_IconLibrary_rocketlauncher_64
   /Game/Athena/UI/InGame/Creative/PropertyEditor/Icon_Picker/MilitaryTextures/T_64x64/T_UI_IconLibrary_grenade_64
   /Game/Athena/UI/InGame/Creative/PropertyEditor/Icon_Picker/Textures/64x64/T_UI_IconLibrary_wolf_64
   /Game/Athena/UI/InGame/Creative/PropertyEditor/Icon_Picker/Textures/64x64/T_UI_IconLibrary_raptor_64
   /Game/Athena/UI/InGame/Creative/PropertyEditor/Icon_Picker/Textures/64x64/T_UI_IconLibrary_chicken_64
   /Game/Athena/UI/InGame/Creative/PropertyEditor/Icon_Picker/Textures/64x64/T_UI_IconLibrary_frog_64
   /Game/Athena/UI/InGame/Creative/PropertyEditor/Icon_Picker/Textures/64x64/T_UI_IconLibrary_boar_64
   /Game/Athena/UI/InGame/Creative/PropertyEditor/Icon_Picker/Textures/64x64/T_UI_IconLibrary_fish_64
   /Game/Athena/UI/InGame/Creative/PropertyEditor/Icon_Picker/Textures/64x64/T_UI_IconLibrary_exclamation_64
   /Game/Athena/UI/InGame/Creative/PropertyEditor/Icon_Picker/Textures/64x64/T_UI_IconLibrary_checkmark_64
   /Game/Athena/UI/InGame/Creative/PropertyEditor/Icon_Picker/Textures/64x64/T_UI_IconLibrary_questionmark_64
   /Game/Athena/UI/InGame/Creative/PropertyEditor/Icon_Picker/Textures/64x64/T_UI_IconLibrary_coins_64
   /Game/Athena/UI/InGame/Creative/PropertyEditor/Icon_Picker/Textures/64x64/T_UI_IconLibrary_moneybag_64
   /Game/Athena/UI/InGame/Creative/PropertyEditor/Icon_Picker/Textures/64x64/T_UI_IconLibrary_heart_64
   /Game/Athena/UI/InGame/Creative/PropertyEditor/Icon_Picker/Textures/64x64/T_UI_IconLibrary_hand_64
   /Game/Athena/UI/InGame/Creative/PropertyEditor/Icon_Picker/Textures/64x64/T_UI_IconLibrary_fist_64
   /Game/Athena/UI/InGame/Creative/PropertyEditor/Icon_Picker/Textures/64x64/T_UI_IconLibrary_flame_64
   /Game/Athena/UI/InGame/Creative/PropertyEditor/Icon_Picker/Textures/64x64/T_UI_IconLibrary_radiation_64
   /Game/Athena/UI/InGame/Creative/PropertyEditor/Icon_Picker/Textures/64x64/T_UI_IconLibrary_sword_64
   /Game/Athena/UI/InGame/Creative/PropertyEditor/Icon_Picker/Textures/64x64/T_UI_IconLibrary_shield_64
   /Game/Athena/UI/InGame/Creative/PropertyEditor/Icon_Picker/Textures/64x64/T_UI_IconLibrary_shoppingcart_64
   /Game/Athena/UI/InGame/Creative/PropertyEditor/Icon_Picker/Textures/64x64/T_UI_IconLibrary_car_64
   /Game/Athena/UI/InGame/Creative/PropertyEditor/Icon_Picker/Textures/64x64/T_UI_IconLibrary_truck_64
   /Game/Athena/UI/InGame/Creative/PropertyEditor/Icon_Picker/Textures/64x64/T_UI_IconLibrary_helicopter_64
   /Game/Athena/UI/InGame/Creative/PropertyEditor/Icon_Picker/Textures/64x64/T_UI_IconLibrary_biplane_64
   /Game/Athena/UI/InGame/Creative/PropertyEditor/Icon_Picker/Textures/64x64/T_UI_IconLibrary_boat_64
   /Game/Athena/UI/InGame/Creative/PropertyEditor/Icon_Picker/Textures/64x64/T_UI_IconLibrary_package_64
   /Game/Athena/UI/InGame/Creative/PropertyEditor/Icon_Picker/MilitaryTextures/T_64x64/T_UI_IconLibrary_magazine_64
   /Game/Athena/UI/InGame/Creative/PropertyEditor/Icon_Picker/Textures/64x64/T_UI_IconLibrary_medkit_64
   /Game/Athena/UI/InGame/Creative/PropertyEditor/Icon_Picker/Textures/64x64/T_UI_IconLibrary_bed_64
   /Game/Athena/UI/InGame/Creative/PropertyEditor/Icon_Picker/Textures/64x64/T_UI_IconLibrary_garage_64
   /Game/Athena/UI/InGame/Creative/PropertyEditor/Icon_Picker/Textures/64x64/T_UI_IconLibrary_biohazard_64
   /Game/Athena/UI/InGame/Creative/PropertyEditor/Icon_Picker/Textures/64x64/T_UI_IconLibrary_recycle_64
   /Game/Athena/UI/InGame/Creative/PropertyEditor/Icon_Picker/Textures/64x64/T_UI_IconLibrary_signal_64
   /Game/Athena/UI/InGame/Creative/PropertyEditor/Icon_Picker/Textures/64x64/T_UI_IconLibrary_signpost_64
   /Game/Athena/UI/InGame/Creative/PropertyEditor/Icon_Picker/Textures/64x64/T_UI_IconLibrary_happy_64
   /Game/Athena/UI/InGame/Creative/PropertyEditor/Icon_Picker/Textures/64x64/T_UI_IconLibrary_sad_64
   /Game/Athena/UI/InGame/Creative/PropertyEditor/Icon_Picker/Textures/64x64/T_UI_IconLibrary_cool_64
   /Game/Athena/UI/InGame/Creative/PropertyEditor/Icon_Picker/Textures/64x64/T_UI_IconLibrary_surprised_64
   /Game/Athena/UI/InGame/Creative/PropertyEditor/Icon_Picker/Textures/64x64/T_UI_IconLibrary_tongueout_64
   /Game/Athena/UI/InGame/Creative/PropertyEditor/Icon_Picker/Textures/64x64/T_UI_IconLibrary_wink_64
   /Game/Athena/UI/InGame/Creative/PropertyEditor/Icon_Picker/Textures/64x64/T_UI_IconLibrary_crying_64
   /Game/Athena/UI/InGame/Creative/PropertyEditor/Icon_Picker/Textures/64x64/T_UI_IconLibrary_running_64
   /Game/Athena/UI/InGame/Creative/PropertyEditor/Icon_Picker/Textures/64x64/T_UI_IconLibrary_sprint_64
   /Game/Athena/UI/InGame/Creative/PropertyEditor/Icon_Picker/Textures/64x64/T_UI_IconLibrary_tree_64
   /Game/Athena/UI/InGame/Creative/PropertyEditor/Icon_Picker/Textures/64x64/T_UI_IconLibrary_forest_64
   /Game/Athena/UI/InGame/Creative/PropertyEditor/Icon_Picker/Textures/64x64/T_UI_IconLibrary_anchor_64
   /Game/Athena/UI/InGame/Creative/PropertyEditor/Icon_Picker/Textures/64x64/T_UI_IconLibrary_wheel_64
   /Game/Athena/UI/InGame/Creative/PropertyEditor/Icon_Picker/Textures/64x64/T_UI_IconLibrary_steeringwheel_64
   /Game/Athena/UI/InGame/Creative/PropertyEditor/Icon_Picker/Textures/64x64/T_UI_IconLibrary_speedometer_64
   /Game/Athena/UI/InGame/Creative/PropertyEditor/Icon_Picker/Textures/64x64/T_UI_IconLibrary_campfire_64
   /Game/Athena/UI/InGame/Creative/PropertyEditor/Icon_Picker/Textures/64x64/T_UI_IconLibrary_cave_64
   /Game/Athena/UI/InGame/Creative/PropertyEditor/Icon_Picker/Textures/64x64/T_UI_IconLibrary_house_64
   /Game/Athena/UI/InGame/Creative/PropertyEditor/Icon_Picker/Textures/64x64/T_UI_IconLibrary_information_64
   /Game/Athena/UI/InGame/Creative/PropertyEditor/Icon_Picker/Textures/64x64/T_UI_IconLibrary_medical_64
   /Game/Athena/UI/InGame/Creative/PropertyEditor/Icon_Picker/Textures/64x64/T_UI_IconLibrary_star_64
   /Game/Athena/UI/InGame/Creative/PropertyEditor/Icon_Picker/Textures/64x64/T_UI_IconLibrary_door_64
   /Game/Athena/UI/InGame/Creative/PropertyEditor/Icon_Picker/Textures/64x64/T_UI_IconLibrary_flag_64
   /Game/Athena/UI/InGame/Creative/PropertyEditor/Icon_Picker/Textures/64x64/T_UI_IconLibrary_chequeredflag_64
   /Game/Athena/UI/InGame/Creative/PropertyEditor/Icon_Picker/Textures/64x64/T_UI_IconLibrary_waypoint_64
   /Game/Athena/UI/InGame/Creative/PropertyEditor/Icon_Picker/Textures/64x64/T_UI_IconLibrary_marker_64
   /Game/Athena/UI/InGame/Creative/PropertyEditor/Icon_Picker/Textures/64x64/T_UI_IconLibrary_padlock_64
   /Game/Athena/UI/InGame/Creative/PropertyEditor/Icon_Picker/Textures/64x64/T_UI_IconLibrary_unlocked_64
   /Game/Athena/UI/InGame/Creative/PropertyEditor/Icon_Picker/Textures/64x64/T_UI_IconLibrary_key_64
   /Game/Athena/UI/InGame/Creative/PropertyEditor/Icon_Picker/Textures/64x64/T_UI_IconLibrary_gaspump_64
   /Game/Athena/UI/InGame/Creative/PropertyEditor/Icon_Picker/Textures/64x64/T_UI_IconLibrary_danger_64
   /Game/Athena/UI/InGame/Creative/PropertyEditor/Icon_Picker/Textures/64x64/T_UI_IconLibrary_enemy_64
   /Game/Athena/UI/InGame/Creative/PropertyEditor/Icon_Picker/Textures_V2/64x64/T_UI_IconLibrary_Pawprint_64
   /Game/Athena/UI/InGame/Creative/PropertyEditor/Icon_Picker/Textures_V2/64x64/T_UI_IconLibrary_PetFoodBowl_64
   /Game/Athena/UI/InGame/Creative/PropertyEditor/Icon_Picker/Textures_V2/64x64/T_UI_IconLibrary_LeftArrow_64
   /Game/Athena/UI/InGame/Creative/PropertyEditor/Icon_Picker/Textures_V2/64x64/T_UI_IconLibrary_RightArrow_64
   /Game/Athena/UI/InGame/Creative/PropertyEditor/Icon_Picker/Textures_V2/64x64/T_UI_IconLibrary_UpArrow_64
   /Game/Athena/UI/InGame/Creative/PropertyEditor/Icon_Picker/Textures_V2/64x64/T_UI_IconLibrary_DownArrow_64
   /Game/Athena/UI/InGame/Creative/PropertyEditor/Icon_Picker/Textures_V2/64x64/T_UI_IconLibrary_SpeechBubble_64
   /Game/Athena/UI/InGame/Creative/PropertyEditor/Icon_Picker/Textures_V2/64x64/T_UI_IconLibrary_List_64
   /Game/Athena/UI/InGame/Creative/PropertyEditor/Icon_Picker/Textures_V2/64x64/T_UI_IconLibrary_OfficeBuilding_64
   /Game/Athena/UI/InGame/Creative/PropertyEditor/Icon_Picker/Textures_V2/64x64/T_UI_IconLibrary_ApartmentBuilding_64
   /Game/Athena/UI/InGame/Creative/PropertyEditor/Icon_Picker/Textures_V2/64x64/T_UI_IconLibrary_Education_64
   /Game/Athena/UI/InGame/Creative/PropertyEditor/Icon_Picker/Textures_V2/64x64/T_UI_IconLibrary_Museum_64
   /Game/Athena/UI/InGame/Creative/PropertyEditor/Icon_Picker/Textures_V2/64x64/T_UI_IconLibrary_Briefcase_64
   /Game/Athena/UI/InGame/Creative/PropertyEditor/Icon_Picker/Textures_V2/64x64/T_UI_IconLibrary_Store_64
   /Game/Athena/UI/InGame/Creative/PropertyEditor/Icon_Picker/Textures_V2/64x64/T_UI_IconLibrary_Compass_64
   /Game/Athena/UI/InGame/Creative/PropertyEditor/Icon_Picker/Textures_V2/64x64/T_UI_IconLibrary_Alien_64
   /Game/Athena/UI/InGame/Creative/PropertyEditor/Icon_Picker/Textures_V2/64x64/T_UI_IconLibrary_MusicNote_64
   /Game/Athena/UI/InGame/Creative/PropertyEditor/Icon_Picker/Textures_V2/64x64/T_UI_IconLibrary_Binoculars_64
   /Game/Athena/UI/InGame/Creative/PropertyEditor/Icon_Picker/Textures_V2/64x64/T_UI_IconLibrary_UFO_64
   /Game/Athena/UI/InGame/Creative/PropertyEditor/Icon_Picker/Textures_V2/64x64/T_UI_IconLibrary_Football_64
   /Game/Athena/UI/InGame/Creative/PropertyEditor/Icon_Picker/Textures_V2/64x64/T_UI_IconLibrary_Arena_64
   /Game/Athena/UI/InGame/Creative/PropertyEditor/Icon_Picker/Textures_V2/64x64/T_UI_IconLibrary_Bridge_64
   /Game/Athena/UI/InGame/Creative/PropertyEditor/Icon_Picker/Textures_V2/64x64/T_UI_IconLibrary_Gravestone_64
   /Game/Athena/UI/InGame/Creative/PropertyEditor/Icon_Picker/Textures_V2/64x64/T_UI_IconLibrary_Hospital_64
   /Game/Athena/UI/InGame/Creative/PropertyEditor/Icon_Picker/Textures_V2/64x64/T_UI_IconLibrary_PoliceBadge_64
   /Game/Athena/UI/InGame/Creative/PropertyEditor/Icon_Picker/Textures_V2/64x64/T_UI_IconLibrary_Cutlery_64
   /Game/Athena/UI/InGame/Creative/PropertyEditor/Icon_Picker/Textures_V2/64x64/T_UI_IconLibrary_SheriffBadge_64
   /Game/Athena/UI/InGame/Creative/PropertyEditor/Icon_Picker/Textures_V2/64x64/T_UI_IconLibrary_Laptop_64
   /Game/Athena/UI/InGame/Creative/PropertyEditor/Icon_Picker/Textures_V2/64x64/T_UI_IconLibrary_Power_64
   /Game/Athena/UI/InGame/Creative/PropertyEditor/Icon_Picker/Textures_V2/64x64/T_UI_IconLibrary_Hammer_64
   /Game/Athena/UI/InGame/Creative/PropertyEditor/Icon_Picker/Textures_V2/64x64/T_UI_IconLibrary_Paintbrush_64
   /Game/Athena/UI/InGame/Creative/PropertyEditor/Icon_Picker/Textures_V2/64x64/T_UI_IconLibrary_Shovel_64
   /Game/Athena/UI/InGame/Creative/PropertyEditor/Icon_Picker/Textures_V2/64x64/T_UI_IconLibrary_Restaurant_64
   /Game/Athena/UI/InGame/Creative/PropertyEditor/Icon_Picker/Textures_V2/64x64/T_UI_IconLibrary_Gift_64
   /Game/Athena/UI/InGame/Creative/PropertyEditor/Icon_Picker/Textures_V2/64x64/T_UI_IconLibrary_Handshake_64
   /Game/Athena/UI/InGame/Creative/PropertyEditor/Icon_Picker/Textures_V2/64x64/T_UI_IconLibrary_Pin_64
   /Game/Athena/UI/InGame/Creative/PropertyEditor/Icon_Picker/Textures_V2/64x64/T_UI_IconLibrary_Camera_64
   /Game/Athena/UI/InGame/Creative/PropertyEditor/Icon_Picker/Textures_V2/64x64/T_UI_IconLibrary_PiggyBank_64
   /Game/Athena/UI/InGame/Creative/PropertyEditor/Icon_Picker/Textures_V2/64x64/T_UI_IconLibrary_BankNotes_64
   /Game/Athena/UI/InGame/Creative/PropertyEditor/Icon_Picker/Textures_V2/64x64/T_UI_IconLibrary_Burger_64
   /Game/Athena/UI/InGame/Creative/PropertyEditor/Icon_Picker/Textures_V2/64x64/T_UI_IconLibrary_Mushroom_64
   /Game/Athena/UI/InGame/Creative/PropertyEditor/Icon_Picker/Textures_V2/64x64/T_UI_IconLibrary_PizzaSlice_64
   /Game/Athena/UI/InGame/Creative/PropertyEditor/Icon_Picker/Textures_V2/64x64/T_UI_IconLibrary_GasMask_64
   /Game/Athena/UI/InGame/Creative/PropertyEditor/Icon_Picker/Textures_V2/64x64/T_UI_IconLibrary_Heartbeat_64.T_UI_IconLibrary_HeartBeat_64
   /Game/Athena/UI/InGame/Creative/PropertyEditor/Icon_Picker/Textures_V2/64x64/T_UI_IconLibrary_IceSkate_64
   /Game/Athena/UI/InGame/Creative/PropertyEditor/Icon_Picker/Textures_V2/64x64/T_UI_IconLibrary_SnowFlake_64.T_UI_IconLibrary_Snowflake_64
   /Game/Athena/UI/InGame/Creative/PropertyEditor/Icon_Picker/Textures_V2/64x64/T_UI_IconLibrary_Ghost_64
   /Game/Athena/UI/InGame/Creative/PropertyEditor/Icon_Picker/Textures_V2/64x64/T_UI_IconLibrary_Electrical_64
   /Game/Athena/UI/InGame/Creative/PropertyEditor/Icon_Picker/Textures_V2/64x64/T_UI_IconLibrary_CookingPot_64
   /Game/Athena/UI/InGame/Creative/PropertyEditor/Icon_Picker/Textures_V2/64x64/T_UI_IconLibrary_Cabin_64
   /Game/Athena/UI/InGame/Creative/PropertyEditor/Icon_Picker/Textures_V2/64x64/T_UI_IconLibrary_OfficePhone_64
   /Game/Athena/UI/InGame/Creative/PropertyEditor/Icon_Picker/Textures_V2/64x64/T_UI_IconLibrary_TV_64
   /Game/Athena/UI/InGame/Creative/PropertyEditor/Icon_Picker/Textures_V2/64x64/T_UI_IconLibrary_Gear_64
   /Game/Athena/UI/InGame/Creative/PropertyEditor/Icon_Picker/Textures_V2/64x64/T_UI_IconLibrary_Helix_64
   /Game/Athena/UI/InGame/Creative/PropertyEditor/Icon_Picker/Textures_V2/64x64/T_UI_IconLibrary_Zap_64
   /Game/Athena/UI/InGame/Creative/PropertyEditor/Icon_Picker/Textures_V2/64x64/T_UI_IconLibrary_Factory_64
   /Game/Athena/UI/InGame/Creative/PropertyEditor/Icon_Picker/Textures_V2/64x64/T_UI_IconLibrary_RadioTower_64
   /Game/Athena/UI/InGame/Creative/PropertyEditor/Icon_Picker/Textures_V2/64x64/T_UI_IconLibrary_Microscope_64
   /Game/Athena/UI/InGame/Creative/PropertyEditor/Icon_Picker/Textures_V2/64x64/T_UI_IconLibrary_SatelliteDish_64
   /Game/Athena/UI/InGame/Creative/PropertyEditor/Icon_Picker/Textures_V2/64x64/T_UI_IconLibrary_Map_64
   /Game/Athena/UI/InGame/Creative/PropertyEditor/Icon_Picker/Textures_V2/64x64/T_UI_IconLibrary_Play_64
   /Game/Athena/UI/InGame/Creative/PropertyEditor/Icon_Picker/Textures_V2/64x64/T_UI_IconLibrary_Stop_64
   /Game/Athena/UI/InGame/Creative/PropertyEditor/Icon_Picker/Textures_V2/64x64/T_UI_IconLibrary_Pause_64
   /Game/Athena/UI/InGame/Creative/PropertyEditor/Icon_Picker/Textures_V2/64x64/T_UI_IconLibrary_FastForward_64
   /Game/Athena/UI/InGame/Creative/PropertyEditor/Icon_Picker/Textures_V2/64x64/T_UI_IconLibrary_Rewind_64
   /Game/Athena/UI/InGame/Creative/PropertyEditor/Icon_Picker/Textures_V2/64x64/T_UI_IconLibrary_Volume_64
   /Game/Athena/UI/InGame/Creative/PropertyEditor/Icon_Picker/Textures_V2/64x64/T_UI_IconLibrary_Repeat_64
   /Game/Athena/UI/InGame/Creative/PropertyEditor/Icon_Picker/Textures_V2/64x64/T_UI_IconLibrary_Shuffle_64
   /Game/Athena/UI/InGame/Creative/PropertyEditor/Icon_Picker/Textures_V2/64x64/T_UI_IconLibrary_Chat_64
   /Game/Athena/UI/InGame/Creative/PropertyEditor/Icon_Picker/Textures_V2/64x64/T_UI_IconLibrary_Crown_64
   /Game/Athena/UI/InGame/Creative/PropertyEditor/Icon_Picker/Textures_V2/64x64/T_UI_IconLibrary_Eye_64
   /Game/Athena/UI/InGame/Creative/PropertyEditor/Icon_Picker/Textures_V2/64x64/T_UI_IconLibrary_Message_64
   /Game/Athena/UI/InGame/Creative/PropertyEditor/Icon_Picker/Textures_V2/64x64/T_UI_IconLibrary_LightBulb_64
   /Game/Athena/UI/InGame/Creative/PropertyEditor/Icon_Picker/Textures_V2/64x64/T_UI_IconLibrary_Bomb_64
   /Game/Athena/UI/InGame/Creative/PropertyEditor/Icon_Picker/Textures_V2/64x64/T_UI_IconLibrary_Cannon_64
   /Game/Athena/UI/InGame/Creative/PropertyEditor/Icon_Picker/MilitaryTextures/T_64x64/T_UI_IconLibrary_JetFighter_64
   /Game/Athena/UI/InGame/Creative/PropertyEditor/Icon_Picker/Textures_V2/64x64/T_UI_IconLibrary_Tent_64
   /Game/Athena/UI/InGame/Creative/PropertyEditor/Icon_Picker/Textures_V2/64x64/T_UI_IconLibrary_CCTV_64
   /Game/Athena/UI/InGame/Creative/PropertyEditor/Icon_Picker/Textures_V2/64x64/T_UI_IconLibrary_Radio_64
   /Game/Athena/UI/InGame/Creative/PropertyEditor/Icon_Picker/Textures_V2/64x64/T_UI_IconLibrary_Wifi_64
   /Game/Athena/UI/InGame/Creative/PropertyEditor/Icon_Picker/Textures_V2/64x64/T_UI_IconLibrary_Grass_64
   /Game/Athena/UI/InGame/Creative/PropertyEditor/Icon_Picker/Textures_V2/64x64/T_UI_IconLibrary_Leaf_64
   /Game/Athena/UI/InGame/Creative/PropertyEditor/Icon_Picker/Textures_V2/64x64/T_UI_IconLibrary_PalmTree_64
   /Game/Athena/UI/InGame/Creative/PropertyEditor/Icon_Picker/Textures_V2/64x64/T_UI_IconLibrary_Book_64
   /Game/Athena/UI/InGame/Creative/PropertyEditor/Icon_Picker/Textures_V2/64x64/T_UI_IconLibrary_Toolbox_64
   /Game/Athena/UI/InGame/Creative/PropertyEditor/Icon_Picker/Textures_V2/64x64/T_UI_IconLibrary_Bug_64
   /Game/Athena/UI/InGame/Creative/PropertyEditor/Icon_Picker/Textures_V2/64x64/T_UI_IconLibrary_FloppyDisc_64
   /Game/Athena/UI/InGame/Creative/PropertyEditor/Icon_Picker/Textures_V2/64x64/T_UI_IconLibrary_Cogs_64
   /Game/Athena/UI/InGame/Creative/PropertyEditor/Icon_Picker/Textures_V2/64x64/T_UI_IconLibrary_Clock_64
   /Game/Athena/UI/InGame/Creative/PropertyEditor/Icon_Picker/Textures_V2/64x64/T_UI_IconLibrary_TestTube_64
   /Game/Athena/UI/InGame/Creative/PropertyEditor/Icon_Picker/Textures_V2/64x64/T_UI_IconLibrary_Stopwatch_64
   /Game/Athena/UI/InGame/Creative/PropertyEditor/Icon_Picker/Textures_V2/64x64/T_UI_IconLibrary_Basketball_64
   /Game/Athena/UI/InGame/Creative/PropertyEditor/Icon_Picker/Textures_V2/64x64/T_UI_IconLibrary_BoxingGlove_64
   /Game/Athena/UI/InGame/Creative/PropertyEditor/Icon_Picker/Textures_V2/64x64/T_UI_IconLibrary_SoccerBall_64
   /Game/Athena/UI/InGame/Creative/PropertyEditor/Icon_Picker/Textures_V2/64x64/T_UI_IconLibrary_Roadblock_64
   /Game/Athena/UI/InGame/Creative/PropertyEditor/Icon_Picker/Textures_V2/64x64/T_UI_IconLibrary_Rocket_64
   /Game/Athena/UI/InGame/Creative/PropertyEditor/Icon_Picker/Textures_V2/64x64/T_UI_IconLibrary_Taxi_64
   /Game/Athena/UI/InGame/Creative/PropertyEditor/Icon_Picker/Textures_V2/64x64/T_UI_IconLibrary_Footprint_64
   /Game/Athena/UI/InGame/Creative/PropertyEditor/Icon_Picker/Textures_V2/64x64/T_UI_IconLibrary_Shoeprint_64
   /Game/Athena/UI/InGame/Creative/PropertyEditor/Icon_Picker/Textures_V2/64x64/T_UI_IconLibrary_Torch_64
   /Game/Athena/UI/InGame/Creative/PropertyEditor/Icon_Picker/Textures_V2/64x64/T_UI_IconLibrary_Axe_64
   /Game/Athena/UI/InGame/Creative/PropertyEditor/Icon_Picker/Textures_V2/64x64/T_UI_IconLibrary_Link_64
   /Game/Athena/UI/InGame/Creative/PropertyEditor/Icon_Picker/Textures_V2/64x64/T_UI_IconLibrary_Jigsaw_64
   /Game/Athena/UI/InGame/Creative/PropertyEditor/Icon_Picker/Textures_V2/64x64/T_UI_IconLibrary_Magnifier_64
   /Game/Athena/UI/InGame/Creative/PropertyEditor/Icon_Picker/Textures_V2/64x64/T_UI_IconLibrary_Wrench_64
   /Game/Athena/UI/InGame/Creative/PropertyEditor/Icon_Picker/Textures_V2/64x64/T_UI_IconLibrary_Input_64
   /Game/Athena/UI/InGame/Creative/PropertyEditor/Icon_Picker/Textures_V2/64x64/T_UI_IconLibrary_Calendar_64
   /Game/Athena/UI/InGame/Creative/PropertyEditor/Icon_Picker/Textures_V2/64x64/T_UI_IconLibrary_Projector_64
   /Game/Athena/UI/InGame/Creative/PropertyEditor/Icon_Picker/Textures_V2/64x64/T_UI_IconLibrary_Drama_64
   /Game/Athena/UI/InGame/Creative/PropertyEditor/Icon_Picker/Textures_V2/64x64/T_UI_IconLibrary_Hourglass_64
   /Game/Athena/UI/InGame/Creative/PropertyEditor/Icon_Picker/Textures_V2/64x64/T_UI_IconLibrary_Ticket_64
   /Game/Athena/UI/InGame/Creative/PropertyEditor/Icon_Picker/Textures_V2/64x64/T_UI_IconLibrary_Cactus_64
   /Game/Athena/UI/InGame/Creative/PropertyEditor/Icon_Picker/Textures_V2/64x64/T_UI_IconLibrary_JetPlane_64
   /Game/Athena/UI/InGame/Creative/PropertyEditor/Icon_Picker/Textures_V2/64x64/T_UI_IconLibrary_Railway_64
   /Game/Athena/UI/InGame/Creative/PropertyEditor/Icon_Picker/Textures_V2/64x64/T_UI_IconLibrary_Handcuffs_64
   /Game/Athena/UI/InGame/Creative/PropertyEditor/Icon_Picker/Textures_V2/64x64/T_UI_IconLibrary_Monument_64
   /Game/Athena/UI/InGame/Creative/PropertyEditor/Icon_Picker/Textures_V2/64x64/T_UI_IconLibrary_Boot_64
   /Game/Athena/UI/InGame/Creative/PropertyEditor/Icon_Picker/Textures_V2/64x64/T_UI_IconLibrary_Shirt_64
   /Game/Athena/UI/InGame/Creative/PropertyEditor/Icon_Picker/Textures_V2/64x64/T_UI_IconLibrary_Controller_64
   /Game/Athena/UI/InGame/Creative/PropertyEditor/Icon_Picker/Textures_V2/64x64/T_UI_IconLibrary_Keyboard_64
   /Game/Athena/UI/InGame/Creative/PropertyEditor/Icon_Picker/Textures_V2/64x64/T_UI_IconLibrary_Tape_64
   /Game/Athena/UI/InGame/Creative/PropertyEditor/Icon_Picker/Textures_V2/64x64/T_UI_IconLibrary_Plant_64
   /Game/Athena/UI/InGame/Creative/PropertyEditor/Icon_Picker/Textures_V2/64x64/T_UI_IconLibrary_WateringCan_64
   /Game/Athena/UI/InGame/Creative/PropertyEditor/Icon_Picker/Textures_V2/64x64/T_UI_IconLibrary_PaperClip_64
   /Game/Athena/UI/InGame/Creative/PropertyEditor/Icon_Picker/Textures_V2/64x64/T_UI_IconLibrary_PaintPot_64
   /Game/Athena/UI/InGame/Creative/PropertyEditor/Icon_Picker/Textures_V2/64x64/T_UI_IconLibrary_Clipboard_64
   /Game/Athena/UI/InGame/Creative/PropertyEditor/Icon_Picker/Textures_V2/64x64/T_UI_IconLibrary_TrashCan_64
   /Game/Athena/UI/InGame/Creative/PropertyEditor/Icon_Picker/Textures_V2/64x64/T_UI_IconLibrary_Scissors_64
   /Game/Athena/UI/InGame/Creative/PropertyEditor/Icon_Picker/Textures_V2/64x64/T_UI_IconLibrary_Pencil_64
   /Game/Athena/UI/InGame/Creative/PropertyEditor/Icon_Picker/Textures_V2/64x64/T_UI_IconLibrary_CoffeeCup_64
   /Game/Athena/UI/InGame/Creative/PropertyEditor/Icon_Picker/Textures_V2/64x64/T_UI_IconLibrary_CoffeeMug_64
   /Game/Athena/UI/InGame/Creative/PropertyEditor/Icon_Picker/Textures_V2/64x64/T_UI_IconLibrary_Cooker_64
   /Game/Athena/UI/InGame/Creative/PropertyEditor/Icon_Picker/Textures_V2/64x64/T_UI_IconLibrary_FrenchFries_64
   /Game/Athena/UI/InGame/Creative/PropertyEditor/Icon_Picker/Textures_V2/64x64/T_UI_IconLibrary_Fridge_64
   /Game/Athena/UI/InGame/Creative/PropertyEditor/Icon_Picker/Textures_V2/64x64/T_UI_IconLibrary_IceCream_64
   /Game/Athena/UI/InGame/Creative/PropertyEditor/Icon_Picker/Textures_V2/64x64/T_UI_IconLibrary_Microwave_64
   /Game/Athena/UI/InGame/Creative/PropertyEditor/Icon_Picker/Textures_V2/64x64/T_UI_IconLibrary_Cards_64
   /Game/Athena/UI/InGame/Creative/PropertyEditor/Icon_Picker/Textures_V2/64x64/T_UI_IconLibrary_Shamrock_64
   /Game/Athena/UI/InGame/Creative/PropertyEditor/Icon_Picker/Textures_V2/64x64/T_UI_IconLibrary_Dice_64
   /Game/Athena/UI/InGame/Creative/PropertyEditor/Icon_Picker/Textures_V2/64x64/T_UI_IconLibrary_ChessPiece_64
   /Game/Athena/UI/InGame/Creative/PropertyEditor/Icon_Picker/Textures_V2/64x64/T_UI_IconLibrary_Bell_64
   /Game/Athena/UI/InGame/Creative/PropertyEditor/Icon_Picker/Textures_V2/64x64/T_UI_IconLibrary_Candle_64
   /Game/Athena/UI/InGame/Creative/PropertyEditor/Icon_Picker/Textures_V2/64x64/T_UI_IconLibrary_Pumpkin_64
   /Game/Athena/UI/InGame/Creative/PropertyEditor/Icon_Picker/Textures_V2/64x64/T_UI_IconLibrary_Bat_64
   /Game/Athena/UI/InGame/Creative/PropertyEditor/Icon_Picker/Textures_V2/64x64/T_UI_IconLibrary_Sofa_64
   /Game/Athena/UI/InGame/Creative/PropertyEditor/Icon_Picker/Textures_V2/64x64/T_UI_IconLibrary_Tap_64
   /Game/Athena/UI/InGame/Creative/PropertyEditor/Icon_Picker/Textures_V2/64x64/T_UI_IconLibrary_Stairs_64
   /Game/Athena/UI/InGame/Creative/PropertyEditor/Icon_Picker/Textures_V2/64x64/T_UI_IconLibrary_Circuit_64
   /Game/Athena/UI/InGame/Creative/PropertyEditor/Icon_Picker/Textures_V2/64x64/T_UI_IconLibrary_Processor_64
   /Game/Athena/UI/InGame/Creative/PropertyEditor/Icon_Picker/Textures_V2/64x64/T_UI_IconLibrary_RobotArm_64
   /Game/Athena/UI/InGame/Creative/PropertyEditor/Icon_Picker/Textures_V2/64x64/T_UI_IconLibrary_Globe_64
   /Game/Athena/UI/InGame/Creative/PropertyEditor/Icon_Picker/Textures_V2/64x64/T_UI_IconLibrary_Satellite_64
   /Game/Athena/UI/InGame/Creative/PropertyEditor/Icon_Picker/Textures_V2/64x64/T_UI_IconLibrary_Atom_64
   /Game/Athena/UI/InGame/Creative/PropertyEditor/Icon_Picker/Textures_V2/64x64/T_UI_IconLibrary_Microphone_64
   /Game/Athena/UI/InGame/Creative/PropertyEditor/Icon_Picker/Textures_V2/64x64/T_UI_IconLibrary_TrebleClef_64
   /Game/Athena/UI/InGame/Creative/PropertyEditor/Icon_Picker/Textures_V2/64x64/T_UI_IconLibrary_Cloud_64
   /Game/Athena/UI/InGame/Creative/PropertyEditor/Icon_Picker/Textures_V2/64x64/T_UI_IconLibrary_Connected_64
   /Game/Athena/UI/InGame/Creative/PropertyEditor/Icon_Picker/Textures_V2/64x64/T_UI_IconLibrary_Disconnected_64
   /Game/Athena/UI/InGame/Creative/PropertyEditor/Icon_Picker/Textures_V2/64x64/T_UI_IconLibrary_Server_64
   /Game/Athena/UI/InGame/Creative/PropertyEditor/Icon_Picker/Textures_V2/64x64/T_UI_IconLibrary_Clapperboard_64
   /Game/Athena/UI/InGame/Creative/PropertyEditor/Icon_Picker/Textures_V2/64x64/T_UI_IconLibrary_Hashtag_64
   /Game/Athena/UI/InGame/Creative/PropertyEditor/Icon_Picker/Textures_V2/64x64/T_UI_IconLibrary_Infinity_64
   /Game/Athena/UI/InGame/Creative/PropertyEditor/Icon_Picker/Textures_V2/64x64/T_UI_IconLibrary_Parachute_64
   /Game/Athena/UI/InGame/Creative/PropertyEditor/Icon_Picker/Textures_V2/64x64/T_UI_IconLibrary_Bus_64
   /Game/Athena/UI/InGame/Creative/PropertyEditor/Icon_Picker/Textures_V2/64x64/T_UI_IconLibrary_BeachBall_64
   /Game/Athena/UI/InGame/Creative/PropertyEditor/Icon_Picker/Textures_V2/64x64/T_UI_IconLibrary_Umbrella_64
   /Game/Athena/UI/InGame/Creative/PropertyEditor/Icon_Picker/Textures_V2/64x64/T_UI_IconLibrary_Suitcase_64
   /Game/Athena/UI/InGame/Creative/PropertyEditor/Icon_Picker/Textures_V2/64x64/T_UI_IconLibrary_AlarmClock_64
   /Game/Athena/UI/InGame/Creative/PropertyEditor/Icon_Picker/Textures_V2/64x64/T_UI_IconLibrary_Calculator_64
   /Game/Athena/UI/InGame/Creative/PropertyEditor/Icon_Picker/Textures_V2/64x64/T_UI_IconLibrary_Raincloud_64
   /Game/Athena/UI/InGame/Creative/PropertyEditor/Icon_Picker/Textures_V2/64x64/T_UI_IconLibrary_SnowCloud_64.T_UI_IconLibrary_Snowcloud_64
   /Game/Athena/UI/InGame/Creative/PropertyEditor/Icon_Picker/Textures_V2/64x64/T_UI_IconLibrary_Stormcloud_64
   /Game/Athena/UI/InGame/Creative/PropertyEditor/Icon_Picker/Textures_V2/64x64/T_UI_IconLibrary_Sun_64
   /Game/Athena/UI/InGame/Creative/PropertyEditor/Icon_Picker/Textures_V2/64x64/T_UI_IconLibrary_Moon_64
   /Game/Athena/UI/InGame/Creative/PropertyEditor/Icon_Picker/Textures_V2/64x64/T_UI_IconLibrary_Tornado_64
   /Game/Athena/UI/InGame/Creative/PropertyEditor/Icon_Picker/Textures_V2/64x64/T_UI_IconLibrary_Flower_64
   /Game/Athena/UI/InGame/Creative/PropertyEditor/Icon_Picker/Textures_V2/64x64/T_UI_IconLibrary_Chest_64
   /Game/Athena/UI/InGame/Creative/PropertyEditor/Icon_Picker/Textures_V2/64x64/T_UI_IconLibrary_Palette_64
   /Game/Athena/UI/InGame/Creative/PropertyEditor/Icon_Picker/Textures_V2/64x64/T_UI_IconLibrary_Cheese_64
   /Game/Athena/UI/InGame/Creative/PropertyEditor/Icon_Picker/Textures_V2/64x64/T_UI_IconLibrary_BowlingPins_64
   /Game/Athena/UI/InGame/Creative/PropertyEditor/Icon_Picker/Textures_V2/64x64/T_UI_IconLibrary_BowlingBall_64
   /Game/Athena/UI/InGame/Creative/PropertyEditor/Icon_Picker/Textures_V2/64x64/T_UI_IconLibrary_Swimming_64
   /Game/Athena/UI/InGame/Creative/PropertyEditor/Icon_Picker/Textures_V2/64x64/T_UI_IconLibrary_BaseballBat_64
   /Game/Athena/UI/InGame/Creative/PropertyEditor/Icon_Picker/Textures_V2/64x64/T_UI_IconLibrary_Baseball_64
   /Game/Athena/UI/InGame/Creative/PropertyEditor/Icon_Picker/Textures_V2/64x64/T_UI_IconLibrary_VolleyBall_64.T_UI_IconLibrary_Volleyball_64
   /Game/Athena/UI/InGame/Creative/PropertyEditor/Icon_Picker/Textures_V2/64x64/T_UI_IconLibrary_Hockey_64
   /Game/Athena/UI/InGame/Creative/PropertyEditor/Icon_Picker/Textures_V2/64x64/T_UI_IconLibrary_Tennis_64
   /Game/Athena/UI/InGame/Creative/PropertyEditor/Icon_Picker/Textures_V2/64x64/T_UI_IconLibrary_Golf_64
   /Game/Athena/UI/InGame/Creative/PropertyEditor/Icon_Picker/Textures_V2/64x64/T_UI_IconLibrary_Archery_64
   /Game/Athena/UI/InGame/Creative/PropertyEditor/Icon_Picker/Textures_V2/64x64/T_UI_IconLibrary_Target_64
   /Game/Athena/UI/InGame/Creative/PropertyEditor/Icon_Picker/Textures_V2/64x64/T_UI_IconLibrary_Crosshair_64
   /Game/Athena/UI/InGame/Creative/PropertyEditor/Icon_Picker/Textures_V2/64x64/T_UI_IconLibrary_Balloon_64
   /Game/Athena/UI/InGame/Creative/PropertyEditor/Icon_Picker/Textures_V2/64x64/T_UI_IconLibrary_Quadcrasher_64.T_UI_IconLibrary_QuadCrasher_64
   /Game/Athena/UI/InGame/Creative/PropertyEditor/Icon_Picker/Textures_V2/64x64/T_UI_IconLibrary_GolfCart_64
   /Game/Athena/UI/InGame/Creative/PropertyEditor/Icon_Picker/Textures_V2/64x64/T_UI_IconLibrary_Driftboard_64.T_UI_IconLibrary_DriftBoard_64
   /Game/Athena/UI/InGame/Creative/PropertyEditor/Icon_Picker/Textures_V2/64x64/T_UI_IconLibrary_HamsterBall_64
   /Game/Athena/UI/InGame/Creative/PropertyEditor/Icon_Picker/Textures_V2/64x64/T_UI_IconLibrary_Motorboat_64
   /Game/Athena/UI/InGame/Creative/PropertyEditor/Icon_Picker/Textures_V2/64x64/T_UI_IconLibrary_Mech_64
   /Game/Athena/UI/InGame/Creative/PropertyEditor/Icon_Picker/Textures_V2/64x64/T_UI_IconLibrary_Pow_64
   /Game/Athena/UI/InGame/Creative/PropertyEditor/Icon_Picker/Textures_V2/64x64/T_UI_IconLibrary_Impact_64
   /Game/Athena/UI/InGame/Creative/PropertyEditor/Icon_Picker/Textures_V2/64x64/T_UI_IconLibrary_Mountains_64
   /Game/Athena/UI/InGame/Creative/PropertyEditor/Icon_Picker/Textures_V2/64x64/T_UI_IconLibrary_Waves_64
   /Game/Athena/UI/InGame/Creative/PropertyEditor/Icon_Picker/Textures_V2/64x64/T_UI_IconLibrary_Water_64
   /Game/Athena/UI/InGame/Creative/PropertyEditor/Icon_Picker/Textures_V2/64x64/T_UI_IconLibrary_Road_64
   /Game/Athena/UI/InGame/Creative/PropertyEditor/Icon_Picker/Textures_V2/64x64/T_UI_IconLibrary_Caution_64
   /Game/Athena/UI/InGame/Creative/PropertyEditor/Icon_Picker/Textures_V2/64x64/T_UI_IconLibrary_StarMarker_64
   /Game/Athena/UI/InGame/Creative/PropertyEditor/Icon_Picker/Textures_V2/64x64/T_UI_IconLibrary_RightLaneMerge_64
   /Game/Athena/UI/InGame/Creative/PropertyEditor/Icon_Picker/Textures_V2/64x64/T_UI_IconLibrary_LeftorRightTurn_64
   /Game/Athena/UI/InGame/Creative/PropertyEditor/Icon_Picker/Textures_V2/64x64/T_UI_IconLibrary_LeftCurve_64
   /Game/Athena/UI/InGame/Creative/PropertyEditor/Icon_Picker/Textures_V2/64x64/T_UI_IconLibrary_RightCurve_64
   /Game/Athena/UI/InGame/Creative/PropertyEditor/Icon_Picker/Textures_V2/64x64/T_UI_IconLibrary_LeftTurn_64
   /Game/Athena/UI/InGame/Creative/PropertyEditor/Icon_Picker/Textures_V2/64x64/T_UI_IconLibrary_RightTurn_64
   /Game/Athena/UI/InGame/Creative/PropertyEditor/Icon_Picker/Textures_V2/64x64/T_UI_IconLibrary_U-Turn_64
   /Game/Athena/UI/InGame/Creative/PropertyEditor/Icon_Picker/Textures_V2/64x64/T_UI_IconLibrary_NarrowLane_64
   /Game/Athena/UI/InGame/Creative/PropertyEditor/Icon_Picker/Textures_V2/64x64/T_UI_IconLibrary_NewLane_64
   /Game/Athena/UI/InGame/Creative/PropertyEditor/Icon_Picker/Textures_V2/64x64/T_UI_IconLibrary_Intersection_64
   /Game/Athena/UI/InGame/Creative/PropertyEditor/Icon_Picker/Textures_V2/64x64/T_UI_IconLibrary_People_64
   /Game/Athena/UI/InGame/Creative/PropertyEditor/Icon_Picker/Textures_V2/64x64/T_UI_IconLibrary_Bicycle_64
   /Game/Athena/UI/InGame/Creative/PropertyEditor/Icon_Picker/Textures_V2/64x64/T_UI_IconLibrary_Wood_64
   /Game/Athena/UI/InGame/Creative/PropertyEditor/Icon_Picker/Textures_V2/64x64/T_UI_IconLibrary_Stone_64
   /Game/Athena/UI/InGame/Creative/PropertyEditor/Icon_Picker/Textures_V2/64x64/T_UI_IconLibrary_Metal_64
   /Game/Athena/UI/InGame/Creative/PropertyEditor/Icon_Picker/Textures_V2/64x64/T_UI_IconLibrary_Design_64
   /Game/Athena/UI/InGame/Creative/PropertyEditor/Icon_Picker/Textures_V2/64x64/T_UI_IconLibrary_CandyBar_64
   /Game/Athena/UI/InGame/Creative/PropertyEditor/Icon_Picker/Textures_V2/64x64/T_UI_IconLibrary_Bookshelf_64
   /Game/Athena/UI/InGame/Creative/PropertyEditor/Icon_Picker/Textures_V2/64x64/T_UI_IconLibrary_Llama_64
   /Game/Athena/UI/InGame/Creative/PropertyEditor/Icon_Picker/Textures_V2/64x64/T_UI_IconLibrary_Saw_64
   /Game/Athena/UI/InGame/Creative/PropertyEditor/Icon_Picker/Textures_V2/64x64/T_UI_IconLibrary_Gnome_64
   /Game/Athena/UI/InGame/Creative/PropertyEditor/Icon_Picker/Textures_V2/64x64/T_UI_IconLibrary_Ham_64
   /Game/Athena/UI/InGame/Creative/PropertyEditor/Icon_Picker/Textures_V2/64x64/T_UI_IconLibrary_Fingerprint_64
   /Game/Athena/UI/InGame/Creative/PropertyEditor/Icon_Picker/Textures_V2/64x64/T_UI_IconLibrary_Hack_64
   /Game/Athena/UI/InGame/Creative/PropertyEditor/Icon_Picker/Textures_V2/64x64/T_UI_IconLibrary_Place_64
   /Game/Athena/UI/InGame/Creative/PropertyEditor/Icon_Picker/Textures_V2/64x64/T_UI_IconLibrary_Cube_64
   /Game/Athena/UI/InGame/Creative/PropertyEditor/Icon_Picker/Textures_V2/64x64/T_UI_IconLibrary_Alarm_64
   /Game/Athena/UI/InGame/Creative/PropertyEditor/Icon_Picker/Textures_V2/64x64/T_UI_IconLibrary_Comet_64
   /Game/Athena/UI/InGame/Creative/PropertyEditor/Icon_Picker/Textures_V2/64x64/T_UI_IconLibrary_Castle_64
   /Game/Athena/UI/InGame/Creative/PropertyEditor/Icon_Picker/Textures_V2/64x64/T_UI_IconLibrary_YinYang_64
   /Game/Athena/UI/InGame/Creative/PropertyEditor/Icon_Picker/Textures_V2/64x64/T_UI_IconLibrary_Helmet_64
   /Game/Athena/UI/InGame/Creative/PropertyEditor/Icon_Picker/Textures_V2/64x64/T_UI_IconLibrary_Jonesy_64
   /Game/Athena/UI/InGame/Creative/PropertyEditor/Icon_Picker/Textures_V2/64x64/T_UI_IconLibrary_Melee_64
   /Game/Athena/UI/InGame/Creative/PropertyEditor/Icon_Picker/Textures_V2/64x64/T_UI_IconLibrary_Medal_64
   /Game/Athena/UI/InGame/Creative/PropertyEditor/Icon_Picker/Textures_V2/64x64/T_UI_IconLibrary_Wing_64
   /Game/Athena/UI/InGame/Creative/PropertyEditor/Icon_Picker/Textures_V2/64x64/T_UI_IconLibrary_Swap_64
   /Game/Athena/UI/InGame/Creative/PropertyEditor/Icon_Picker/Textures_V2/64x64/T_UI_IconLibrary_Reset_64
   /Game/Athena/UI/InGame/Creative/PropertyEditor/Icon_Picker/Textures_V2/64x64/T_UI_IconLibrary_Shark_64
   /Game/Athena/UI/InGame/Creative/PropertyEditor/Icon_Picker/Textures_V2/64x64/T_UI_IconLibrary_FreezerChest_64
   /Game/Athena/UI/InGame/Creative/PropertyEditor/Icon_Picker/Textures_V2/64x64/T_UI_IconLibrary_PaperAirplane_64
   /Game/Athena/UI/InGame/Creative/PropertyEditor/Icon_Picker/Textures_V2/64x64/T_UI_IconLibrary_Crafting_64
   /Game/Athena/UI/InGame/Creative/PropertyEditor/Icon_Picker/Textures_V2/64x64/T_UI_IconLibrary_T-Rex_64
   /Game/Athena/UI/InGame/Creative/PropertyEditor/Icon_Picker/Textures_V2/64x64/T_UI_IconLibrary_Zipline_64.T_UI_IconLibrary_ZipLine_64
   /Game/Athena/UI/InGame/Creative/PropertyEditor/Icon_Picker/Textures_V2/64x64/T_UI_IconLibrary_Tree_64
   /Game/Athena/UI/InGame/Creative/PropertyEditor/Icon_Picker/Textures_V2/64x64/T_UI_IconLibrary_Diamond_64
   /Game/Athena/UI/InGame/Creative/PropertyEditor/Icon_Picker/Textures_V2/64x64/T_UI_IconLibrary_Directions_64
   /Game/Athena/UI/InGame/Creative/PropertyEditor/Icon_Picker/Textures_V2/64x64/T_UI_IconLibrary_Share_64
   /Game/Athena/UI/InGame/Creative/PropertyEditor/Icon_Picker/Textures_V2/64x64/T_UI_IconLibrary_SkiMask_64
   /Game/Athena/UI/InGame/Creative/PropertyEditor/Icon_Picker/Textures_V2/64x64/T_UI_IconLibrary_Folder_64
   /Game/Athena/UI/InGame/Creative/PropertyEditor/Icon_Picker/Textures_V2/64x64/T_UI_IconLibrary_Muscle_64
   /Game/Athena/UI/InGame/Creative/PropertyEditor/Icon_Picker/Textures_V2/64x64/T_UI_IconLibrary_Mask_64
   /Game/Athena/UI/InGame/Creative/PropertyEditor/Icon_Picker/Textures_V2/64x64/T_UI_IconLibrary_Lamp_64
   /Game/Athena/UI/InGame/Creative/PropertyEditor/Icon_Picker/Textures_V2/64x64/T_UI_IconLibrary_Incognito_64
   /Game/Athena/UI/InGame/Creative/PropertyEditor/Icon_Picker/Textures_V2/64x64/T_UI_IconLibrary_Radar_64
   /Game/Athena/UI/InGame/Creative/PropertyEditor/Icon_Picker/Textures_V2/64x64/T_UI_IconLibrary_ClothesHanger_64
   /Game/Athena/UI/InGame/Creative/PropertyEditor/Icon_Picker/Textures_V2/64x64/T_UI_IconLibrary_ShoppingBag_64
   /Game/Athena/UI/InGame/Creative/PropertyEditor/Icon_Picker/Textures_V2/64x64/T_UI_IconLibrary_Scenery_64
   /Game/Athena/UI/InGame/Creative/PropertyEditor/Icon_Picker/Textures_V2/64x64/T_UI_IconLibrary_DogTags_64
   /Game/Athena/UI/InGame/Creative/PropertyEditor/Icon_Picker/Textures_V2/64x64/T_UI_IconLibrary_Locker_64
   /Game/Athena/UI/InGame/Creative/PropertyEditor/Icon_Picker/Textures_V2/64x64/T_UI_IconLibrary_Bracket_64
  ]
 }
 PrimaryDataAsset {type=PDA_UI_Creative_CoreGamer_C, default=/Game/Athena/HUD/LTM/Playground/CoreGamer/Data/DA_UI_Creative_CoreGamer}
 ScreenCenterOpacityCurve {type=CurveFloat, default=&/CoreGamerScreenCenterOpacityCurve}
 TimerUpdateRate {type=number, default=0.05}
]
widgets {
 info {inheritedWidgetCount=0, parentClass=/Script/FortniteUI.FortCoreGamerObjectiveRadial, rootWidgetClass=&GridPanel, widgetCount=14}
 widgets [
  Root {widgetClassPath=&GridPanel, widget=Root}
  Main {
   widgetClassPath=&HorizontalBox
   bIsVariable=true
   parent=Root
   widget=Main
   slot {class=GridSlot, properties {column=1, horizontalAlignment=HAlign_Center, row=2, verticalAlignment=VAlign_Center}}
  }
  Visuals {widgetClassPath=&Overlay, parent=Main, widget=Visuals, slot {class=HorizontalBoxSlot}}
  Image_Marker {
   widgetClassPath=&Image
   bIsVariable=true
   parent=Visuals
   widget=Image_Marker
   properties {brush {
     bIsDynamicallyLoaded=false
     drawAs=Image
     imageType=NoImage
     mirroring=NoMirror
     resourceName=None
     resourceObject=/Game/Athena/HUD/LTM/Playground/CoreGamer/Materials/MI_UI_Creative_CoreGamer_MarkerProgress
     tiling=NoTile
     imageSize=<170 170>
     margin=<ltrb 0 0 0 0>
     outlineSettings {
      bUseBrushTransparency=false
      roundingType=HalfHeightRadius
      width=0
      color {colorUseRule=UseColor_Specified, specifiedColor=<rgba 0 0 0 0>}
      cornerRadii=<0 0 0 0>
     }
     tintColor {colorUseRule=UseColor_Specified, specifiedColor=<rgba 1 1 1 1>}
     uVRegion {bIsValid=false, max=<0 0>, min=<0 0>}
    }}
   slot {class=OverlaySlot, properties {horizontalAlignment=HAlign_Center, verticalAlignment=VAlign_Center, padding=<ltrb 14 14 14 14>}}
  }
  Image_Marker_Pulse {
   widgetClassPath=&Image
   bIsVariable=true
   parent=Visuals
   widget=Image_Marker_Pulse
   properties {
    visibility=Collapsed
    brush {
     bIsDynamicallyLoaded=false
     drawAs=Image
     imageType=NoImage
     mirroring=NoMirror
     resourceName=None
     resourceObject=/Game/Athena/HUD/LTM/Playground/CoreGamer/Materials/MI_UI_Creative_CoreGamer_MarkerPulse
     tiling=NoTile
     imageSize=<170 170>
     margin=<ltrb 0 0 0 0>
     outlineSettings {
      bUseBrushTransparency=false
      roundingType=HalfHeightRadius
      width=0
      color {colorUseRule=UseColor_Specified, specifiedColor=<rgba 0 0 0 0>}
      cornerRadii=<0 0 0 0>
     }
     tintColor {colorUseRule=UseColor_Specified, specifiedColor=<rgba 1 1 1 1>}
     uVRegion {bIsValid=false, max=<0 0>, min=<0 0>}
    }
    renderTransform {angle=-45, scale=<1 1>, shear=<0 0>, translation=<0 0>}
   }
   slot {class=OverlaySlot, properties {horizontalAlignment=HAlign_Center, verticalAlignment=VAlign_Center, padding=<ltrb 12 12 12 12>}}
  }
  Image_Rings {
   widgetClassPath=&Image
   bIsVariable=true
   parent=Visuals
   widget=Image_Rings
   properties {
    renderOpacity=0
    visibility=HitTestInvisible
    brush {
     bIsDynamicallyLoaded=false
     drawAs=Image
     imageType=NoImage
     mirroring=NoMirror
     resourceName=None
     resourceObject=/Game/Athena/HUD/LTM/Playground/CoreGamer/Materials/MI_UI_Creative_CoreGamer_AdditiveRingsEffect
     tiling=NoTile
     imageSize=<32 32>
     margin=<ltrb 0 0 0 0>
     outlineSettings {
      bUseBrushTransparency=false
      roundingType=HalfHeightRadius
      width=0
      color {colorUseRule=UseColor_Specified, specifiedColor=<rgba 0 0 0 0>}
      cornerRadii=<0 0 0 0>
     }
     tintColor {colorUseRule=UseColor_Specified, specifiedColor=<rgba 1 1 1 1>}
     uVRegion {bIsValid=false, max=<0 0>, min=<0 0>}
    }
   }
   slot {class=OverlaySlot, properties {horizontalAlignment=HAlign_Fill, verticalAlignment=VAlign_Fill, padding=<ltrb 12 12 12 12>}}
  }
  Image_Flare {
   widgetClassPath=&Image
   bIsVariable=true
   parent=Visuals
   widget=Image_Flare
   properties {
    renderOpacity=0
    visibility=HitTestInvisible
    brush {
     bIsDynamicallyLoaded=false
     drawAs=Image
     imageType=NoImage
     mirroring=NoMirror
     resourceName=None
     resourceObject=/Game/Athena/HUD/LTM/Playground/CoreGamer/Materials/MI_UI_Creative_CoreGamer_AdditiveFlareEffect
     tiling=NoTile
     imageSize=<32 32>
     margin=<ltrb 0 0 0 0>
     outlineSettings {
      bUseBrushTransparency=false
      roundingType=HalfHeightRadius
      width=0
      color {colorUseRule=UseColor_Specified, specifiedColor=<rgba 0 0 0 0>}
      cornerRadii=<0 0 0 0>
     }
     tintColor {colorUseRule=UseColor_Specified, specifiedColor=<rgba 1 1 1 1>}
     uVRegion {bIsValid=false, max=<0 0>, min=<0 0>}
    }
   }
   slot {class=OverlaySlot, properties {horizontalAlignment=HAlign_Fill, verticalAlignment=VAlign_Fill, padding=<ltrb 12 12 12 12>}}
  }
  TextContainer {
   widgetClassPath=&VerticalBox
   parent=Main
   widget=TextContainer
   slot {class=HorizontalBoxSlot, properties {horizontalAlignment=HAlign_Left, verticalAlignment=VAlign_Center}}
  }
  OptionalText {
   widgetClassPath=&CommonTextBlock
   bIsVariable=true
   parent=TextContainer
   widget=OptionalText
   properties {
    justification=Center
    text=CAPTURE
    visibility=Collapsed
    font {
     bForceMonospaced=false
     bMaterialIsStencil=false
     fontMaterial=null
     fontObject=&BurbankSmall
     letterSpacing=0
     monospacedWidth=1
     size=28
     skewAmount=0
     typefaceFontName=Bold
     outlineSettings {
      bApplyOutlineToDropShadows=false
      bMiteredCorners=false
      bSeparateFillAlpha=false
      outlineMaterial=null
      outlineSize=0
      outlineColor=<rgba 0 0 0 1>
     }
    }
    shadowOffset=<0 0>
   }
   slot {class=VerticalBoxSlot, properties {horizontalAlignment=HAlign_Left, verticalAlignment=VAlign_Center}}
  }
  DistanceText {
   widgetClassPath=&CommonTextBlock
   bIsVariable=true
   parent=TextContainer
   widget=DistanceText
   properties {
    justification=Center
    text=100M
    visibility=Collapsed
    font {
     bForceMonospaced=false
     bMaterialIsStencil=false
     fontMaterial=null
     fontObject=&Burbank
     letterSpacing=0
     monospacedWidth=1
     size=24
     skewAmount=0
     typefaceFontName=Burbank-PlayerPower
     outlineSettings {
      bApplyOutlineToDropShadows=false
      bMiteredCorners=false
      bSeparateFillAlpha=false
      outlineMaterial=null
      outlineSize=0
      outlineColor=<rgba 0 0 0 1>
     }
    }
    shadowOffset=<0 0>
   }
   slot {class=VerticalBoxSlot, properties {horizontalAlignment=HAlign_Left, verticalAlignment=VAlign_Center}}
  }
  ArrowTop {
   widgetClassPath=&Image
   bIsVariable=true
   parent=Root
   widget=ArrowTop
   properties {
    visibility=Collapsed
    brush {
     bIsDynamicallyLoaded=false
     drawAs=Image
     imageType=NoImage
     mirroring=NoMirror
     resourceName=None
     resourceObject=/Game/Athena/HUD/Creative/MaterialInstances/MI_UI_Creative_DirectionArrow
     tiling=NoTile
     imageSize=<40 40>
     margin=<ltrb 0 0 0 0>
     outlineSettings {
      bUseBrushTransparency=false
      roundingType=HalfHeightRadius
      width=0
      color {colorUseRule=UseColor_Specified, specifiedColor=<rgba 0 0 0 0>}
      cornerRadii=<0 0 0 0>
     }
     tintColor {colorUseRule=UseColor_Specified, specifiedColor=<rgba 1 1 1 1>}
     uVRegion {bIsValid=false, max=<0 0>, min=<0 0>}
    }
   }
   slot {class=GridSlot, properties {column=1, horizontalAlignment=HAlign_Center, verticalAlignment=VAlign_Center}}
  }
  ArrowBottom {
   widgetClassPath=&Image
   bIsVariable=true
   parent=Root
   widget=ArrowBottom
   properties {
    visibility=Collapsed
    brush {
     bIsDynamicallyLoaded=false
     drawAs=Image
     imageType=NoImage
     mirroring=NoMirror
     resourceName=None
     resourceObject=&MI_UI_Creative_DirectionArrow
     tiling=NoTile
     imageSize=<40 40>
     margin=<ltrb 0 0 0 0>
     outlineSettings {
      bUseBrushTransparency=false
      roundingType=HalfHeightRadius
      width=0
      color {colorUseRule=UseColor_Specified, specifiedColor=<rgba 0 0 0 0>}
      cornerRadii=<0 0 0 0>
     }
     tintColor {colorUseRule=UseColor_Specified, specifiedColor=<rgba 1 1 1 1>}
     uVRegion {bIsValid=false, max=<0 0>, min=<0 0>}
    }
    renderTransform {angle=180, scale=<1 1>, shear=<0 0>, translation=<0 0>}
   }
   slot {class=GridSlot, properties {column=1, horizontalAlignment=HAlign_Center, row=4, verticalAlignment=VAlign_Center}}
  }
  ArrowLeft {
   widgetClassPath=&Image
   bIsVariable=true
   parent=Root
   widget=ArrowLeft
   properties {
    visibility=Collapsed
    brush {
     bIsDynamicallyLoaded=false
     drawAs=Image
     imageType=NoImage
     mirroring=NoMirror
     resourceName=None
     resourceObject=&MI_UI_Creative_DirectionArrow
     tiling=NoTile
     imageSize=<40 40>
     margin=<ltrb 0 0 0 0>
     outlineSettings {
      bUseBrushTransparency=false
      roundingType=HalfHeightRadius
      width=0
      color {colorUseRule=UseColor_Specified, specifiedColor=<rgba 0 0 0 0>}
      cornerRadii=<0 0 0 0>
     }
     tintColor {colorUseRule=UseColor_Specified, specifiedColor=<rgba 1 1 1 1>}
     uVRegion {bIsValid=false, max=<0 0>, min=<0 0>}
    }
    renderTransform {angle=-90, scale=<1 1>, shear=<0 0>, translation=<0 0>}
   }
   slot {class=GridSlot, properties {horizontalAlignment=HAlign_Center, row=2, verticalAlignment=VAlign_Center}}
  }
  ArrowRight {
   widgetClassPath=&Image
   bIsVariable=true
   parent=Root
   widget=ArrowRight
   properties {
    visibility=Collapsed
    brush {
     bIsDynamicallyLoaded=false
     drawAs=Image
     imageType=NoImage
     mirroring=NoMirror
     resourceName=None
     resourceObject=&MI_UI_Creative_DirectionArrow
     tiling=NoTile
     imageSize=<40 40>
     margin=<ltrb 0 0 0 0>
     outlineSettings {
      bUseBrushTransparency=false
      roundingType=HalfHeightRadius
      width=0
      color {colorUseRule=UseColor_Specified, specifiedColor=<rgba 0 0 0 0>}
      cornerRadii=<0 0 0 0>
     }
     tintColor {colorUseRule=UseColor_Specified, specifiedColor=<rgba 1 1 1 1>}
     uVRegion {bIsValid=false, max=<0 0>, min=<0 0>}
    }
    renderTransform {angle=90, scale=<1 1>, shear=<0 0>, translation=<0 0>}
   }
   slot {class=GridSlot, properties {column=2, horizontalAlignment=HAlign_Center, row=2, verticalAlignment=VAlign_Center, padding=<ltrb 5 0 0 0>}}
  }
 ]
}
