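# Playground LTM HUD widget that lays out objective indicators for the CoreGamer mode, based on the native FortCoreGamerObjectiveWidget.
# A minimum-size SizeBox centers the inherited ObjectiveContainer, a horizontal box alternating three radial and two destructible objective child widgets, all vertically centered.
# The event graph is empty and no variables are declared, so the Blueprint supplies only layout and leaves any objective logic to the parent class or child widgets.
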WidgetBlueprint /Game/Athena/HUD/LTM/Playground/CoreGamer/CoreGamer_ObjectiveWidget
graphs [EventGraph {links [], nodes []}]
variables []
widgets {
 info {inheritedWidgetCount=2, parentClass=/Script/FortniteUI.FortCoreGamerObjectiveWidget, rootWidgetClass=&SizeBox, widgetCount=8}
 widgets [
  SizeBox_0 {
   widgetClassPath=&SizeBox
   widget=SizeBox_0
   properties {bOverride_MinDesiredHeight=true, bOverride_MinDesiredWidth=true, minDesiredHeight=75, minDesiredWidth=200}
  }
  ObjectiveContainer {
   widgetClassPath=&HorizontalBox
   bInherited=true
   parent=SizeBox_0
   widget=ObjectiveContainer
   slot {class=SizeBoxSlot, properties {horizontalAlignment=HAlign_Center}}
  }
  CoreGamer_ObjectiveRadial {
   widgetClassPath=&/CoreGamer_ObjectiveRadial
   bIsVariable=true
   parent=ObjectiveContainer
   widget=CoreGamer_ObjectiveRadial
   slot {class=HorizontalBoxSlot, properties {verticalAlignment=VAlign_Center}}
  }
  CoreGamer_DestructibleObjective {
   widgetClassPath=&/CoreGamer_DestructibleObjective
   bIsVariable=true
   parent=ObjectiveContainer
   widget=CoreGamer_DestructibleObjective
   slot {class=HorizontalBoxSlot, properties {verticalAlignment=VAlign_Center}}
  }
  CoreGamer_ObjectiveRadial_71 {
   widgetClassPath=&CoreGamer_ObjectiveRadial
   bIsVariable=true
   parent=ObjectiveContainer
   widget=CoreGamer_ObjectiveRadial_71
   slot {class=HorizontalBoxSlot, properties {verticalAlignment=VAlign_Center}}
  }
  CoreGamer_DestructibleObjective_1 {
   widgetClassPath=&CoreGamer_DestructibleObjective
   bIsVariable=true
   parent=ObjectiveContainer
   widget=CoreGamer_DestructibleObjective_1
   slot {class=HorizontalBoxSlot, properties {verticalAlignment=VAlign_Center}}
  }
  CoreGamer_ObjectiveRadial_134 {
   widgetClassPath=&CoreGamer_ObjectiveRadial
   bIsVariable=true
   parent=ObjectiveContainer
   widget=CoreGamer_ObjectiveRadial_134
   slot {class=HorizontalBoxSlot, properties {verticalAlignment=VAlign_Center}}
  }
  TutorialHighlight {widgetClassPath=&TutorialHighlight}
 ]
}
