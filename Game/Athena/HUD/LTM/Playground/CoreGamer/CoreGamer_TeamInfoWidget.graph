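# HUD widget for Creative "CoreGamer" minigames that shows one team's name, score, progress bar, and either an icon, an ordinal placement or neither.
# Custom events set each element, with score over goal driving a dynamic material's Progress parameter.
# SetTeam tints the icon and bar using ownership colors from the CoreGamer data asset.
# PreConstruct applies width, side display and mirroring through flow direction and material parameters, while mirrored X swaps progress colors.
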WidgetBlueprint /Game/Athena/HUD/LTM/Playground/CoreGamer/CoreGamer_TeamInfoWidget
graphs [
 EventGraph {nodes [
   n1 = SetPlacement!() -> (OutputDelegate, then, Placement:int) {is=title K2Node_CustomEvent}
   > UpdatePlacementNumber!(execute, self, InPlacementNumber:int=@n1.Placement) -> (then)
   n2 = SetTeamName!() -> (OutputDelegate, then, TeamName:Text) {is=title K2Node_CustomEvent}
   > Set Text(Text=@n2.TeamName, self=Get TeamNameText() -> (TeamNameText:Common Text Object Reference))
   n3 = SetLeftDisplayType!() -> (OutputDelegate, then, DisplayType:CoreGamerTeamInfoLeftDisplayType Enum=NewEnumerator0) {is=title K2Node_CustomEvent}
   > SetSideDisplayType!(execute, self, InSideDisplayType:CoreGamerTeamInfoLeftDisplayType Enum=@n3.DisplayType) -> (then)
   n4 = SetScore!() -> (OutputDelegate, then, NewScore:int, GoalScore:int) {is=title K2Node_CustomEvent}
   > UpdateScoreAndProgress!(execute, self, InNewScore:int=@n4.NewScore, InGoalScore:int=@n4.GoalScore) -> (then)
   n5 = SetTeamIcon!() -> (OutputDelegate, then, IconOption:Creative Icon Option Structure) {is=title K2Node_CustomEvent}
   > UpdateIcon!(execute, self, InIconOption:"Creative Icon Option Structure (by ref)"=@n5.IconOption) -> (then)
   Event PreConstruct()
   > GetProgressMaterial!(execute, self) -> (then, OutProgressMaterial:Material Instance Dynamic Object Reference)
   > Set HUDTeam Parameters!(
    execute
    self
    InWidthOverride:float=Get WidthOverride() -> (WidthOverride:float)
    InTeamState:E_UI_Creative_CoreGamer_TeamStates Enum=Get TeamState() -> (TeamState:E_UI_Creative_CoreGamer_TeamStates Enum=NewEnumerator0)
    InDisplayType:CoreGamerTeamInfoLeftDisplayType Enum=Get SideDisplayType() -> (SideDisplayType:CoreGamerTeamInfoLeftDisplayType Enum=NewEnumerator0)
    InMirrorX:bool=Get bMirrorX() -> (bMirrorX:bool)
    InMirrorY:bool=Get bMirrorY() -> (bMirrorY:bool)
   ) -> (then) {is=title K2Node_CallFunction}
  ]}
 GetProgressMaterial {nodes [
   GetProgressMaterial() {is=title K2Node_FunctionEntry}
   > n1 = GetDynamicMaterial(self=Get Image_Progress() -> (Image_Progress:Image Object Reference))
   > n2 = Set ProgressDynamicMaterial(ProgressDynamicMaterial:Material Instance Dynamic Object Reference=@n1) -> (Output_Get:Material Instance Dynamic Object Reference)
   > Return Node!(execute, OutProgressMaterial:Material Instance Dynamic Object Reference=@n2)
  ]}
 Set HUDTeam Parameters {nodes [
   n1 = Set HUDTeam Parameters!() -> (
    then
    InWidthOverride:float=~
    InTeamState:E_UI_Creative_CoreGamer_TeamStates Enum=NewEnumerator0
    InDisplayType:CoreGamerTeamInfoLeftDisplayType Enum=NewEnumerator0
    InMirrorX:bool=~
    InMirrorY:bool=~
   ) {is=title K2Node_FunctionEntry}
   > n2 = Sequence!(execute) -> (then_0, then_1, then_2:Exec)
   > SetLeftDisplayType!(
    execute
    self
    DisplayType:CoreGamerTeamInfoLeftDisplayType Enum=Get InDisplayType() -> (InDisplayType:CoreGamerTeamInfoLeftDisplayType Enum=NewEnumerator0)
   ) -> (then) {is=title K2Node_CallFunction}
   > SetWidthOverride(self=Get Main() -> (Main:Size Box Object Reference), InWidthOverride=Get InWidthOverride() -> (InWidthOverride:float))
   Set MirrorX(execute=@n2.then_1, MirrorX:bool=@n1.InMirrorX) -> (Output_Get:bool)
   Set FlowDirectionPreference!(
    execute=@n2.then_2
    FlowDirectionPreference:EFlowDirectionPreference Enum=Select(
     Option 0:EFlowDirectionPreference Enum=LeftToRight
     Option 1:EFlowDirectionPreference Enum=RightToLeft
     Index=Get InMirrorX() -> (InMirrorX:bool)
    ) -> (ReturnValue:EFlowDirectionPreference Enum)
    self:Widget Object Reference=Get ContentHorizontalBox() -> (ContentHorizontalBox:Horizontal Box Object Reference)
   ) -> (then, Output_Get:EFlowDirectionPreference Enum=Inherit)
   n3 = Get ProgressDynamicMaterial() -> (ProgressDynamicMaterial:Material Instance Dynamic Object Reference)
   > SetScalarParameterValue(self=@n3, ParameterName=Mirror X, Value=Select(Option 0:float=0, Option 1:float=1, Index=Get InMirrorX()) -> (ReturnValue:float))
   > SetScalarParameterValue(self=@n3, ParameterName=Mirror Y, Value=Select(Option 0=0, Option 1=1, Index=Get bMirrorY()))
  ]}
 SetSideDisplayType {nodes [
   n1 = SetSideDisplayType!() -> (then, InSideDisplayType:CoreGamerTeamInfoLeftDisplayType Enum=NewEnumerator0) {is=title K2Node_FunctionEntry}
   > n2 = Set SideDisplayType(SideDisplayType:CoreGamerTeamInfoLeftDisplayType Enum=@n1) -> (Output_Get:CoreGamerTeamInfoLeftDisplayType Enum=NewEnumerator0)
   > n3 = Switch on CoreGamerTeamInfoLeftDisplayType!(execute, Selection:CoreGamerTeamInfoLeftDisplayType Enum=@n2) -> (NewEnumerator0:Exec, NewEnumerator1:Exec, NewEnumerator2:Exec) {is=title K2Node_SwitchEnum}
   n4 = Get IconPlacementSwitcher() -> (IconPlacementSwitcher:Widget Switcher Object Reference)
   n5 = SetActiveWidget(
    execute=@n3.NewEnumerator1
    self:Widget Switcher Object Reference=@n4
    Widget=Get IconContainer() -> (IconContainer:Size Box Object Reference)
   )
   SetVisibility(execute=[@n6.then, @n5.then], self=Get IconPlacementSwitcher(), InVisibility=SelfHitTestInvisible)
   n6 = SetActiveWidget(execute=@n3.NewEnumerator2, self=@n4, Widget=Get PlacementContainer() -> (PlacementContainer:Scale Box Object Reference))
   SetVisibility(execute=@n3.NewEnumerator0, self=Get IconPlacementSwitcher(), InVisibility=Collapsed)
  ]}
 SetTeam {nodes [
   n1 = SetTeam!() -> (then, InTeamByte:Byte) {is=title K2Node_FunctionEntry}
   n2 = GetOwningPlayer()
   n3 = GetOwnershipTeamColor!(
    self:PDA UI Creative Core Gamer Object Reference=Get PrimaryDataAsset() -> (PrimaryDataAsset:PDA UI Creative Core Gamer Object Reference)
    InMinigame:Fort Minigame Object Reference=GetMinigame!(ContextActor:Actor Object Reference=@n2) -> (ReturnValue:Fort Minigame Object Reference)
    InLocalPlayerController:Player Controller Object Reference=@n2
    Team:Byte=@n1
   ) -> (
    OutColorSet_Lighter_20_#1:Linear Color Structure
    OutColorSet_Light_21_#2:Linear Color Structure
    OutColorSet_Medium_22_#3:Linear Color Structure
    OutColorSet_Dark_23_#4:Linear Color Structure
    OutColorSet_Darker_24_#5:Linear Color Structure
    OutColorSet_Darker1_27_#6:Linear Color Structure
   )
   > UpdateProgressBarColors!(
    execute
    self
    InProgressColor1:Linear Color Structure=@n3.OutColorSet_Medium_22_#3
    InProgressColor2:Linear Color Structure=@n3
    InBackgroundColor:Linear Color Structure=@n3.OutColorSet_Darker1_27_#6
    InMirrorX:bool=Get MirrorX() -> (MirrorX:bool)
   ) -> (then)
   > SetColorAndOpacity(
    self:Image Object Reference=Get TeamIconImage() -> (TeamIconImage:Fort Lazy Image Object Reference)
    InColorAndOpacity:Linear Color Structure=@n3.OutColorSet_Medium_22_#3
   )
  ]}
 UpdateIcon {nodes [
   n1 = UpdateIcon!() -> (then, InIconOption:"Creative Icon Option Structure (by ref)") {is=title K2Node_FunctionEntry}
   n2 = Get Icon Library Texture!(Icon Option:"Creative Icon Option Structure (by ref)"=@n1, Should Use Large Icon:bool=true) -> (Icon Texture:Texture 2D Soft Object Reference) {is=title K2Node_CallFunction}
   > n3 = Branch(Condition=IsValidSoftObjectReference(SoftObjectReference=@n2))
   n4 = Get TeamIconImage()
   > SetBrushFromLazyTexture!(execute, self:Common Lazy Image Object Reference=@n4, LazyTexture:"Texture 2D Soft Object Reference (by ref)"=@n2, bMatchSize:bool=false) -> (then)
   > SetVisibility(self=@n4, InVisibility=SelfHitTestInvisible)
   SetVisibility(execute=@n3.else, self=Get TeamIconImage(), InVisibility=Collapsed)
  ]}
 UpdatePlacementNumber {nodes [
   UpdatePlacementNumber!() -> (then, InPlacementNumber:int) {is=title K2Node_FunctionEntry}
   > Set Text(
    Text=ToOrdinal!(Number:int=Get InPlacementNumber() -> (InPlacementNumber:int=0)) -> (ReturnValue:Text)
    self=Get PlacementText() -> (PlacementText:Common Text Object Reference)
   )
  ]}
 UpdateProgressBarColors {nodes [
   n1 = UpdateProgressBarColors!() -> (
    then
    InProgressColor1:Linear Color Structure
    InProgressColor2:Linear Color Structure
    InBackgroundColor:Linear Color Structure
    InMirrorX:bool=~
   ) {is=title K2Node_FunctionEntry}
   > SetVectorParameterValue(self=Get ProgressDynamicMaterial(), ParameterName=Background Color, Value=@n1.InBackgroundColor)
   > n2 = Branch(Condition=Get InMirrorX())
   n3 = Get ProgressDynamicMaterial()
   > SetVectorParameterValue(self=@n3, ParameterName=Progress Color 1, Value=Get InProgressColor2() -> (InProgressColor2:Linear Color Structure))
   > SetVectorParameterValue(self=@n3, ParameterName=Progress Color 2, Value=Get InProgressColor1() -> (InProgressColor1:Linear Color Structure))
   n4 = Get ProgressDynamicMaterial()
   SetVectorParameterValue(execute=@n2.else, self=@n4, ParameterName=Progress Color 1, Value=Get InProgressColor1())
   > SetVectorParameterValue(self=@n4, ParameterName=Progress Color 2, Value=Get InProgressColor2())
  ]}
 UpdateScoreAndProgress {nodes [
   UpdateScoreAndProgress!() -> (then, InNewScore:int, InGoalScore:int) {is=title K2Node_FunctionEntry}
   > n1 = Branch(Condition="integer > integer"(A=Get InGoalScore() -> (InGoalScore:int=0)))
   > Set Text(
    Text=Format Text!(Format="LOCTEXT(\"{Current}/{Goal}\")", Current:int=Get InNewScore() -> (InNewScore:int=0), Goal:int=Get InGoalScore() -> (InGoalScore=0)) -> (Result)
    self=Get ScoreText() -> (ScoreText:Common Text Object Reference)
   )
   n2 = Get InNewScore() -> (InNewScore=0)
   n3 = Get InGoalScore() -> (InGoalScore=0)
   n4 = Get ProgressDynamicMaterial()
   > SetScalarParameterValue(self=@n4, ParameterName=Progress, Value=float / float(A="To Float (Integer)"(InInt=@n2), B="To Float (Integer)"(InInt=@n3)))
   Set Text(execute=@n1.else, Text="To Text (Integer)"(Value=Get InNewScore() -> (InNewScore=0)), self=Get ScoreText())
   > SetScalarParameterValue(self=@n4, ParameterName=Progress, Value=0)
  ]}
]
variables [
 BackgroundDynamicMaterial {type=MaterialInstanceDynamic}
 bMirrorX {type=boolean}
 bMirrorY {type=boolean}
 Goal Score {type=integer}
 MirrorX {type=boolean}
 New Score {type=integer}
 PrimaryDataAsset {type=PDA_UI_Creative_CoreGamer_C, default=/Game/Athena/HUD/LTM/Playground/CoreGamer/Data/DA_UI_Creative_CoreGamer}
 ProgressDynamicMaterial {type=MaterialInstanceDynamic}
 SideDisplayType {type=CoreGamerTeamInfoLeftDisplayType, default=ShowIcon}
 TeamState {type=E_UI_Creative_CoreGamer_TeamStates, default=Friendly}
 WidthOverride {type=number, default=350}
]
widgets {
 info {inheritedWidgetCount=0, parentClass=&UserWidget, rootWidgetClass=&Overlay, widgetCount=13}
 widgets [
  Root {widgetClassPath=&Overlay, widget=Root}
  Main {
   widgetClassPath=&SizeBox
   bIsVariable=true
   parent=Root
   widget=Main
   properties {bOverride_WidthOverride=true, widthOverride=350}
   slot {class=OverlaySlot, properties {horizontalAlignment=HAlign_Fill, verticalAlignment=VAlign_Center, padding=<ltrb 16 14 16 14>}}
  }
  ContentHorizontalBox {
   widgetClassPath=&HorizontalBox
   bIsVariable=true
   parent=Main
   widget=ContentHorizontalBox
   properties {flowDirectionPreference=LeftToRight}
   slot {class=SizeBoxSlot, properties {verticalAlignment=VAlign_Top}}
  }
  IconPlacementSwitcher {
   widgetClassPath=&WidgetSwitcher
   bIsVariable=true
   parent=ContentHorizontalBox
   widget=IconPlacementSwitcher
   slot {class=HorizontalBoxSlot, properties {horizontalAlignment=HAlign_Left, verticalAlignment=VAlign_Center}}
  }
  IconContainer {
   widgetClassPath=&SizeBox
   bIsVariable=true
   parent=IconPlacementSwitcher
   widget=IconContainer
   properties {bOverride_HeightOverride=true, bOverride_WidthOverride=true, heightOverride=50, widthOverride=50}
   slot {class=WidgetSwitcherSlot, properties {padding=<ltrb 0 0 2 0>}}
  }
  TeamIconImage {
   widgetClassPath=&FortLazyImage
   bIsVariable=true
   parent=IconContainer
   widget=TeamIconImage
   properties {
    visibility=Collapsed
    brush {
     bIsDynamicallyLoaded=false
     drawAs=Image
     imageType=NoImage
     mirroring=NoMirror
     resourceName=None
     resourceObject=/Game/Athena/UI/InGame/Creative/PropertyEditor/Icon_Picker/Textures/64x64/T_UI_IconLibrary_exclamation_64
     tiling=NoTile
     imageSize=<64 64>
     margin=<ltrb 0 0 0 0>
     outlineSettings {
      bUseBrushTransparency=false
      roundingType=HalfHeightRadius
      width=0
      color {colorUseRule=UseColor_Specified, specifiedColor=<rgba 0 0 0 0>}
      cornerRadii=<0 0 0 0>
     }
     tintColor {colorUseRule=UseColor_Specified, specifiedColor=<rgba 1 1 1 1>}
     uVRegion {bIsValid=false, max=<0 0>, min=<0 0>}
    }
   }
   slot {class=SizeBoxSlot, properties {padding=<ltrb 5 5 5 5>}}
  }
  PlacementContainer {
   widgetClassPath=&ScaleBox
   bIsVariable=true
   parent=IconPlacementSwitcher
   widget=PlacementContainer
   slot {class=WidgetSwitcherSlot, properties {horizontalAlignment=HAlign_Left, verticalAlignment=VAlign_Top, padding=<ltrb 0 0 8 0>}}
  }
  PlacementText {
   widgetClassPath=&CommonTextBlock
   bIsVariable=true
   parent=PlacementContainer
   widget=PlacementText
   properties {
    clipping=ClipToBoundsAlways
    text=2nd
    textOverflowPolicy=Ellipsis
    font {
     bForceMonospaced=false
     bMaterialIsStencil=false
     fontMaterial=null
     fontObject=&BurbankSmall
     letterSpacing=0
     monospacedWidth=1
     size=16
     skewAmount=0
     typefaceFontName=Black
     outlineSettings {
      bApplyOutlineToDropShadows=false
      bMiteredCorners=false
      bSeparateFillAlpha=true
      outlineMaterial=null
      outlineSize=1
      outlineColor=<rgba 0 0 0 0.5>
     }
    }
    shadowOffset=<0 0>
   }
   slot {class=ScaleBoxSlot, properties {horizontalAlignment=HAlign_Left, verticalAlignment=VAlign_Top}}
  }
  VerticalBox {
   widgetClassPath=&VerticalBox
   parent=ContentHorizontalBox
   widget=VerticalBox
   slot {class=HorizontalBoxSlot, properties {size {sizeRule=Fill, value=1}}}
  }
  HorizontalBox_1 {
   widgetClassPath=&HorizontalBox
   parent=VerticalBox
   widget=HorizontalBox_1
   slot {class=VerticalBoxSlot, properties {verticalAlignment=VAlign_Top, padding=<ltrb 0 0 0 6>}}
  }
  TeamNameText {
   widgetClassPath=&CommonTextBlock
   bIsVariable=true
   parent=HorizontalBox_1
   widget=TeamNameText
   properties {
    clipping=ClipToBoundsAlways
    text=TEAM NAME
    textOverflowPolicy=Ellipsis
    textTransformPolicy=ToUpper
    font {
     bForceMonospaced=false
     bMaterialIsStencil=false
     fontMaterial=null
     fontObject=&BurbankSmall
     letterSpacing=0
     monospacedWidth=1
     size=16
     skewAmount=0
     typefaceFontName=Black
     outlineSettings {
      bApplyOutlineToDropShadows=true
      bMiteredCorners=false
      bSeparateFillAlpha=true
      outlineMaterial=&M-FontShadow
      outlineSize=2
      outlineColor=<rgba 0 0 0 1>
     }
    }
    shadowOffset=<0 0>
   }
   slot {
    class=HorizontalBoxSlot
    properties {horizontalAlignment=HAlign_Left, verticalAlignment=VAlign_Bottom, padding=<ltrb 10 0 0 -2>, size {sizeRule=Fill, value=1}}
   }
  }
  ScoreText {
   widgetClassPath=&CommonTextBlock
   bIsVariable=true
   parent=HorizontalBox_1
   widget=ScoreText
   properties {
    text=9999
    font {
     bForceMonospaced=false
     bMaterialIsStencil=false
     fontMaterial=null
     fontObject=&BurbankSmall
     letterSpacing=0
     monospacedWidth=1
     size=16
     skewAmount=0
     typefaceFontName=Black
     outlineSettings {
      bApplyOutlineToDropShadows=true
      bMiteredCorners=false
      bSeparateFillAlpha=true
      outlineMaterial=&M-FontShadow
      outlineSize=2
      outlineColor=<rgba 0 0 0 1>
     }
    }
    shadowOffset=<0 0>
   }
   slot {class=HorizontalBoxSlot, properties {horizontalAlignment=HAlign_Right, padding=<ltrb 10 0 0 -2>}}
  }
  Image_Progress {
   widgetClassPath=&Image
   bIsVariable=true
   parent=VerticalBox
   widget=Image_Progress
   properties {brush {
     bIsDynamicallyLoaded=false
     drawAs=Image
     imageType=NoImage
     mirroring=NoMirror
     resourceName=None
     resourceObject=/Game/Athena/HUD/LTM/Playground/CoreGamer/Materials/MI_UI_Creative_CoreGamer_ProgressBar_Skewed
     tiling=NoTile
     imageSize=<32 10>
     margin=<ltrb 0 0 0 0>
     outlineSettings {
      bUseBrushTransparency=false
      roundingType=HalfHeightRadius
      width=0
      color {colorUseRule=UseColor_Specified, specifiedColor=<rgba 0 0 0 0>}
      cornerRadii=<0 0 0 0>
     }
     tintColor {colorUseRule=UseColor_Specified, specifiedColor=<rgba 1 1 1 1>}
     uVRegion {bIsValid=false, max=<0 0>, min=<0 0>}
    }}
   slot {class=VerticalBoxSlot, properties {verticalAlignment=VAlign_Top}}
  }
 ]
}
