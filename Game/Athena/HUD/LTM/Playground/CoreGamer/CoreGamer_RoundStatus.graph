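# HUD widget built on FortCoreGamerScoreTracker that shows a minigame round's status: local team versus highest-scoring opponent, a central time or goal readout, and an objectives strip.
# OnUpdate switches the center between remaining time and goal, depending on whether a time limit exists.
# Stat updates and team changes refresh both team panels' score, name, icon and team.
# It starts collapsed, appearing only when minigame in-progress UI displays outside a curated hub.
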WidgetBlueprint /Game/Athena/HUD/LTM/Playground/CoreGamer/CoreGamer_RoundStatus
graphs [EventGraph {nodes [
   Event OnUpdate() {is=title K2Node_Event}
   n1 = GetTimeRemainingText!(self:Fort Core Gamer Score Tracker Object Reference) -> (bHasTimeLimit:bool, ReturnValue:Text)
   > n2 = Branch(Condition=@n1)
   > Set Text(Text=@n1.ReturnValue, self=Get TimerText() -> (TimerText:Common Text Object Reference))
   n3 = Get Text_MatchType() -> (Text_MatchType:Common Text Object Reference)
   > Set Text(Text="LOCTEXT(\"TIME\")", self=@n3)
   n4 = Get CentralSwitcher() -> (CentralSwitcher:Widget Switcher Object Reference)
   > SetActiveWidget(self:Widget Switcher Object Reference=@n4, Widget=Get TimerText())
   Set Text(execute=@n2.else, Text="LOCTEXT(\"GOAL\")", self=@n3)
   > SetActiveWidget(self=@n4, Widget=Get GoalText() -> (GoalText:Common Text Object Reference))
   Event Construct()
   > n5 = Sequence()
   > SetVisibility(self=Get Root() -> (Root:Overlay Object Reference), InVisibility=Collapsed)
   SetScore!(
    execute=@n5.then_1
    self:Core Gamer Team Info Widget Object Reference=[
     Get OtherTeamInfo() -> (OtherTeamInfo:Core Gamer Team Info Widget Object Reference)
     Get LocalTeamInfo() -> (LocalTeamInfo:Core Gamer Team Info Widget Object Reference)
    ]
    NewScore:int=0
    GoalScore:int=GetGoalScore!(self:Fort Core Gamer Score Tracker Object Reference) -> (ReturnValue:int=0)
   ) -> (then)
   n6 = Event OnDisplayMinigameInProgressUI!() -> (OutputDelegate, then, bDisplay:bool) {is=title K2Node_Event}
   > Parent: OnDisplayMinigameInProgressUI!(execute, bDisplay:bool=@n6.bDisplay) -> (then) {is=title K2Node_CallParentFunction}
   > n7 = Get!(execute:Exec) -> (then:Exec, else:Exec, CurrentMinigame:Fort Minigame Object Reference)
   n8 = NOT Boolean(A=IsCuratedHub!(self:Fort Minigame Object Reference=@n7) -> (ReturnValue:bool))
   > SetVisibility(self=Get Root(), InVisibility=Select(Option 0=Collapsed, Option 1=SelfHitTestInvisible, Index=AND Boolean(A=@n8, B=@n6.bDisplay)))
   SetVisibility(execute=@n7.else, self=Get Root(), InVisibility=Collapsed)
   n9 = Event OnPlayerStatUpdated() {is=title K2Node_Event}
   n10 = Sequence!(execute=[@n9.then, @n18.then]) -> (then_0, then_1, then_2:Exec)
   n11 = GetGoalScore() -> (ReturnValue=0)
   Set Text(execute=@n10.then_2, Text="To Text (Integer)"(Value=@n11), self=Get GoalText())
   n12 = GetLocalTeamIndex!(self:Fort Core Gamer Score Tracker Object Reference) -> (ReturnValue:int=0)
   n13 = Get LocalTeamInfo()
   SetScore(
    execute=@n10.then_0
    self=@n13
    NewScore=GetScoreForTeam!(self:Fort Core Gamer Score Tracker Object Reference, InTeamIndex:int=@n12) -> (ReturnValue:int=0)
    GoalScore=@n11
   )
   > SetTeamName!(
    execute
    self:Core Gamer Team Info Widget Object Reference=@n13
    TeamName:Text=GetTeamName!(self:Fort Core Gamer Score Tracker Object Reference, InTeamIndex:int=@n12) -> (ReturnValue:Text)
   ) -> (then)
   n14 = Break Minigame Team!(MinigameTeam:"Minigame Team Structure (by ref)"=GetTeamData!(self:Fort Core Gamer Score Tracker Object Reference, DesiredTeamIndex:int=@n12) -> (OutTeam:Minigame Team Structure, bFoundTeam:bool)) -> (
    TeamIndex:Byte
    TeamName:Text
    TeamDescription:Text
    TeamColorIndex:int
    MaxInitTeamSize:int
    InitTeamSizeWeight:int
    bHasBucketAvailable:bool=~
    EliminatedCount:Byte
    TeamSize:Byte
    TeamIcon:Creative Icon Option Structure
    bUseTeamScore:bool=~
   ) {is=title K2Node_BreakStruct}
   > SetTeamIcon!(execute, self:Core Gamer Team Info Widget Object Reference=@n13, IconOption:Creative Icon Option Structure=@n14.TeamIcon) -> (then)
   > SetTeam!(execute, self:Core Gamer Team Info Widget Object Reference=@n13, InTeamByte:Byte=@n14) -> (then)
   n15 = GetHighestOpposingTeamIndex!(self:Fort Core Gamer Score Tracker Object Reference) -> (ReturnValue:int=0)
   n16 = Get OtherTeamInfo()
   SetScore(execute=@n10.then_1, self=@n16, NewScore=GetScoreForTeam(InTeamIndex=@n15) -> (ReturnValue=0), GoalScore=@n11)
   > SetTeamName(self=@n16, TeamName=GetTeamName(InTeamIndex=@n15))
   n17 = Break Minigame Team(MinigameTeam=GetTeamData(DesiredTeamIndex=@n15)) -> (bHasBucketAvailable=~, bUseTeamScore=~)
   > SetTeamIcon(self=@n16, IconOption=@n17.TeamIcon)
   > SetTeam(self=@n16, InTeamByte=@n17)
   n18 = Event OnOwningPlayerChangedTeams() {is=title K2Node_Event}
  ]}]
variables [ScoreStateClass {type=Class, default=/Game/Athena/Playset/Minigames/StatFilters/MinigameStat_Score.MinigameStat_Score_C}]
widgets {
 info {inheritedWidgetCount=1, parentClass=/Script/FortniteUI.FortCoreGamerScoreTracker, rootWidgetClass=&Overlay, widgetCount=14}
 widgets [
  Root {widgetClassPath=&Overlay, bIsVariable=true, widget=Root, properties {visibility=Collapsed}}
  Main {
   widgetClassPath=&VerticalBox
   parent=Root
   widget=Main
   slot {class=OverlaySlot, properties {horizontalAlignment=HAlign_Center, padding=<ltrb 0 15 0 0>}}
  }
  HorizontalBox_0 {widgetClassPath=&HorizontalBox, parent=Main, widget=HorizontalBox_0, slot {class=VerticalBoxSlot}}
  LocalTeamInfo {
   widgetClassPath=&/CoreGamer_TeamInfoWidget
   bIsVariable=true
   parent=HorizontalBox_0
   widget=LocalTeamInfo
   slot {class=HorizontalBoxSlot}
  }
  CentralBlock {
   widgetClassPath=&SizeBox
   parent=HorizontalBox_0
   widget=CentralBlock
   properties {bOverride_WidthOverride=true, widthOverride=100}
   slot {class=HorizontalBoxSlot, properties {horizontalAlignment=HAlign_Center, verticalAlignment=VAlign_Center}}
  }
  VerticalBox_1 {widgetClassPath=&VerticalBox, parent=CentralBlock, widget=VerticalBox_1, slot {class=SizeBoxSlot}}
  Text_MatchType {
   widgetClassPath=&CommonTextBlock
   bIsVariable=true
   parent=VerticalBox_1
   widget=Text_MatchType
   properties {
    justification=Center
    text=TYPE
    textTransformPolicy=ToUpper
    colorAndOpacity {colorUseRule=UseColor_Specified, specifiedColor=<rgba 1 1 1 0.6>}
    font {
     bForceMonospaced=false
     bMaterialIsStencil=false
     fontMaterial=null
     fontObject=&BurbankSmall
     letterSpacing=0
     monospacedWidth=1
     size=14
     skewAmount=0
     typefaceFontName=Black
     outlineSettings {
      bApplyOutlineToDropShadows=true
      bMiteredCorners=false
      bSeparateFillAlpha=true
      outlineMaterial=&M-FontShadow
      outlineSize=2
      outlineColor=<rgba 0 0 0 1>
     }
    }
    shadowOffset=<0 0>
   }
   slot {class=VerticalBoxSlot, properties {horizontalAlignment=HAlign_Center, padding=<ltrb 0 0 0 -6>}}
  }
  CentralSwitcher {
   widgetClassPath=&WidgetSwitcher
   bIsVariable=true
   parent=VerticalBox_1
   widget=CentralSwitcher
   properties {activeWidgetIndex=1}
   slot {class=VerticalBoxSlot, properties {horizontalAlignment=HAlign_Center}}
  }
  GoalText {
   widgetClassPath=&CommonTextBlock
   bIsVariable=true
   parent=CentralSwitcher
   widget=GoalText
   properties {
    justification=Center
    text=100
    font {
     bForceMonospaced=false
     bMaterialIsStencil=false
     fontMaterial=null
     fontObject=&Burbank
     letterSpacing=0
     monospacedWidth=1
     size=24
     skewAmount=0
     typefaceFontName=BurbankRegular-Black
     outlineSettings {
      bApplyOutlineToDropShadows=true
      bMiteredCorners=false
      bSeparateFillAlpha=true
      outlineMaterial=&M-FontShadow
      outlineSize=2
      outlineColor=<rgba 0 0 0 1>
     }
    }
    shadowOffset=<0 0>
   }
   slot {class=WidgetSwitcherSlot}
  }
  TimerText {
   widgetClassPath=&CommonTextBlock
   bIsVariable=true
   parent=CentralSwitcher
   widget=TimerText
   properties {
    justification=Center
    text=12:00
    font {
     bForceMonospaced=false
     bMaterialIsStencil=false
     fontMaterial=null
     fontObject=&BurbankSmall
     letterSpacing=0
     monospacedWidth=1
     size=24
     skewAmount=0
     typefaceFontName=Black
     outlineSettings {
      bApplyOutlineToDropShadows=true
      bMiteredCorners=false
      bSeparateFillAlpha=true
      outlineMaterial=&M-FontShadow
      outlineSize=2
      outlineColor=<rgba 0 0 0 1>
     }
    }
    shadowOffset=<0 0>
   }
   slot {class=WidgetSwitcherSlot, properties {verticalAlignment=VAlign_Center}}
  }
  OtherTeamInfo {
   widgetClassPath=&CoreGamer_TeamInfoWidget
   bIsVariable=true
   parent=HorizontalBox_0
   widget=OtherTeamInfo
   properties {bMirrorX=true, teamState=Enemy}
   slot {class=HorizontalBoxSlot}
  }
  ObjectivesContainer {
   widgetClassPath=&ScaleBox
   parent=Main
   widget=ObjectivesContainer
   properties {stretch=UserSpecified, userSpecifiedScale=0.35}
   slot {class=VerticalBoxSlot, properties {padding=<ltrb 0 -8 0 0>}}
  }
  CoreGamerObjectiveWidget {
   widgetClassPath=&/CoreGamer_ObjectiveWidget
   bIsVariable=true
   parent=ObjectivesContainer
   widget=CoreGamerObjectiveWidget
   properties {objectivePadding=<ltrb 10 0 10 0>}
   slot {class=ScaleBoxSlot, properties {verticalAlignment=VAlign_Fill}}
  }
  TutorialHighlight {widgetClassPath=&TutorialHighlight}
 ]
}
