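# Creative CoreGamer scoreboard row widget showing one team's ordinal placement, uppercase name and score.
# OnRowDataSet fills the texts and calls SetHighlightState, which tints the highlight border and gradient edges with team colors from the CoreGamer data asset.
# Local or first-place rows get full-opacity text and bold names, while gradient edges show only for first place.
# A cached flag plays the Takeover animation when the local team newly takes first; PreConstruct previews states at design time.
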WidgetBlueprint /Game/Athena/HUD/LTM/Playground/CoreGamer/CoreGamer_ScoreTracker_Row
animations [
 SlideOut {
  binding_count=1
  length_seconds=0.2
  object=&self:SlideOut
  sequence {bindings [Overlay_25 {
     id=#1
     possessed_class=Overlay
     tracks [
      {
       type=MovieScene2DTransformTrack
       sections [{
         type=MovieScene2DTransformSection
         channels [
          Translation.X {type=MovieSceneScriptingFloatChannel, keys [[0,-60],[3,0]]}
          Translation.Y {type=MovieSceneScriptingFloatChannel, const [0,2]}
          Angle {type=MovieSceneScriptingFloatChannel}
          Scale.X {type=MovieSceneScriptingFloatChannel}
          Scale.Y {type=MovieSceneScriptingFloatChannel}
          Shear.X {type=MovieSceneScriptingFloatChannel}
          Shear.Y {type=MovieSceneScriptingFloatChannel}
         ]
        }]
      }
      {
       type=MovieSceneFloatTrack
       display_name=Render Opacity
       sections [{type=MovieSceneFloatSection, channels [None {type=MovieSceneScriptingFloatChannel, keys [[0,0],[4,1]]}]}]
      }
     ]
    }]}
 }
 Takeover {
  binding_count=3
  length_seconds=0.5834
  object=&self:Takeover
  sequence {
   bindings [
    HighlightBorder_Main {
     id=#2
     possessed_class=Border
     tracks [{
       type=MovieSceneWidgetMaterialTrack
       sections [{
         type=MovieSceneParameterSection
         channels [
          ColorR {type=MovieSceneScriptingFloatChannel, keys [[0,0.015],[3,0.015],[17,0.245],[24,0.245],[35,0.015]]}
          ColorG {type=MovieSceneScriptingFloatChannel, keys [[0,0.0183],[3,0.0183],[17,0.3983],[24,0.3983],[35,0.0183]]}
          ColorB {type=MovieSceneScriptingFloatChannel, keys [[0,0.0395],[3,0.0395],[17,0.6395],[24,0.6395],[35,0.0395]]}
          ColorA {type=MovieSceneScriptingFloatChannel, const [0.25,5]}
         ]
        }]
      }]
    }
    PlacementText {
     id=#3
     possessed_class=CommonTextBlock
     tracks [{
       type=MovieScene2DTransformTrack
       sections [{
         type=MovieScene2DTransformSection
         channels [
          Scale.X {type=MovieSceneScriptingFloatChannel, keys [[0,1],[4,1.1],[18,1.15],[30,1]]}
          Scale.Y {type=MovieSceneScriptingFloatChannel, keys [[0,1],[4,1.1],[18,1.15],[30,1]]}
         ]
        }]
      }]
    }
    ShineImage {
     id=#4
     possessed_class=Image
     tracks [
      {
       type=MovieSceneByteTrack
       display_name=Visibility
       sections [{type=MovieSceneByteSection, end_frame=30, start_frame=0, channels [None {type=MovieSceneScriptingByteChannel, keys [[0,4],[23,2]]}]}]
      }
      {
       type=MovieScene2DTransformTrack
       sections [{
         type=MovieScene2DTransformSection
         channels [
          Translation.X {type=MovieSceneScriptingFloatChannel, keys [[0,0],[13,35],[18,198.0921],[24,333.9582]]}
          Translation.Y {type=MovieSceneScriptingFloatChannel, const [0,3]}
         ]
        }]
      }
      {
       type=MovieSceneFloatTrack
       display_name=Render Opacity
       sections [{type=MovieSceneFloatSection, channels [None {type=MovieSceneScriptingFloatChannel, keys [[0,0],[3,0.6],[13,0.6],[23,0]]}]}]
      }
     ]
    }
   ]
   display_rate {numerator=60}
  }
 }
]
graphs [
 EventGraph {nodes [
   n1 = SetScoreText!() -> (OutputDelegate, then, NewScore:Text) {is=title K2Node_CustomEvent}
   > Set Text(Text=@n1.NewScore, self=Get ScoreText() -> (ScoreText:Common Text Object Reference))
   n2 = SetPlacement!() -> (OutputDelegate, then, InPlacement:int) {is=title K2Node_CustomEvent}
   > Set Text(Text=ToOrdinal!(Number:int=@n2.InPlacement) -> (ReturnValue:Text), self=Get PlacementText() -> (PlacementText:Common Text Object Reference))
   n3 = SetTeamName!() -> (OutputDelegate, then, TeamName:Text) {is=title K2Node_CustomEvent}
   > Set Text(
    Text=TextToUpper!(InText:"Text (by ref)"=@n3.TeamName) -> (ReturnValue:Text)
    self=Get TeamNameText() -> (TeamNameText:Common Text Object Reference)
   )
   n4 = SetScore!() -> (OutputDelegate, then, NewScore:int) {is=title K2Node_CustomEvent}
   > Set Text(Text="To Text (Integer)"(Value=@n4.NewScore), self=Get ScoreText())
   n5 = SetHighlightState!() -> (OutputDelegate, then, IsLocalTeam:bool=~, IsFirstPlace:bool=~, TeamIndex:int) {is=title K2Node_CustomEvent}
   > n6 = Sequence!(execute) -> (then_0, then_1, then_2:Exec)
   > n7 = Branch(Condition=@n5.IsLocalTeam)
   n8 = GetOwningPlayer()
   > Is Valid(InputObject=@n8)
   n9 = GetOwnershipTeamColor!(
    self:PDA UI Creative Core Gamer Object Reference=Get CoreGamerDataAsset() -> (CoreGamerDataAsset:PDA UI Creative Core Gamer Object Reference)
    InMinigame:Fort Minigame Object Reference=GetMinigame!(ContextActor:Actor Object Reference=@n8) -> (ReturnValue:Fort Minigame Object Reference)
    InLocalPlayerController:Player Controller Object Reference=@n8
    Team:Byte=GetActorTeam!(Actor:Actor Object Reference=@n8) -> (ReturnValue:Byte=0)
   ) -> (
    OutColorSet_Lighter_20_#5:Linear Color Structure
    OutColorSet_Light_21_#6:Linear Color Structure
    OutColorSet_Medium_22_#7:Linear Color Structure
    OutColorSet_Dark_23_#8:Linear Color Structure
    OutColorSet_Darker_24_#9:Linear Color Structure
    OutColorSet_Darker1_27_#10:Linear Color Structure
   )
   > SetBrushColor(self=Get HighlightBorder() -> (HighlightBorder:Border Object Reference), InBrushColor=@n9)
   > SetBrushTintColor!(
    execute
    self:Image Object Reference=[Get TopGradImage() -> (TopGradImage:Image Object Reference), Get bottomgradimage() -> (bottomgradimage:Image Object Reference)]
    TintColor:Slate Color Structure=Make Slate Color(SpecifiedColor=@n9.OutColorSet_Light_21_#6)
   ) -> (then)
   n10 = GetOwningPlayer()
   n11 = GetOwnershipTeamColor(
    self=Get CoreGamerDataAsset()
    InMinigame=GetMinigame(ContextActor=@n10)
    InLocalPlayerController=@n10
    Team="To Byte (Integer)"!(InInt:int=@n5.TeamIndex) -> (ReturnValue:Byte=0)
   )
   SetBrushColor(execute=@n7.else, self=Get HighlightBorder(), InBrushColor=@n11)
   > SetBrushTintColor(
    self=[Get bottomgradimage(), Get TopGradImage()]
    TintColor=Make Slate Color(SpecifiedColor=@n11.OutColorSet_Light_21_#6)
   )
   UpdateTextOnStateChange!(execute=@n6.then_1, self, bInIsLocalTeam:bool=@n5.IsLocalTeam, bInIsFirstPlace:bool=@n5.IsFirstPlace) -> (then)
   UpdateFirstPlaceEffect!(execute=@n6.then_2, self, InIsLocalTeam:bool=@n5.IsLocalTeam, InIsFirstPlace:bool=@n5.IsFirstPlace) -> (then)
   n12 = Event OnRowDataSet!() -> (OutputDelegate, then, TeamPlacement:int=0, TeamName:"Text (by ref)", TeamScore:int=0, bIsLocalTeam:bool, TeamIndex:int=0) {is=title K2Node_Event}
   > SetPlacement!(execute, self, InPlacement:int=@n12.TeamPlacement) -> (then) {is=title K2Node_CallFunction}
   > SetTeamName!(execute, self, TeamName:Text=@n12.TeamName) -> (then) {is=title K2Node_CallFunction}
   > SetScore!(execute, self, NewScore:int=@n12.TeamScore) -> (then) {is=title K2Node_CallFunction}
   > SetHighlightState!(
    execute
    self
    IsLocalTeam:bool=@n12.bIsLocalTeam
    IsFirstPlace:bool="Equal (Integer)"(A=@n12.TeamPlacement, B=1)
    TeamIndex:int=@n12.TeamIndex
   ) -> (then) {is=title K2Node_CallFunction}
   n13 = Event PreConstruct()
   > Branch(Condition=@n13.IsDesignTime)
   > SetHighlightState(
    IsLocalTeam=Get EditorPreviewIsLocalTeam() -> (EditorPreviewIsLocalTeam:bool)
    IsFirstPlace=Get EditorPreviewIsFirstPlace() -> (EditorPreviewIsFirstPlace:bool)
    TeamIndex=0
   )
  ]}
 UpdateFirstPlaceEffect {nodes [
   n1 = UpdateFirstPlaceEffect!() -> (then, InIsLocalTeam:bool=~, InIsFirstPlace:bool=~) {is=title K2Node_FunctionEntry}
   > n2 = Sequence!(execute) -> (then_0, then_1)
   n3 = AND Boolean(A=@n1, B=@n1.InIsFirstPlace)
   > Branch(Condition="Not Equal (Boolean)"(A=@n3, B=Get CachedLocalTeamInFirstPlace() -> (CachedLocalTeamInFirstPlace:bool)) -> (ReturnValue=~))
   > n4 = Set CachedLocalTeamInFirstPlace(CachedLocalTeamInFirstPlace:bool=@n3) -> (Output_Get:bool)
   > Branch(Condition=@n4)
   > PlayAnimation(InAnimation=Get Takeover() -> (Takeover:Widget Animation Object Reference))
   n5 = Branch(execute=@n2.then_1, Condition=@n1.InIsFirstPlace)
   > SetVisibility(self=[Get TopGradImage(), Get bottomgradimage()], InVisibility=HitTestInvisible)
   SetVisibility(execute=@n5.else, self=[Get TopGradImage(), Get bottomgradimage()], InVisibility=Hidden)
  ]}
 UpdateTextOnStateChange {nodes [
   n1 = UpdateTextOnStateChange!() -> (then, bInIsLocalTeam:bool=~, bInIsFirstPlace:bool=~) {is=title K2Node_FunctionEntry}
   > n2 = Branch(Condition=OR Boolean(A=@n1, B=@n1.bInIsFirstPlace))
   n3 = Get TeamNameText()
   n4 = Get ScoreText()
   n5 = Get PlacementText()
   > SetRenderOpacity(self=[@n3, @n4, @n5], InOpacity=1)
   n6 = Get TeamNameText()
   > Set Font!(execute, Font:Slate Font Info Structure=Get FontFamily-Bold() -> (FontFamily-Bold:Slate Font Info Structure), self:Text Object Reference=@n6) -> (then, Output_Get:Slate Font Info Structure)
   SetRenderOpacity(execute=@n2.else, self=[@n3, @n4, @n5], InOpacity=0.7)
   > Set Font(Font=Get FontFamily-Medium() -> (FontFamily-Medium:Slate Font Info Structure), self=@n6)
  ]}
]
variables [
 CachedLocalTeaminFirstPlace {type=boolean}
 CoreGamerDataAsset {type=PDA_UI_Creative_CoreGamer_C, default=/Game/Athena/HUD/LTM/Playground/CoreGamer/Data/DA_UI_Creative_CoreGamer}
 DefaultMainBrushColor {type=LinearColor, default=<rgba 1 1 1 0>}
 EditorPreviewIsFirstPlace {type=boolean}
 EditorPreviewIsLocalTeam {type=boolean}
 FirstPlaceMainBrushColor {type=LinearColor, default=<rgba 1 1 1 1>}
 FontFamily-Bold {
  type=SlateFontInfo
  default {
   bForceMonospaced=false
   bMaterialIsStencil=false
   fontMaterial=null
   fontObject=&BurbankSmall
   letterSpacing=0
   monospacedWidth=1
   size=14
   skewAmount=0
   typefaceFontName=Bold
   outlineSettings {
    bApplyOutlineToDropShadows=false
    bMiteredCorners=false
    bSeparateFillAlpha=false
    outlineMaterial=null
    outlineSize=0
    outlineColor=<rgba 0 0 0 1>
   }
  }
 }
 FontFamily-Medium {
  type=SlateFontInfo
  default {
   bForceMonospaced=false
   bMaterialIsStencil=false
   fontMaterial=null
   fontObject=&BurbankSmall
   letterSpacing=0
   monospacedWidth=1
   size=14
   skewAmount=0
   typefaceFontName=Medium
   outlineSettings {
    bApplyOutlineToDropShadows=false
    bMiteredCorners=false
    bSeparateFillAlpha=false
    outlineMaterial=null
    outlineSize=0
    outlineColor=<rgba 0 0 0 1>
   }
  }
 }
]
widgets {
 info {inheritedWidgetCount=0, parentClass=/Script/FortniteUI.FortCoreGamerScoreTrackerRow, rootWidgetClass=&Overlay, widgetCount=11}
 widgets [
  Overlay_25 {widgetClassPath=&Overlay, widget=Overlay_25}
  HighlightBorder {
   widgetClassPath=&Border
   bIsVariable=true
   parent=Overlay_25
   widget=HighlightBorder
   properties {
    background {
     bIsDynamicallyLoaded=false
     drawAs=Image
     imageType=NoImage
     mirroring=NoMirror
     resourceName=None
     resourceObject=/Game/Athena/HUD/LTM/Playground/CoreGamer/Materials/MI_UI_Creative_CoreGamer_ScoreboardHighlight
     tiling=NoTile
     imageSize=<32 32>
     margin=<ltrb 0 0 0 0>
     outlineSettings {
      bUseBrushTransparency=false
      roundingType=HalfHeightRadius
      width=0
      color {colorUseRule=UseColor_Specified, specifiedColor=<rgba 0 0 0 0>}
      cornerRadii=<0 0 0 0>
     }
     tintColor {colorUseRule=UseColor_Specified, specifiedColor=<rgba 1 1 1 1>}
     uVRegion {bIsValid=false, max=<0 0>, min=<0 0>}
    }
    brushColor=<rgba 1 1 1 0>
    padding=<ltrb 368 48 0 0>
   }
   slot {class=OverlaySlot, properties {horizontalAlignment=HAlign_Fill, verticalAlignment=VAlign_Fill}}
  }
  ShineImage {
   widgetClassPath=&Image
   bIsVariable=true
   parent=Overlay_25
   widget=ShineImage
   properties {
    visibility=Hidden
    brush {
     bIsDynamicallyLoaded=false
     drawAs=Image
     imageType=NoImage
     mirroring=NoMirror
     resourceName=None
     resourceObject=/Game/Athena/HUD/LTM/Playground/CoreGamer/Materials/MI_UI_Creative_CoreGamer_ScoreboardBracket
     tiling=NoTile
     imageSize=<64 32>
     margin=<ltrb 0 0 0 0>
     outlineSettings {
      bUseBrushTransparency=false
      roundingType=HalfHeightRadius
      width=0
      color {colorUseRule=UseColor_Specified, specifiedColor=<rgba 0 0 0 0>}
      cornerRadii=<0 0 0 0>
     }
     tintColor {colorUseRule=UseColor_Specified, specifiedColor=<rgba 1 1 1 1>}
     uVRegion {bIsValid=false, max=<0 0>, min=<0 0>}
    }
   }
   slot {class=OverlaySlot, properties {verticalAlignment=VAlign_Fill}}
  }
  HighlightBorder_Main {
   widgetClassPath=&Border
   bIsVariable=true
   parent=Overlay_25
   widget=HighlightBorder_Main
   properties {
    background {
     bIsDynamicallyLoaded=false
     drawAs=Image
     imageType=NoImage
     mirroring=NoMirror
     resourceName=None
     resourceObject=/Game/Athena/HUD/LTM/Playground/CoreGamer/Materials/MI_UI_Creative_CoreGamer_ScoreboardBkg
     tiling=NoTile
     imageSize=<368 48>
     margin=<ltrb 0.2 0.2 0.2 0.2>
     outlineSettings {
      bUseBrushTransparency=false
      roundingType=HalfHeightRadius
      width=0
      color {colorUseRule=UseColor_Specified, specifiedColor=<rgba 0 0 0 0>}
      cornerRadii=<0 0 0 0>
     }
     tintColor {colorUseRule=UseColor_Specified, specifiedColor=<rgba 1 1 1 1>}
     uVRegion {bIsValid=false, max=<0 0>, min=<0 0>}
    }
    padding=<ltrb 0 0 0 0>
   }
   slot {class=OverlaySlot, properties {horizontalAlignment=HAlign_Fill, verticalAlignment=VAlign_Fill}}
  }
  HorizontalBox_0 {
   widgetClassPath=&HorizontalBox
   parent=HighlightBorder_Main
   widget=HorizontalBox_0
   slot {class=BorderSlot, properties {padding=<ltrb 0 0 0 0>}}
  }
  PlacementText {
   widgetClassPath=&CommonTextBlock
   bIsVariable=true
   parent=HorizontalBox_0
   widget=PlacementText
   properties {
    renderOpacity=0.7
    text=42
    font {
     bForceMonospaced=false
     bMaterialIsStencil=false
     fontMaterial=null
     fontObject=&BurbankSmall
     letterSpacing=0
     monospacedWidth=1
     size=14
     skewAmount=0
     typefaceFontName=Black
     outlineSettings {
      bApplyOutlineToDropShadows=false
      bMiteredCorners=false
      bSeparateFillAlpha=false
      outlineMaterial=null
      outlineSize=0
      outlineColor=<rgba 0 0 0 1>
     }
    }
    shadowOffset=<0 0>
   }
   slot {class=HorizontalBoxSlot, properties {horizontalAlignment=HAlign_Left, verticalAlignment=VAlign_Center, padding=<ltrb 16 0 0 0>}}
  }
  SizeBox_0 {
   widgetClassPath=&SizeBox
   parent=HorizontalBox_0
   widget=SizeBox_0
   properties {bOverride_WidthOverride=true, widthOverride=135}
   slot {class=HorizontalBoxSlot, properties {horizontalAlignment=HAlign_Left, verticalAlignment=VAlign_Center, padding=<ltrb 30 0 0 0>}}
  }
  TeamNameText {
   widgetClassPath=&CommonTextBlock
   bIsVariable=true
   parent=SizeBox_0
   widget=TeamNameText
   properties {
    bIsScrollingEnabled=false
    clipping=OnDemand
    renderOpacity=0.7
    text=EOATXLIK_JHSTIH99VLTSV
    textOverflowPolicy=Ellipsis
    font {
     bForceMonospaced=false
     bMaterialIsStencil=false
     fontMaterial=null
     fontObject=&BurbankSmall
     letterSpacing=0
     monospacedWidth=1
     size=14
     skewAmount=0
     typefaceFontName=Medium
     outlineSettings {
      bApplyOutlineToDropShadows=false
      bMiteredCorners=false
      bSeparateFillAlpha=false
      outlineMaterial=null
      outlineSize=0
      outlineColor=<rgba 0 0 0 1>
     }
    }
    shadowOffset=<0 0>
   }
   slot {class=SizeBoxSlot}
  }
  ScoreText {
   widgetClassPath=&CommonTextBlock
   bIsVariable=true
   parent=HorizontalBox_0
   widget=ScoreText
   properties {
    renderOpacity=0.7
    text="999,999"
    font {
     bForceMonospaced=false
     bMaterialIsStencil=false
     fontMaterial=null
     fontObject=&BurbankSmall
     letterSpacing=0
     monospacedWidth=1
     size=14
     skewAmount=0
     typefaceFontName=Black
     outlineSettings {
      bApplyOutlineToDropShadows=false
      bMiteredCorners=false
      bSeparateFillAlpha=false
      outlineMaterial=null
      outlineSize=0
      outlineColor=<rgba 0 0 0 1>
     }
    }
    shadowOffset=<0 0>
   }
   slot {
    class=HorizontalBoxSlot
    properties {horizontalAlignment=HAlign_Right, verticalAlignment=VAlign_Center, padding=<ltrb 5 0 16 0>, size {sizeRule=Fill, value=1}}
   }
  }
  bottomgradimage {
   widgetClassPath=&Image
   bIsVariable=true
   parent=Overlay_25
   widget=bottomgradimage
   properties {brush {
     bIsDynamicallyLoaded=false
     drawAs=Image
     imageType=NoImage
     mirroring=NoMirror
     resourceName=None
     resourceObject=&MI_UI_Creative_CoreGamer_ScoreboardBracket
     tiling=NoTile
     imageSize=<32 1>
     margin=<ltrb 0 0 0 0>
     outlineSettings {
      bUseBrushTransparency=false
      roundingType=HalfHeightRadius
      width=0
      color {colorUseRule=UseColor_Specified, specifiedColor=<rgba 0 0 0 0>}
      cornerRadii=<0 0 0 0>
     }
     tintColor {colorUseRule=UseColor_Specified, specifiedColor=<rgba 0.0307 0.3515 1 1>}
     uVRegion {bIsValid=false, max=<0 0>, min=<0 0>}
    }}
   slot {class=OverlaySlot, properties {horizontalAlignment=HAlign_Fill, verticalAlignment=VAlign_Bottom}}
  }
  TopGradImage {
   widgetClassPath=&Image
   bIsVariable=true
   parent=Overlay_25
   widget=TopGradImage
   properties {brush {
     bIsDynamicallyLoaded=false
     drawAs=Image
     imageType=NoImage
     mirroring=NoMirror
     resourceName=None
     resourceObject=&MI_UI_Creative_CoreGamer_ScoreboardBracket
     tiling=NoTile
     imageSize=<32 1>
     margin=<ltrb 0 0 0 0>
     outlineSettings {
      bUseBrushTransparency=false
      roundingType=HalfHeightRadius
      width=0
      color {colorUseRule=UseColor_Specified, specifiedColor=<rgba 0 0 0 0>}
      cornerRadii=<0 0 0 0>
     }
     tintColor {colorUseRule=UseColor_Specified, specifiedColor=<rgba 0.0307 0.3515 1 1>}
     uVRegion {bIsValid=false, max=<0 0>, min=<0 0>}
    }}
   slot {class=OverlaySlot, properties {horizontalAlignment=HAlign_Fill}}
  }
 ]
}
