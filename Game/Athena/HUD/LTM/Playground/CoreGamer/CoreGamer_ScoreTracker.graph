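# HUD widget for Fortnite's Playground CoreGamer mode that shows team score tracker rows during a minigame.
# On construct it collapses itself and caches the pre-placed row widgets from TeamListContainer.
# The minigame in-progress event reveals it only when display is requested outside a curated hub, collapsing it without a valid minigame.
# Player stat and team changes pass the cached rows to the native parent's PopulateTeamRows.
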WidgetBlueprint /Game/Athena/HUD/LTM/Playground/CoreGamer/CoreGamer_ScoreTracker
graphs [EventGraph {nodes [
   n1 = Event OnDisplayMinigameInProgressUI!() -> (OutputDelegate, then, bDisplay:bool) {is=title K2Node_Event}
   > Parent: OnDisplayMinigameInProgressUI!(execute, bDisplay:bool=@n1.bDisplay) -> (then) {is=title K2Node_CallParentFunction}
   > n2 = Get!(execute:Exec) -> (then:Exec, else:Exec, CurrentMinigame:Fort Minigame Object Reference)
   n3 = NOT Boolean(A=IsCuratedHub!(self:Fort Minigame Object Reference=@n2) -> (ReturnValue:bool))
   > SetVisibility(
    self=Get Root() -> (Root:Overlay Object Reference)
    InVisibility=Select(Option 0=Collapsed, Option 1=SelfHitTestInvisible, Index=AND Boolean(A=@n3, B=@n1.bDisplay))
   )
   SetVisibility(execute=@n2.else, self=Get Root(), InVisibility=Collapsed)
   n4 = Event OnPlayerStatUpdated() {is=title K2Node_Event}
   Get(execute=[@n4.then, @n8.then])
   > PopulateTeamRows!(
    execute
    self:Fort Core Gamer Score Tracker Object Reference
    ScoreTrackerRows:Array of Fort Core Gamer Score Tracker Row Object References=Get ScoreTrackerRows() -> (ScoreTrackerRows:Array of Fort Core Gamer Score Tracker Row Object References)
   ) -> (then)
   Event Construct()
   > n5 = Sequence()
   n6 = For Each Loop(
    Exec=@n5.then_1
    Array:Array of Widget Object References=GetAllChildren(self=Get TeamListContainer() -> (TeamListContainer:Vertical Box Object Reference))
   ) -> (Array Element:Widget Object Reference)
   > n7 = Cast To FortCoreGamerScoreTrackerRow!(execute, Object=@n6) -> (then, CastFailed, AsFort Core Gamer Score Tracker Row:Fort Core Gamer Score Tracker Row Object Reference) {is=title K2Node_DynamicCast}
   > Add(
    TargetArray:Array of Fort Core Gamer Score Tracker Row Object References=Get ScoreTrackerRows()
    NewItem:"Fort Core Gamer Score Tracker Row Object Reference (by ref)"=@n7
   ) {is=title K2Node_CallArrayFunction}
   SetVisibility(execute=@n5.then_0, self=Get Root(), InVisibility=Collapsed)
   n8 = Event OnOwningPlayerChangedTeams() {is=title K2Node_Event}
   SetVisibility(
    self=Get Divider() -> (Divider:Size Box Object Reference)
    InVisibility=Select(Option 0=Collapsed, Option 1=SelfHitTestInvisible, Index=false)
   )
   Event OnSeparatorVisibilitySet!() -> (OutputDelegate, then, bVisible:bool) {is=title K2Node_Event}
  ]}]
variables [
 RoundWinsStatClass {type=Class, default=/Game/Athena/Playset/Minigames/StatFilters/MinigameStat_RoundWins.MinigameStat_RoundWins_C}
 RowClass {type=Class, default=&/CoreGamer_ScoreTracker_Row.CoreGamer_ScoreTracker_Row_C}
 ScoreStatClass {type=Class, default=/Game/Athena/Playset/Minigames/StatFilters/MinigameStat_Score.MinigameStat_Score_C}
 ScoreTrackerRows {type="array<FortCoreGamerScoreTrackerRow>", default []}
]
widgets {
 info {inheritedWidgetCount=1, parentClass=/Script/FortniteUI.FortCoreGamerScoreTracker, rootWidgetClass=&Overlay, widgetCount=9}
 widgets [
  Root {widgetClassPath=&Overlay, bIsVariable=true, widget=Root, properties {visibility=Collapsed}}
  SizeBox_4 {
   widgetClassPath=&SizeBox
   parent=Root
   widget=SizeBox_4
   properties {bOverride_WidthOverride=true, heightOverride=120, widthOverride=300}
   slot {class=OverlaySlot, properties {padding=<ltrb 0 20 0 0>}}
  }
  VerticalBox_0 {
   widgetClassPath=&VerticalBox
   parent=SizeBox_4
   widget=VerticalBox_0
   slot {class=SizeBoxSlot, properties {verticalAlignment=VAlign_Top}}
  }
  TeamListContainer {widgetClassPath=&VerticalBox, bIsVariable=true, parent=VerticalBox_0, widget=TeamListContainer, slot {class=VerticalBoxSlot}}
  CoreGamer_ScoreTracker_Row {
   widgetClassPath=&/CoreGamer_ScoreTracker_Row
   bIsVariable=true
   parent=TeamListContainer
   widget=CoreGamer_ScoreTracker_Row
   slot {class=VerticalBoxSlot}
  }
  CoreGamer_ScoreTracker_Row_56 {
   widgetClassPath=&CoreGamer_ScoreTracker_Row
   bIsVariable=true
   parent=TeamListContainer
   widget=CoreGamer_ScoreTracker_Row_56
   slot {class=VerticalBoxSlot}
  }
  CoreGamer_ScoreTracker_Row_256 {
   widgetClassPath=&CoreGamer_ScoreTracker_Row
   bIsVariable=true
   parent=TeamListContainer
   widget=CoreGamer_ScoreTracker_Row_256
   slot {class=VerticalBoxSlot}
  }
  CoreGamer_ScoreTracker_Row_1 {
   widgetClassPath=&CoreGamer_ScoreTracker_Row
   bIsVariable=true
   parent=TeamListContainer
   widget=CoreGamer_ScoreTracker_Row_1
   slot {class=VerticalBoxSlot}
  }
  TutorialHighlight {widgetClassPath=&TutorialHighlight}
 ]
}
