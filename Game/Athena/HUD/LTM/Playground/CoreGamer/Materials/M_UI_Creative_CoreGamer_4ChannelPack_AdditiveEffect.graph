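# Additive UI material for Creative CoreGamer HUD effects, animating four grayscale layers packed into one flare texture's RGBA channels.
# Each channel has its own rotation angle, spin speed, color and opacity.
# Tinted channels are layered by nested lerps over a transparency color, with A at the bottom and R on top, while channel alphas are summed and saturated into opacity.
# A time-driven sine provides adjustable-range flicker.
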Material /Game/Athena/HUD/LTM/Playground/CoreGamer/Materials/M_UI_Creative_CoreGamer_4ChannelPack_AdditiveEffect
graph {
 properties {blendMode=BLEND_Additive, materialDomain=MD_UI, pixelDepthOffsetMode=PDOM_Legacy, usedShadingModels=Default Lit}
 nodes [
  Max()
  Max()
  Max()
  n1 = Multiply(A=Time(), B=ScalarParameter() {group=0. General Settings, parameterName=Flicker Speed, sortPriority=1})
  MaterialExpressionRemap_1 = MaterialXRemap!(None=Sine(None=Frac(None=@n1)), InputLow, InputHigh, TargetLow, TargetHigh) {inputLowDefault=-1}
  Multiply(
   A=LinearInterpolate(
    A=ScalarParameter() {group=0. General Settings, parameterName=Fliker Range Min, sliderMax=1, sortPriority=2}
    B=ScalarParameter() {defaultValue=1, group=0. General Settings, parameterName=Flicker Range Max, sliderMax=1, sortPriority=3}
    Alpha=@MaterialExpressionRemap_1
   )
   B=NamedRerouteUsage()
  )
  n2 = ScalarParameter() {desc="Use negative values to inverse the rotation ", group=1. R Channel Settings, parameterName=R Spin Speed, sortPriority=1}
  n3 = Rotator(Time=Multiply(A=Time(), B=@n2))
  n4 = ScalarParameter() {group=1. R Channel Settings, parameterName=R Rotation Angle, sortPriority=0}
  NamedRerouteDeclaration(None=TextureSample(
    UVs=MaterialFunctionCall(UVs=@n3, "Rotation Angle (0-1)"=@n4) {materialFunction=&CustomRotator}.Rotated Values
    Tex=NamedRerouteUsage()
   ) {texture=/Game/Athena/HUD/Creative/Textures/T_UI_Creative_ChannelPack_FlareEffect}.R) {name=Channel R, nodeColor=<rgba 0 0 0 1>}
  n5 = ScalarParameter() {desc="Use negative values to inverse the rotation ", group=2. G Channel Settings, parameterName=G Spin Speed, sortPriority=1}
  n6 = Rotator(Time=Multiply(A=Time(), B=@n5))
  n7 = ScalarParameter() {group=2. G Channel Settings, parameterName=G Rotation Angle, sortPriority=0}
  NamedRerouteDeclaration(None=TextureSample(
    UVs=MaterialFunctionCall(UVs=@n6, "Rotation Angle (0-1)"=@n7) {materialFunction=&CustomRotator}.Rotated Values
    Tex=NamedRerouteUsage()
   ) {texture=&T_UI_Creative_ChannelPack_FlareEffect}.G) {name=Channel G, nodeColor=<rgba 0 0 0 1>}
  n8 = ScalarParameter() {desc="Use negative values to inverse the rotation ", group=3. B Channel Settings, parameterName=B Spin Speed, sortPriority=1}
  n9 = Rotator(Time=Multiply(A=Time(), B=@n8))
  n10 = ScalarParameter() {group=3. B Channel Settings, parameterName=B Rotation Angle, sortPriority=0}
  NamedRerouteDeclaration(None=TextureSample(
    UVs=MaterialFunctionCall(UVs=@n9, "Rotation Angle (0-1)"=@n10) {materialFunction=&CustomRotator}.Rotated Values
    Tex=NamedRerouteUsage()
   ) {texture=&T_UI_Creative_ChannelPack_FlareEffect}.B) {name=Channel B, nodeColor=<rgba 0 0 0 1>}
  n11 = ScalarParameter() {desc="Use negative values to inverse the rotation ", group=4. A Channel Settings, parameterName=A Spin Speed, sortPriority=1}
  n12 = Rotator(Time=Multiply(A=Time(), B=@n11))
  n13 = ScalarParameter() {group=4. A Channel Settings, parameterName=A Rotation Angle, sortPriority=0}
  NamedRerouteDeclaration(None=TextureSample(
    UVs=MaterialFunctionCall(UVs=@n12, "Rotation Angle (0-1)"=@n13) {materialFunction=&CustomRotator}.Rotated Values
    Tex=NamedRerouteUsage()
   ) {texture=&T_UI_Creative_ChannelPack_FlareEffect}.A) {name=Channel A, nodeColor=<rgba 0 0 0 1>}
  n14 = NamedRerouteUsage()
  n15 = VectorParameter() {group=4. A Channel Settings, parameterName=A Color, sortPriority=2, defaultValue=<rgba 1 1 1 1>}
  n16 = VectorParameter() {group=0. General Settings, parameterName=Color for Transparency, sortPriority=4, defaultValue=<rgba 0 0 0 1>}
  n17 = LinearInterpolate(
   A=LinearInterpolate(A=@n16.RGB, B=Multiply(A=@n14, B=@n15.RGB), Alpha=NamedRerouteUsage())
   B=Multiply(
    A=NamedRerouteUsage()
    B=VectorParameter() {group=3. B Channel Settings, parameterName=B Color, sortPriority=2, defaultValue=<rgba 1 1 1 1>}.RGB
   )
   Alpha=NamedRerouteUsage()
  )
  n18 = Multiply(
   A=NamedRerouteUsage()
   B=VectorParameter() {group=2. G Channel Settings, parameterName=G Color, sortPriority=2, defaultValue=<rgba 1 1 1 1>}.RGB
  )
  n19 = NamedRerouteUsage()
  n20 = NamedRerouteUsage()
  n21 = VectorParameter() {group=1. R Channel Settings, parameterName=R Color, sortPriority=2, defaultValue=<rgba 1 1 1 1>}
  NamedRerouteDeclaration(None=LinearInterpolate(A=LinearInterpolate(A=@n17, B=@n18, Alpha=@n19), B=Multiply(A=@n20, B=@n21.RGB), Alpha=NamedRerouteUsage())) {name=Colors, nodeColor=<rgba 0 0 0 1>}
  n22 = TextureObjectParameter() {
   group=0. General Settings
   parameterName=sRGB Texture
   sortPriority=0
   texture=&T_UI_Creative_ChannelPack_FlareEffect
  }
  NamedRerouteDeclaration(None=@n22) {name=Texture, nodeColor=<rgba 0 0 0 1>}
  n23 = Multiply(
   A=NamedRerouteUsage()
   B=ScalarParameter() {defaultValue=1, group=1. R Channel Settings, parameterName=R Opacity, sliderMax=1, sortPriority=3}
  )
  n24 = Multiply(
   A=NamedRerouteUsage()
   B=ScalarParameter() {defaultValue=1, group=2. G Channel Settings, parameterName=G Opacity, sliderMax=1, sortPriority=3}
  )
  n25 = NamedRerouteUsage()
  n26 = ScalarParameter() {defaultValue=1, group=3. B Channel Settings, parameterName=B Opacity, sliderMax=1, sortPriority=3}
  n27 = Add(
   A=Add(
    A=NamedRerouteDeclaration(None=@n23) {name=Channel R Alpha, nodeColor=<rgba 0 0 0 1>}
    B=NamedRerouteDeclaration(None=@n24) {name=Channel G Alpha, nodeColor=<rgba 0 0 0 1>}
   )
   B=NamedRerouteDeclaration(None=Multiply(A=@n25, B=@n26)) {name=Channel B Alpha, nodeColor=<rgba 0 0 0 1>}
  )
  n28 = NamedRerouteDeclaration(None=Multiply(
    A=NamedRerouteUsage()
    B=ScalarParameter() {defaultValue=1, group=4. A Channel Settings, parameterName=A Opacity, sliderMax=1, sortPriority=3}
   )) {name=Channel A Alpha, nodeColor=<rgba 0 0 0 1>}
  NamedRerouteDeclaration(None=Saturate(None=Add(A=@n27, B=@n28))) {name=Opacity, nodeColor=<rgba 0 0 0 1>}
  NamedRerouteUsage()
 ]
}
