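# Additive UI material for the Creative CoreGamer Playground HUD that draws three glowing rings.
# Each ring takes the Outline output of MF_UI_Circle, with its own size, thickness, glow range, color and opacity parameters.
# Named reroutes carry the rings into chained lerps that layer their tinted colors over black, which serves as transparent.
# Opacity-scaled ring masks are summed and saturated into a combined Opacity value.
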Material /Game/Athena/HUD/LTM/Playground/CoreGamer/Materials/M_UI_Creative_CoreGamer_AdditiveRingsEffect
graph {
 properties {blendMode=BLEND_Additive, materialDomain=MD_UI, pixelDepthOffsetMode=PDOM_Legacy, usedShadingModels=Default Lit}
 nodes [
  NamedRerouteDeclaration(None=MaterialFunctionCall(
    Size=ScalarParameter() {defaultValue=0.5, group=Ring 1, parameterName=Ring 1 Size, sliderMax=1, sortPriority=2}
    OutlineThickness=ScalarParameter() {group=Ring 1, parameterName=Ring 1 Thickness, sliderMax=0.1, sortPriority=3}
    GlowMin=ScalarParameter() {group=Ring 1, parameterName=Ring 1 Glow Min, sliderMax=1, sortPriority=4}
    GlowMax=ScalarParameter() {defaultValue=0.15, group=Ring 1, parameterName=Ring 1 Glow Max, sliderMax=1, sortPriority=5}
   ) {materialFunction=&MF_UI_Circle}.Outline) {name=Ring 1, nodeColor=<rgba 0 1 0.4609 1>}
  NamedRerouteDeclaration(None=MaterialFunctionCall(
    Size=ScalarParameter() {defaultValue=0.25, group=Ring 2, parameterName=Ring 2 Size, sliderMax=1, sortPriority=2}
    OutlineThickness=ScalarParameter() {group=Ring 2, parameterName=Ring 2 Thickness, sliderMax=0.1, sortPriority=3}
    GlowMin=ScalarParameter() {group=Ring 2, parameterName=Ring 2 Glow Min, sliderMax=1, sortPriority=4}
    GlowMax=ScalarParameter() {defaultValue=0.065, group=Ring 2, parameterName=Ring 2 Glow Max, sliderMax=1, sortPriority=5}
   ) {materialFunction=&MF_UI_Circle}.Outline) {name=Ring 2, nodeColor=<rgba 0 1 0.4609 1>}
  NamedRerouteDeclaration(None=MaterialFunctionCall(
    Size=ScalarParameter() {defaultValue=0.125, group=Ring 3, parameterName=Ring 3 Size, sliderMax=1, sortPriority=2}
    OutlineThickness=ScalarParameter() {group=Ring 3, parameterName=Ring 3 Thickness, sliderMax=0.1, sortPriority=3}
    GlowMin=ScalarParameter() {group=Ring 3, parameterName=Ring 3 Glow Min, sliderMax=1, sortPriority=4}
    GlowMax=ScalarParameter() {defaultValue=0.01, group=Ring 3, parameterName=Ring 3 Glow Max, sliderMax=1, sortPriority=5}
   ) {materialFunction=&MF_UI_Circle}.Outline) {name=Ring 3, nodeColor=<rgba 0 1 0.4609 1>}
  n1 = Constant4Vector() {desc=Color for Transparency, constant=<rgba 0 0 0 1>}
  n2 = LinearInterpolate(
   A=MaterialFunctionCall(X=@n1.R, Y=@n1.R, Z=@n1.R) {materialFunction=&MakeFloat3}.Result
   B=Multiply(A=NamedRerouteUsage(), B=VectorParameter() {group=Ring 1, parameterName=Ring 1 Color, sortPriority=0, defaultValue=<rgba 1 1 1 1>}.RGB)
   Alpha=NamedRerouteUsage()
  )
  n3 = Multiply(A=NamedRerouteUsage(), B=VectorParameter() {group=Ring 2, parameterName=Ring 2 Color, sortPriority=0, defaultValue=<rgba 1 1 1 1>}.RGB)
  n4 = NamedRerouteUsage()
  n5 = NamedRerouteUsage()
  n6 = VectorParameter() {group=Ring 3, parameterName=Ring 3 Color, sortPriority=0, defaultValue=<rgba 1 1 1 1>}
  NamedRerouteDeclaration(None=LinearInterpolate(A=LinearInterpolate(A=@n2, B=@n3, Alpha=@n4), B=Multiply(A=@n5, B=@n6.RGB), Alpha=NamedRerouteUsage())) {name=Colors, nodeColor=<rgba 0 0 0 1>}
  n7 = NamedRerouteUsage()
  n8 = ScalarParameter() {defaultValue=1, group=Ring 1, parameterName=RIng 1 Opacity, sliderMax=1, sortPriority=1}
  n9 = NamedRerouteUsage()
  n10 = ScalarParameter() {defaultValue=1, group=Ring 2, parameterName=Ring 2 Opacity, sliderMax=1, sortPriority=1}
  n11 = Add(
   A=NamedRerouteDeclaration(None=Multiply(A=@n7, B=@n8)) {name=Ring 1 Alpha, nodeColor=<rgba 0 0 0 1>}
   B=NamedRerouteDeclaration(None=Multiply(A=@n9, B=@n10)) {name=Ring 2 Alpha, nodeColor=<rgba 0 0 0 1>}
  )
  n12 = NamedRerouteDeclaration(None=Multiply(A=NamedRerouteUsage(), B=ScalarParameter() {defaultValue=1, group=Ring 3, parameterName=Ring 3 Opacity, sliderMax=1, sortPriority=1})) {name=Ring 3 Alpha, nodeColor=<rgba 0 0 0 1>}
  NamedRerouteDeclaration(None=Saturate(None=Add(A=@n11, B=@n12))) {name=Opacity, nodeColor=<rgba 0 0 0 1>}
  NamedRerouteUsage()
  NamedRerouteUsage()
 ]
}
