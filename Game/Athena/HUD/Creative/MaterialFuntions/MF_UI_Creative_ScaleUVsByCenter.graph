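# Material function for Creative HUD materials that scales UVs around their center.
# It subtracts 0.5 from the input UVs, divides by a two-component Scale and adds 0.5 back, so larger factors enlarge the sampled image.
# Both inputs use their previews as defaults: texture coordinates for UVs and a constant vector for Scale.
# Scale is not clamped, and its description requires values above zero to avoid division by zero.
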MaterialFunction /Game/Athena/HUD/Creative/MaterialFuntions/MF_UI_Creative_ScaleUVsByCenter
graph {
 properties {description=Scale The UVs by the center}
 nodes [
  n1 = Subtract(A=FunctionInput(Preview=TextureCoordinate()) {bUsePreviewValueAsDefault=true, description=The Material UVs, inputName=UVs, inputType=FunctionInput_Vector2}) {constB=0.5}
  n2 = FunctionInput(Preview=Constant2Vector() {g=0.5, r=0.5}) {
   bUsePreviewValueAsDefault=true
   description="The Scale Factor. To avoid a divide by 0, this factor must always be higher than 0 (e.q. 0.001) "
   inputName=Scale
   inputType=FunctionInput_Vector2
   sortPriority=1
   previewValue=<1 1 0 1>
  }
  FunctionOutput(None=Add(A=Divide(A=@n1, B=@n2) {constB=0.7}) {constB=0.5}) {outputName=Scaled UVs}
 ]
}
