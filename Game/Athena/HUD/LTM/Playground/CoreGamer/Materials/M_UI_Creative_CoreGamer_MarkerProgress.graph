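# Translucent UI material for the Creative Core Gamer HUD marker: a scalable, rotatable square with a progress border, inner box and icon.
# Nested box masks, subtracted from each other, form the progress band, its background, a gap and the gradient-filled center.
# Percentage against a square or radial angle gradient masks the fill, with start point and static clockwise switch.
# Layers are lerped for color and combined by opacity-weighted masks for alpha.
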Material /Game/Athena/HUD/LTM/Playground/CoreGamer/Materials/M_UI_Creative_CoreGamer_MarkerProgress
graph {
 properties {blendMode=BLEND_Translucent, materialDomain=MD_UI, pixelDepthOffsetMode=PDOM_Legacy, usedShadingModels=Default Lit}
 nodes [
  n1 = NamedRerouteUsage()
  n2 = NamedRerouteUsage()
  Add(A=@n1, B=@n2)
  n3 = Subtract(A=@n1, B=@n2)
  Floor(None=@n3)
  MaterialFunctionCall!(UVs=NamedRerouteUsage(), Rotation) {materialFunction=/FortUILibrary/MaterialFunctions/Gradients/MF_UI_AngleGradient}
  n4 = ScalarParameter() {
   defaultValue=0.5
   desc="Marker scale factor. Warning: Because of a divide, this value MUST not be 0 !"
   group=1. Sizing
   parameterName=Base Scale
   sliderMax=1
   sliderMin=0.001
   sortPriority=0
  }
  NamedRerouteDeclaration(None=@n4) {name=Base Scale, nodeColor=<rgba 0 0 0 1>}
  n5 = NamedRerouteUsage()
  NamedRerouteDeclaration(None=Multiply(A=@n5, B=NamedRerouteUsage())) {name=Progress, nodeColor=<rgba 0 0 0 1>}
  n6 = NamedRerouteUsage()
  n7 = NamedRerouteUsage()
  n8 = Multiply(A=@n7, B=NamedRerouteUsage()) {constB=0.9333, desc="Progress Background \r\nScale"}
  n9 = Multiply(A=NamedRerouteUsage(), B=NamedRerouteUsage()) {constB=0.85, desc=Blank Space Scale}
  n10 = MaterialFunctionCall(UVs=NamedRerouteUsage(), Size=@n9) {materialFunction=&MF_UI_Box}
  NamedRerouteDeclaration(None=Subtract(A=MaterialFunctionCall(UVs=@n6, Size=@n8) {materialFunction=&MF_UI_Box}.Fill, B=@n10.Fill)) {name=Progress bg Shape, nodeColor=<rgba 0 0 0 1>}
  n11 = OneMinus(None=ScalarParameter() {defaultValue=0.1, group=1. Sizing, parameterName=Progress Thickness, sliderMax=1, sortPriority=3})
  NamedRerouteDeclaration(None=LinearInterpolate(
    A=NamedRerouteDeclaration(None=@n11) {name=Progress Thickness, nodeColor=<rgba 0 0 0 1>}
    Alpha=ScalarParameter() {defaultValue=0.5, group=1. Sizing, parameterName=Progress Background Thickness, sliderMax=1, sortPriority=4}
   )) {name=Progress Bg Thickness, nodeColor=<rgba 0 0 0 1>}
  n12 = VectorParameter() {group=3. Backgrounds, parameterName=Color for Transparency, sortPriority=4, defaultValue=<rgba 0 0 0 1>}
  n13 = VectorParameter() {group=3. Backgrounds, parameterName=Progress Background Color, sortPriority=2, defaultValue=<rgba 0 0.0177 0.0885 1>}
  n14 = NamedRerouteUsage()
  n15 = LinearInterpolate(
   A=LinearInterpolate(A=LinearInterpolate(A=@n12.RGB, B=@n13.RGB, Alpha=@n14), B=NamedRerouteUsage(), Alpha=NamedRerouteUsage())
   B=VectorParameter() {group=3. Backgrounds, parameterName=Background Color, sortPriority=0, defaultValue=<rgba 1 1 1 1>}.RGB
   Alpha=NamedRerouteUsage()
  )
  n16 = NamedRerouteUsage()
  n17 = NamedRerouteUsage()
  NamedRerouteDeclaration(None=LinearInterpolate(
    A=LinearInterpolate(A=@n15, B=@n16, Alpha=@n17)
    B=VectorParameter() {group=4. Icon, parameterName=Icon Color, sortPriority=3, defaultValue=<rgba 1 1 1 1>}.RGB
    Alpha=NamedRerouteUsage()
   ) {constAlpha=0}) {name=Final Colored Shape, nodeColor=<rgba 0 0 0 1>}
  n18 = NamedRerouteDeclaration(None=Multiply(A=NamedRerouteUsage(), B=ScalarParameter() {defaultValue=1, group=2. Progress, parameterName=Progress Opacity, sliderMax=1, sortPriority=4})) {name=Progress Shape with Alpha, nodeColor=<rgba 0 0 0 1>}
  n19 = NamedRerouteDeclaration(None=Multiply(
    A=NamedRerouteUsage()
    B=ScalarParameter() {defaultValue=1, group=3. Backgrounds, parameterName=Progress Background Opacity, sliderMax=1, sortPriority=3}
   )) {name=Progress Background Shape with Alpha, nodeColor=<rgba 0 0 0 1>}
  n20 = Multiply(
   A=NamedRerouteUsage()
   B=ScalarParameter() {defaultValue=1, group=3. Backgrounds, parameterName=Background Opacity, sliderMax=1, sortPriority=1}
  )
  n21 = NamedRerouteUsage()
  n22 = ScalarParameter() {defaultValue=1, group=3. Inner Box, parameterName=Inner Box Opacity, sliderMax=1, sortPriority=0}
  n23 = Max(
   A=Max(A=Max(A=@n18, B=@n19), B=NamedRerouteDeclaration(None=@n20) {name=Blank Space Shape with Alpha, nodeColor=<rgba 0 0 0 1>})
   B=NamedRerouteDeclaration(None=Multiply(A=@n21, B=@n22)) {name=Inner Box Shape with Alpha, nodeColor=<rgba 0 0 0 1>}
  )
  n24 = NamedRerouteDeclaration(None=Multiply(A=NamedRerouteUsage(), B=ScalarParameter() {defaultValue=1, group=4. Icon, parameterName=Icon Opacity, sliderMax=1, sortPriority=4})) {name=Icon Shape with Alpha, nodeColor=<rgba 0 0 0 1>}
  NamedRerouteDeclaration(None=Saturate(None=Add(A=@n23, B=@n24))) {name=Final Alpha, nodeColor=<rgba 0 0 0 1>}
  NamedRerouteDeclaration(None=ScalarParameter() {defaultValue=0.8, group=1. Sizing, parameterName=Inner Square Scale, sliderMax=1, sortPriority=2}) {name=Inner Square Scale, nodeColor=<rgba 0 0 0 1>}
  n25 = NamedRerouteUsage()
  NamedRerouteDeclaration(None=Subtract(A=MaterialFunctionCall(UVs=@n25, Size=@n7) {materialFunction=&MF_UI_Box}.Fill, B=@n10.Fill)) {name=Progress Shape, nodeColor=<rgba 0 0 0 1>}
  NamedRerouteDeclaration(None=ScalarParameter() {defaultValue=0.7, group=2. Progress, parameterName=Percentage, sliderMax=1, sortPriority=1}) {name=Progress Percentage, nodeColor=<rgba 0 0 0 1>}
  NamedRerouteDeclaration(None=NamedRerouteUsage()) {name=Progress Background, nodeColor=<rgba 0 0 0 1>}
  NamedRerouteDeclaration(None=ScalarParameter() {group=0. Misc., parameterName=Rotation Angle, sliderMax=1, sortPriority=0}) {name=Angle, nodeColor=<rgba 0 0 0 1>}
  n26 = VectorParameter() {group=3. Inner Box, parameterName=Inner Box Color 1, sortPriority=1, defaultValue=<rgba 0.0003 0.1301 0.7011 1>}
  n27 = VectorParameter() {group=3. Inner Box, parameterName=Inner Box Color 2, sortPriority=2, defaultValue=<rgba 0 0.0545 0.2747 1>}
  n28 = MaterialFunctionCall(
   GradientStart=Constant() {r=0.25}
   GradientEnd=Constant() {r=0.75}
   RotationAngle=ScalarParameter() {defaultValue=0.25, group=3. Inner Box, parameterName=Inner Box Gradient Rotation, sliderMax=1, sortPriority=3}
  ) {materialFunction=&MF_UI_LinearGradient}
  NamedRerouteDeclaration(None=Saturate(None=LinearInterpolate(A=@n26.RGB, B=@n27.RGB, Alpha=@n28.Result))) {name=Inner Box Gradient, nodeColor=<rgba 0 0 0 1>}
  n29 = VectorParameter() {group=2. Progress, parameterName=Progress Color 1, sortPriority=5, defaultValue=<rgba 0.0561 0.6514 1 1>}
  n30 = VectorParameter() {group=2. Progress, parameterName=Progress Color 2, sortPriority=6, defaultValue=<rgba 0.0097 0.2705 0.8879 1>}
  n31 = MaterialFunctionCall(
   GradientStart=Constant() {r=0.25}
   GradientEnd=Constant() {r=0.75}
   RotationAngle=ScalarParameter() {defaultValue=0.125, group=2. Progress, parameterName=Progress Gradient Rotation, sliderMax=1, sortPriority=7}
  ) {materialFunction=&MF_UI_LinearGradient}
  NamedRerouteDeclaration(None=Saturate(None=LinearInterpolate(A=@n29.RGB, B=@n30.RGB, Alpha=@n31.Result))) {name=Progress Gradient, nodeColor=<rgba 0 0 0 1>}
  NamedRerouteDeclaration(None=StaticBoolParameter() {defaultValue=true, group=2. Progress, parameterName=Clockwise, sortPriority=0}) {name=Clockwise, nodeColor=<rgba 0 0 0 1>}
  n32 = ScalarParameter() {
   defaultValue=0.5
   desc="Icon scale factor. Warning: Because of a divide, this value MUST not be 0 !"
   group=4. Icon
   parameterName=Icon Scale
   sliderMax=2
   sliderMin=0.001
   sortPriority=1
  }
  NamedRerouteDeclaration(None=Multiply(A=NamedRerouteUsage(), B=@n32)) {name=Icon Scale, nodeColor=<rgba 0 0 0 1>}
  NamedRerouteDeclaration(None=MaterialFunctionCall(UVs=NamedRerouteUsage(), "Rotation Angle (0-1)"=NamedRerouteUsage()) {materialFunction=&CustomRotator}.Rotated Values) {name=Rotated UVs, nodeColor=<rgba 0 0 0 1>}
  n33 = ScalarParameter() {desc="0 for Square progress, 1 for Radial progress", group=2. Progress, parameterName=Use Radial progress, sliderMax=1, sortPriority=2}
  NamedRerouteDeclaration(None=Ceil(None=@n33)) {name=Radial Progress, nodeColor=<rgba 0 0 0 1>}
  NamedRerouteDeclaration(None=MaterialFunctionCall(UVs=NamedRerouteUsage(), "Rotation Angle (0-1)"=NamedRerouteUsage()) {materialFunction=&CustomRotator}.Rotated Values) {name=Icon Rotated UVs, nodeColor=<rgba 0 0 0 1>}
  n34 = TextureSampleParameter2D(UVs=MaterialFunctionCall!(UVs=NamedRerouteUsage(), Scale=NamedRerouteUsage()) {materialFunction=/Game/Athena/HUD/Creative/MaterialFuntions/MF_UI_Creative_ScaleUVsByCenter}.Scaled UVs) {
   group=4. Icon
   parameterName=Icon Texture
   samplerSource=SSM_Clamp_WorldGroupSettings
   sortPriority=2
   texture=/Game/Athena/UI/InGame/Creative/PropertyEditor/Icon_Picker/Textures/64x64/T_UI_IconLibrary_A_64
  }
  NamedRerouteDeclaration(None=@n34.A) {name=Icon Alpha Channel, nodeColor=<rgba 0 0 0 1>}
  NamedRerouteDeclaration(None=ScalarParameter() {group=4. Icon, parameterName=Icon Rotation Angle, sliderMax=1, sortPriority=0}) {name=Icon Angle, nodeColor=<rgba 0 0 0 1>}
  NamedRerouteDeclaration(None=MaterialFunctionCall(UVs=NamedRerouteUsage(), Scale=NamedRerouteUsage()) {materialFunction=&MF_UI_Creative_ScaleUVsByCenter}.Scaled UVs) {name=UVs Scaled Depending Base Scale, nodeColor=<rgba 0 0 0 1>}
  NamedRerouteDeclaration(None=MaterialFunctionCall() {materialFunction=&GetUserInterfaceUV}.Normalized UV) {name=Normalized UVs, nodeColor=<rgba 0 0 0 1>}
  n35 = MaterialFunctionCall(UVs=NamedRerouteUsage(), Size=Multiply(A=@n9, B=NamedRerouteUsage()) {constB=0.8, desc=Box Scale}) {materialFunction=&MF_UI_Box}
  NamedRerouteDeclaration(None=Subtract(A=@n10.Fill, B=@n35.Fill)) {name=Blank Space, nodeColor=<rgba 0 0 0 1>}
  NamedRerouteDeclaration(None=@n35.Fill) {name=Inner Box, nodeColor=<rgba 0 0 0 1>}
  NamedRerouteDeclaration(None=ScalarParameter() {defaultValue=0.875, group=2. Progress, parameterName=Start Point, sliderMax=1, sortPriority=3}) {name=Start Point, nodeColor=<rgba 0 0 0 1>}
  n36 = MaterialFunctionCall(UVs=NamedRerouteUsage(), "Rotation Angle (0-1)"=Add(A=NamedRerouteUsage()) {constB=0.125}) {materialFunction=&CustomRotator}
  n37 = LinearInterpolate(
   A=MaterialFunctionCall!(UVs=NamedRerouteUsage(), Start Point=NamedRerouteUsage()) {materialFunction=/FortUILibrary/MaterialFunctions/Gradients/MF_UI_SquareGradient}.Gradient
   B=OneMinus(None=MaterialFunctionCall!(Vector or UVs=@n36.Rotated Values) {materialFunction=/FortUILibrary/MaterialFunctions/Basics/MF_UI_CheaperVectorToRadial}.Vector Converted to Angle)
   Alpha=NamedRerouteUsage()
  )
  n38 = Subtract(A=NamedRerouteUsage(), B=StaticSwitch(True=@n37, False=OneMinus(None=@n37), Value=NamedRerouteUsage()))
  NamedRerouteDeclaration(None=Ceil(None=Saturate(None=@n38))) {name=Progress Fill Mask, nodeColor=<rgba 0 0 0 1>}
  NamedRerouteUsage()
  NamedRerouteUsage()
  NamedRerouteUsage()
  NamedRerouteUsage()
  Reroute() {desc="the UVs applied to the Square gradient must be scaled with the base marker \r\nto keep the gradient maximum value visible on the mask edges "}
  Saturate(None=Add(A=@n18, B=@n19))
  StaticSwitch(True=@n3, False=NamedRerouteUsage())
  Subtract(A=NamedRerouteUsage(), B=@n5) {desc="Use this if the background must be hidden \r\nwhen the progress is transparent"}
 ]
}
