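# Translucent UI material that draws a skewed progress bar with an end cap and optional animated pulse, for the Creative CoreGamer playground HUD.
# Dot products on optionally mirrored UVs create the slanted edges, with pixel-size aspect ratios adjusting the skew.
# Max-combined gradients, sharpened by SmoothStep, form background, progress and end-cap masks that feed layered color and alpha.
# A Time-driven Frac pulse supports inverted direction and a static switch for progress-dependent speed.
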Material /Game/Athena/HUD/LTM/Playground/CoreGamer/Materials/M_UI_Creative_CoreGamer_ProgressBar_Skewed
graph {
 properties {blendMode=BLEND_Translucent, materialDomain=MD_UI, pixelDepthOffsetMode=PDOM_Legacy, usedShadingModels=Default Lit}
 nodes [
  n1 = NamedRerouteUsage()
  n2 = NamedRerouteUsage()
  n3 = DotProduct(A=OneMinus(None=@n1), B=@n2) {desc="The dot product gets you the angle between two vectors. \r\nbasically it'll give you a -1 -1 value :\r\nOr if the two vectors are parallel Or how perpendicular they are.\r\nAnd we are using that to get the distance between one edge (vector A) \r\nand the other edge (1) to use to create an angle."}
  Floor(None=@n3) {desc="Preview \r\nthe angle"}
  n4 = MaterialFunctionCall() {materialFunction=&GetUserInterfaceUV}
  NamedRerouteDeclaration(None=@n4.Normalized UV) {name=Normalized UVs, nodeColor=<rgba 0 0 0 1>}
  n5 = NamedRerouteUsage()
  n6 = ScalarParameter() {defaultValue=1, group=3. Pulse Settings, parameterName=Pulse Length, sliderMax=1, sliderMin=0.01, sortPriority=1}
  n7 = Time()
  n8 = ScalarParameter() {defaultValue=1, group=3. Pulse Settings, parameterName=Progress Speed Multiplier, sortPriority=5}
  n9 = ScalarParameter() {defaultValue=0.2213, group=1. Global Settings, parameterName=Progress, sliderMax=1, sortPriority=3}
  n10 = ScalarParameter() {defaultValue=0.5, group=3. Pulse Settings, parameterName=Pulse Speed}
  n11 = StaticBoolParameter() {
   desc=Allow to increase or decrease the speed depending the current progress value.
   group=3. Pulse Settings
   parameterName=Change Speed Depending Progress
   sortPriority=4
  }
  n12 = StaticSwitch(True=Add(A=LinearInterpolate(A=@n8, Alpha=@n9) {constA=1, constB=0}, B=@n10), False=@n10, Value=@n11)
  n13 = NamedRerouteUsage()
  n14 = OneMinus(None=@n13)
  n15 = ScalarParameter() {desc="0= Normal   /   1= Inverted", group=3. Pulse Settings, parameterName=Invert Pulse Direction, sliderMax=1, sortPriority=0}
  n16 = Saturate(None=Ceil(None=Abs(None=@n15)))
  n17 = Multiply(
   A=ScalarParameter() {defaultValue=1, group=3. Pulse Settings, parameterName=Pulse Period, sortPriority=2}
   B=Add(A=Multiply(A=@n12, B=@n7), B=LinearInterpolate(A=@n13, B=@n14, Alpha=@n16))
  )
  n18 = SmoothStep(Min=OneMinus(None=@n6), Value=OneMinus(None=Frac(None=@n17))) {constMin=0.9}
  n19 = Ceil(None=Abs(None=ScalarParameter() {desc="0= Inactive   /   1= Active", group=3. Pulse Settings, parameterName=Active Pulse, sliderMax=1, sortPriority=0}))
  NamedRerouteDeclaration(None=LinearInterpolate(B=DotProduct(A=@n5, B=@n18), Alpha=Saturate(None=@n19)) {constB=0}) {name=Pulse, nodeColor=<rgba 0 0 0 1>}
  n20 = NamedRerouteUsage()
  n21 = ScalarParameter() {defaultValue=1, group=2. Colors, parameterName=Background Opacity, sliderMax=1, sortPriority=1}
  n22 = NamedRerouteUsage()
  n23 = ScalarParameter() {defaultValue=1, group=2. Colors, parameterName=Progress Opacity, sliderMax=1, sortPriority=6}
  n24 = Add(
   A=NamedRerouteDeclaration(None=Multiply(A=@n20, B=@n21)) {name=Background Alpha, nodeColor=<rgba 0 0 0 1>}
   B=NamedRerouteDeclaration(None=Multiply(A=@n22, B=@n23)) {name=Progress Alpha, nodeColor=<rgba 0 0 0 1>}
  )
  n25 = Multiply(A=NamedRerouteUsage(), B=ScalarParameter() {defaultValue=1, group=2. Colors, parameterName=Pulse Opacity, sortPriority=8})
  n26 = Saturate(None=Add(A=Saturate(None=@n24), B=NamedRerouteDeclaration(None=@n25) {name=Pulse Alpha, nodeColor=<rgba 0 0 0 1>}))
  n27 = NamedRerouteDeclaration(None=Multiply(A=NamedRerouteUsage(), B=ScalarParameter() {defaultValue=1, group=2. Colors, parameterName=End Cap Opacity, sliderMax=1, sortPriority=3})) {name=End Cap Alpha, nodeColor=<rgba 0 0 0 1>}
  NamedRerouteDeclaration(None=Saturate(None=Add(A=@n26, B=@n27))) {name=Final Alpha, nodeColor=<rgba 0 0 0 1>}
  n28 = NamedRerouteUsage()
  NamedRerouteDeclaration(None=Multiply(A=@n28, B=NamedRerouteUsage())) {name=Skew for Horizontal Ratio, nodeColor=<rgba 0 0 0 1>}
  n29 = NamedRerouteUsage()
  n30 = ScalarParameter() {defaultValue=0.2213, group=1. Global Settings, parameterName=Progress, sliderMax=1, sortPriority=3}
  NamedRerouteDeclaration(None=OneMinus(None=LinearInterpolate(A=@n29, Alpha=@n30) {constA=1})) {name=Progress Value, nodeColor=<rgba 0 0 0 1>}
  n31 = NamedRerouteUsage()
  n32 = ScalarParameter() {defaultValue=0.04, group=1. Global Settings, parameterName=End Cap Width, sliderMax=1, sortPriority=4}
  NamedRerouteDeclaration(None=OneMinus(None=LinearInterpolate(A=@n31, Alpha=@n32) {constA=1})) {name=End Cap Value, nodeColor=<rgba 0 0 0 1>}
  n33 = Floor(None=NamedRerouteUsage() {desc="provide 0 under 0.5\r\nprivide 1 for everthing\r\nover 0.5"})
  NamedRerouteDeclaration(None=LinearInterpolate(A=NamedRerouteUsage(), B=NamedRerouteUsage(), Alpha=Saturate(None=@n33))) {name=Final Skew, nodeColor=<rgba 0 0 0 1>}
  n34 = NamedRerouteUsage()
  n35 = Subtract(A=Add(A=NamedRerouteUsage(), B=NamedRerouteUsage()) {constB=0.7}, B=@n34)
  n36 = Add(A=@n34, B=NamedRerouteUsage())
  NamedRerouteDeclaration(None=OneMinus(None=Max(A=@n35, B=@n36))) {name=End Cap Gradient, nodeColor=<rgba 0 0 0 1>}
  n37 = NamedRerouteUsage()
  NamedRerouteDeclaration(None=Divide(A=ComponentMask(None=@n37) {r=true}, B=ComponentMask(None=@n37) {g=true})) {name=Pixel Size Horizontal Ratio, nodeColor=<rgba 0 0 0 1>}
  NamedRerouteDeclaration(None=ScalarParameter() {defaultValue=0.99, group=1. Global Settings, parameterName=Skew Amount, sliderMax=0.99, sortPriority=0}) {name=Skew Amount Value, nodeColor=<rgba 0 0 0 1>}
  n38 = NamedRerouteUsage()
  NamedRerouteDeclaration(None=Divide(A=ComponentMask(None=@n38) {g=true}, B=ComponentMask(None=@n38) {r=true})) {name=Pixel Size Vertical Ratio, nodeColor=<rgba 0 0 0 1>}
  n39 = OneMinus(None=Max(A=NamedRerouteUsage(), B=NamedRerouteUsage()))
  NamedRerouteDeclaration(None=Subtract(A=Multiply(A=@n39) {constB=2}, B=NamedRerouteUsage()) {constA=0.408, constB=0.001}) {name=Progress Gradient Mirrored, nodeColor=<rgba 0 0 0 1>}
  n40 = NamedRerouteUsage()
  n41 = ComponentMask(None=@n40) {r=true}
  NamedRerouteDeclaration(None=OneMinus(None=@n41) {desc=Mirror X}) {name=Mirror X, nodeColor=<rgba 0 0 0 1>}
  n42 = ComponentMask(None=@n40) {g=true}
  NamedRerouteDeclaration(None=OneMinus(None=@n42) {desc=Mirror Y}) {name=Mirror Y, nodeColor=<rgba 0 0 0 1>}
  n43 = NamedRerouteUsage()
  n44 = NamedRerouteUsage()
  n45 = ScalarParameter() {desc="0= Normal X   /   1= Mirror X", group=1. Global Settings, parameterName=Mirror X, sliderMax=1, sortPriority=1}
  n46 = Saturate(None=Ceil(None=Abs(None=@n45)))
  n47 = NamedRerouteUsage()
  n48 = NamedRerouteUsage()
  n49 = ScalarParameter() {desc="0= Normal Y   /   1= Mirror Y", group=1. Global Settings, parameterName=Mirror Y, sliderMax=1, sortPriority=2}
  n50 = Saturate(None=Ceil(None=Abs(None=@n49)))
  NamedRerouteDeclaration(None=AppendVector(
    A=LinearInterpolate(A=@n43, B=@n44, Alpha=@n46) {desc="0: Normal X\r\n1: Mirror X"}
    B=LinearInterpolate(A=@n47, B=@n48, Alpha=@n50) {desc="0: Normal Y\r\n1: Mirror Y"}
   )) {name=Normal or Mirror UVs, nodeColor=<rgba 0 0 0 1>}
  NamedRerouteDeclaration(None=@n41) {name=Normal X, nodeColor=<rgba 0 0 0 1>}
  NamedRerouteDeclaration(None=@n4.Pixel Size) {name=Pixel Size, nodeColor=<rgba 0 0 0 1>}
  NamedRerouteDeclaration(None=@n42) {name=Normal Y, nodeColor=<rgba 0 0 0 1>}
  NamedRerouteDeclaration(None=AppendVector(A=Constant() {r=1}, B=NamedRerouteUsage())) {name=Skew Y Vector, nodeColor=<rgba 0 0 0 1>}
  n51 = NamedRerouteUsage()
  n52 = NamedRerouteUsage()
  NamedRerouteDeclaration(None=OneMinus(None=Max(A=@n51, B=@n52))) {name=Full Gradient, nodeColor=<rgba 0 0 0 1>}
  n53 = VectorParameter() {group=2. Colors, parameterName=Background Color, sortPriority=0, defaultValue=<rgba 0 0.0024 0.0799 1>}
  n54 = LinearInterpolate(
   A=LinearInterpolate(A=NamedRerouteUsage(), B=@n53.RGB, Alpha=NamedRerouteUsage()) {constAlpha=1}
   B=NamedRerouteUsage()
   Alpha=NamedRerouteUsage()
  )
  n55 = VectorParameter() {group=2. Colors, parameterName=Pulse Color, sortPriority=7, defaultValue=<rgba 1 1 1 1>}
  n56 = NamedRerouteUsage()
  NamedRerouteDeclaration(None=LinearInterpolate(
    A=LinearInterpolate(A=@n54, B=@n55.RGB, Alpha=@n56)
    B=VectorParameter() {group=2. Colors, parameterName=End Cap Color, sortPriority=2, defaultValue=<rgba 1 1 1 1>}.RGB
    Alpha=NamedRerouteUsage()
   )) {name=Final Colored Shape, nodeColor=<rgba 0 0 0 1>}
  NamedRerouteDeclaration(None=Multiply(A=NamedRerouteUsage(), B=@n28)) {name=Skew for Vertical Ratio, nodeColor=<rgba 0 0 0 1>}
  n57 = VectorParameter() {group=2. Colors, parameterName=Background Color, sortPriority=0, defaultValue=<rgba 0 0.0024 0.0799 1>}
  n58 = LinearInterpolate(
   A=LinearInterpolate(
    A=@n57.RGB
    B=VectorParameter() {group=2. Colors, parameterName=Pulse Color, sortPriority=7, defaultValue=<rgba 1 1 1 1>}.RGB
    Alpha=NamedRerouteUsage()
   )
   B=VectorParameter() {group=2. Colors, parameterName=End Cap Color, sortPriority=2, defaultValue=<rgba 1 1 1 1>}.RGB
   Alpha=NamedRerouteUsage()
  )
  n59 = NamedRerouteUsage()
  n60 = NamedRerouteUsage()
  n61 = VectorParameter() {group=2. Colors, parameterName=Background Color, sortPriority=0, defaultValue=<rgba 0 0.0024 0.0799 1>}
  NamedRerouteDeclaration(None=LinearInterpolate(A=LinearInterpolate(A=@n58, B=@n59, Alpha=@n60), B=@n61.RGB, Alpha=NamedRerouteUsage())) {name=Transparency Color, nodeColor=<rgba 0 0 0 1>}
  NamedRerouteDeclaration(None=SmoothStep(Value=NamedRerouteUsage()) {constMax=0.002}) {name=Background, nodeColor=<rgba 0 0 0 1>}
  NamedRerouteDeclaration(None=SmoothStep(Value=NamedRerouteUsage()) {constMax=0.002}) {name=Progress, nodeColor=<rgba 0 0 0 1>}
  n62 = NamedRerouteUsage()
  n63 = NamedRerouteUsage()
  NamedRerouteDeclaration(None=SmoothStep(Value=Min(A=@n62, B=@n63)) {constMax=0.002}) {name=End Cap, nodeColor=<rgba 0 0 0 1>}
  n64 = VectorParameter() {group=2. Colors, parameterName=Progress Color 1, sortPriority=4, defaultValue=<rgba 0.0094 0.2489 0.5365 1>}
  n65 = VectorParameter() {group=2. Colors, parameterName=Progress Color 2, sortPriority=5, defaultValue=<rgba 0.2561 0.3835 0.5365 1>}
  n66 = MaterialFunctionCall(GradientStart=Constant() {r=0.25}, GradientEnd=Constant() {r=0.75}, RotationAngle=Constant()) {materialFunction=&MF_UI_LinearGradient}
  NamedRerouteDeclaration(None=Saturate(None=LinearInterpolate(A=@n64.RGB, B=@n65.RGB, Alpha=@n66.Result))) {name=Progress Color Gradient, nodeColor=<rgba 0 0 0 1>}
  n67 = DotProduct(A=@n2, B=@n1) {desc="The dot product gets you the angle between two vectors. \r\nbasically it'll give you a -1 -1 value :\r\nOr if the two vectors are parallel Or how perpendicular they are.\r\nAnd we are using that to get the distance between one edge (vector A) \r\nand the other edge (1) to use to create an angle."}
  NamedRerouteDeclaration(None=@n67) {name=Skew Angle End, nodeColor=<rgba 0 0 0 1>}
  n68 = Time()
  n69 = Constant() {desc=Speed, r=0.35}
  NamedRerouteDeclaration(None=Frac(None=Multiply(A=@n68, B=@n69) {constB=0.5})) {name=Animation, nodeColor=<rgba 1 0.5859 0 1>}
  NamedRerouteDeclaration(None=@n3) {name=Skew Angle Start, nodeColor=<rgba 0 0 0 1>}
  n70 = NamedRerouteUsage()
  n71 = Add(A=NamedRerouteUsage(), B=NamedRerouteUsage())
  NamedRerouteDeclaration(None=OneMinus(None=Max(A=@n70, B=@n71))) {name=Progress Gradient, nodeColor=<rgba 0 0 0 1>}
  NamedRerouteUsage()
  NamedRerouteUsage()
  NamedRerouteUsage()
  Saturate()
 ]
}
